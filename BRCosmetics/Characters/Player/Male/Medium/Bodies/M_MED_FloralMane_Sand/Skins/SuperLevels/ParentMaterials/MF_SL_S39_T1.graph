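# Material function for the first Super Level (S39 T1) skin of M_MED_FloralMane_Sand, restyling incoming material attributes with an animated water look.
# Mask channels drive a chain of attribute blends over the original material, with scalar parameters controlling the final layers.
# Pre-skinned local position anchors noise-distorted panning, curve-atlas water gradients and chromatic caustics.
# Refracted view rays and bump offsets add parallax depth inside the surface.
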MaterialFunction /BRCosmetics/Characters/Player/Male/Medium/Bodies/M_MED_FloralMane_Sand/Skins/SuperLevels/ParentMaterials/MF_SL_S39_T1
graph {expressions {
  MaterialExpressionAdd_0 {
   class=MaterialExpressionAdd
   inputs [
    A {from=MaterialExpressionMultiply_3}
    B {from=MaterialExpressionMultiply_5}
   ]
  }
  MaterialExpressionAdd_1 {
   class=MaterialExpressionAdd
   inputs [
    A {from=MaterialExpressionTextureCoordinate_0}
    B {from=MaterialExpressionMultiply_4}
   ]
  }
  MaterialExpressionAdd_10 {
   class=MaterialExpressionAdd
   inputs [
    A {from=MaterialExpressionNamedRerouteUsage_22}
    B {from=MaterialExpressionScalarParameter_10}
   ]
  }
  MaterialExpressionAdd_11 {
   class=MaterialExpressionAdd
   inputs [
    A {from=MaterialExpressionNamedRerouteUsage_24}
    B {from=MaterialExpressionScalarParameter_11}
   ]
  }
  MaterialExpressionAdd_12 {
   class=MaterialExpressionAdd
   inputs [
    A {from=MaterialExpressionAdd_6}
    B {from=MaterialExpressionMultiply_53}
   ]
   properties {constB=0}
  }
  MaterialExpressionAdd_13 {
   class=MaterialExpressionAdd
   inputs [
    A {from=MaterialExpressionNamedRerouteUsage_20}
    B {from=MaterialExpressionNamedRerouteUsage_25}
   ]
  }
  MaterialExpressionAdd_14 {
   class=MaterialExpressionAdd
   inputs [
    A {from=MaterialExpressionMultiply_44}
    B {from=MaterialExpressionNoise_0}
   ]
  }
  MaterialExpressionAdd_15 {
   class=MaterialExpressionAdd
   inputs [
    A {from=MaterialExpressionReroute_0}
    B {from=MaterialExpressionMultiply_25}
   ]
  }
  MaterialExpressionAdd_16 {
   class=MaterialExpressionAdd
   inputs [
    A {from=MaterialExpressionMultiply_26}
    B {from=MaterialExpressionNamedRerouteDeclaration_20}
   ]
  }
  MaterialExpressionAdd_17 {
   class=MaterialExpressionAdd
   inputs [
    A {from=MaterialExpressionShadingPathSwitch_0}
    B {from=MaterialExpressionMultiply_29}
   ]
  }
  MaterialExpressionAdd_18 {
   class=MaterialExpressionAdd
   inputs [
    A {from=MaterialExpressionDesaturation_0}
    B {from=MaterialExpressionOneMinus_6}
   ]
   properties {constB=0.5}
  }
  MaterialExpressionAdd_19 {
   class=MaterialExpressionAdd
   inputs [
    A {from=MaterialExpressionMultiply_30}
    B {from=MaterialExpressionMultiply_28}
   ]
  }
  MaterialExpressionAdd_2 {
   class=MaterialExpressionAdd
   inputs [
    A {from=MaterialExpressionPanner_3}
    B {from=MaterialExpressionConstantBiasScale_0}
   ]
  }
  MaterialExpressionAdd_20 {
   class=MaterialExpressionAdd
   inputs [
    A {from=MaterialExpressionMultiply_36}
    B {from=MaterialExpressionAdd_19}
   ]
  }
  MaterialExpressionAdd_21 {
   class=MaterialExpressionAdd
   inputs [
    A {from=MaterialExpressionNamedRerouteUsage_48}
    B {from=""}
   ]
   properties {constB=0.8}
  }
  MaterialExpressionAdd_22 {
   class=MaterialExpressionAdd
   inputs [
    A {from=MaterialExpressionMultiply_41}
    B {from=MaterialExpressionMultiply_43}
   ]
  }
  MaterialExpressionAdd_23 {
   class=MaterialExpressionAdd
   inputs [
    A {from=MaterialExpressionComponentMask_16}
    B {from=MaterialExpressionDotProduct_0}
   ]
  }
  MaterialExpressionAdd_24 {
   class=MaterialExpressionAdd
   inputs [
    A {from=MaterialExpressionMultiply_46}
    B {from=""}
   ]
   properties {constB=0.1}
  }
  MaterialExpressionAdd_25 {
   class=MaterialExpressionAdd
   inputs [
    A {from=MaterialExpressionPanner_2}
    B {from=MaterialExpressionMultiply_47}
   ]
  }
  MaterialExpressionAdd_26 {
   class=MaterialExpressionAdd
   inputs [
    A {from=MaterialExpressionNamedRerouteUsage_31}
    B {from=MaterialExpressionNamedRerouteUsage_32}
   ]
  }
  MaterialExpressionAdd_27 {
   class=MaterialExpressionAdd
   inputs [
    A {from=MaterialExpressionSaturate_8}
    B {from=MaterialExpressionMultiply_52}
   ]
  }
  MaterialExpressionAdd_28 {
   class=MaterialExpressionAdd
   inputs [
    A {from=MaterialExpressionMultiply_27}
    B {from=MaterialExpressionNamedRerouteUsage_33}
   ]
  }
  MaterialExpressionAdd_29 {
   class=MaterialExpressionAdd
   inputs [
    A {from=MaterialExpressionMultiply_54}
    B {from=MaterialExpressionAdd_28}
   ]
  }
  MaterialExpressionAdd_3 {
   class=MaterialExpressionAdd
   inputs [
    A {from=MaterialExpressionConvert_2}
    B {from=MaterialExpressionMultiply_8}
   ]
  }
  MaterialExpressionAdd_30 {
   class=MaterialExpressionAdd
   inputs [
    A {from=MaterialExpressionNamedRerouteUsage_43}
    B {from=""}
   ]
   properties {constB=-75}
  }
  MaterialExpressionAdd_31 {
   class=MaterialExpressionAdd
   inputs [
    A {from=MaterialExpressionNamedRerouteUsage_43}
    B {from=""}
   ]
   properties {constB=-77}
  }
  MaterialExpressionAdd_32 {
   class=MaterialExpressionAdd
   inputs [
    A {from=""}
    B {from=MaterialExpressionPower_1}
   ]
   properties {constA=1}
  }
  MaterialExpressionAdd_4 {
   class=MaterialExpressionAdd
   inputs [
    A {from=MaterialExpressionSmoothStep_0}
    B {from=MaterialExpressionSmoothStep_1}
   ]
  }
  MaterialExpressionAdd_5 {
   class=MaterialExpressionAdd
   inputs [
    A {from=MaterialExpressionOneMinus_0}
    B {from=""}
   ]
   properties {constB=0.3}
  }
  MaterialExpressionAdd_6 {
   class=MaterialExpressionAdd
   inputs [
    A {from=MaterialExpressionMultiply_16}
    B {from=MaterialExpressionLinearInterpolate_1}
   ]
  }
  MaterialExpressionAdd_7 {
   class=MaterialExpressionAdd
   inputs [
    A {from=MaterialExpressionTextureSampleParameter2D_0, output=G}
    B {from=MaterialExpressionTextureSampleParameter2D_5, output=B}
   ]
  }
  MaterialExpressionAdd_8 {
   class=MaterialExpressionAdd
   inputs [
    A {from=MaterialExpressionMultiply_13}
    B {from=MaterialExpressionMultiply_18}
   ]
   properties {constB=-12}
  }
  MaterialExpressionAdd_9 {
   class=MaterialExpressionAdd
   inputs [
    A {from=MaterialExpressionPanner_0}
    B {from=MaterialExpressionMultiply_17}
   ]
  }
  MaterialExpressionAppendVector_0 {
   class=MaterialExpressionAppendVector
   inputs [
    A {from=MaterialExpressionScalarParameter_12, output=""}
    B {from=MaterialExpressionScalarParameter_13, output=""}
   ]
  }
  MaterialExpressionAppendVector_1 {
   class=MaterialExpressionAppendVector
   inputs [
    A {from=MaterialExpressionScalarParameter_16, output=""}
    B {from=MaterialExpressionScalarParameter_17, output=""}
   ]
  }
  MaterialExpressionBlendMaterialAttributes_0 {
   class=MaterialExpressionBlendMaterialAttributes
   inputs [
    A {from=MaterialExpressionNamedRerouteUsage_1, output=""}
    B {from=MaterialExpressionSetMaterialAttributes_2, output=""}
    Alpha {from=MaterialExpressionComponentMask_0, output=""}
   ]
   properties {}
  }
  MaterialExpressionBlendMaterialAttributes_1 {
   class=MaterialExpressionBlendMaterialAttributes
   inputs [
    A {from=MaterialExpressionBlendMaterialAttributes_0, output=""}
    B {from=MaterialExpressionSetMaterialAttributes_1, output=""}
    Alpha {from=MaterialExpressionComponentMask_1, output=""}
   ]
   properties {}
  }
  MaterialExpressionBlendMaterialAttributes_2 {
   class=MaterialExpressionBlendMaterialAttributes
   inputs [
    A {from=MaterialExpressionBlendMaterialAttributes_1, output=""}
    B {from=MaterialExpressionSetMaterialAttributes_3, output=""}
    Alpha {from=MaterialExpressionComponentMask_2, output=""}
   ]
   properties {}
  }
  MaterialExpressionBlendMaterialAttributes_3 {
   class=MaterialExpressionBlendMaterialAttributes
   inputs [
    A {from=MaterialExpressionBlendMaterialAttributes_4, output=""}
    B {from=MaterialExpressionSetMaterialAttributes_4, output=""}
    Alpha {from=MaterialExpressionComponentMask_7, output=""}
   ]
   properties {}
  }
  MaterialExpressionBlendMaterialAttributes_4 {
   class=MaterialExpressionBlendMaterialAttributes
   inputs [
    A {from=MaterialExpressionBlendMaterialAttributes_2, output=""}
    B {from=MaterialExpressionSetMaterialAttributes_5, output=""}
    Alpha {from=MaterialExpressionMultiply_9, output=""}
   ]
   properties {}
  }
  MaterialExpressionBlendMaterialAttributes_5 {
   class=MaterialExpressionBlendMaterialAttributes
   inputs [
    A {from=MaterialExpressionBlendMaterialAttributes_3, output=""}
    B {from=MaterialExpressionSetMaterialAttributes_0, output=""}
    Alpha {from=MaterialExpressionLinearInterpolate_0, output=""}
   ]
   properties {}
  }
  MaterialExpressionBlendMaterialAttributes_6 {
   class=MaterialExpressionBlendMaterialAttributes
   inputs [
    A {from=MaterialExpressionBlendMaterialAttributes_5, output=""}
    B {from=MaterialExpressionSetMaterialAttributes_6, output=""}
    Alpha {from=MaterialExpressionScalarParameter_2, output=""}
   ]
   properties {}
  }
  MaterialExpressionBumpOffset_0 {
   class=MaterialExpressionBumpOffset
   inputs [
    Coordinate {from=MaterialExpressionAdd_9, output=""}
    Height {from=MaterialExpressionConstant_7, output=""}
    HeightRatioInput {from="", output=""}
   ]
   properties {}
  }
  MaterialExpressionBumpOffset_1 {
   class=MaterialExpressionBumpOffset
   inputs [
    Coordinate {from=MaterialExpressionPanner_2, output=""}
    Height {from=MaterialExpressionConstant_14, output=""}
    HeightRatioInput {from="", output=""}
   ]
   properties {}
  }
  MaterialExpressionCameraPositionWS_0 {class=MaterialExpressionCameraPositionWS, inputs [], properties {}}
  MaterialExpressionComponentMask_0 {class=MaterialExpressionComponentMask, inputs [None {from=MaterialExpressionNamedRerouteUsage_2}], properties {r=true}}
  MaterialExpressionComponentMask_1 {class=MaterialExpressionComponentMask, inputs [None {from=MaterialExpressionNamedRerouteUsage_3}], properties {g=true}}
  MaterialExpressionComponentMask_10 {class=MaterialExpressionComponentMask, inputs [None {from=MaterialExpressionNamedRerouteUsage_41}], properties {b=true}}
  MaterialExpressionComponentMask_11 {class=MaterialExpressionComponentMask, inputs [None {from=MaterialExpressionNamedRerouteUsage_49}], properties {g=true}}
  MaterialExpressionComponentMask_12 {class=MaterialExpressionComponentMask, inputs [None {from=MaterialExpressionNamedRerouteUsage_21}], properties {r=true}}
  MaterialExpressionComponentMask_15 {class=MaterialExpressionComponentMask, inputs [None {from=MaterialExpressionNamedRerouteUsage_71}], properties {b=true, r=true}}
  MaterialExpressionComponentMask_16 {class=MaterialExpressionComponentMask, inputs [None {from=MaterialExpressionNamedRerouteUsage_27}], properties {b=true}}
  MaterialExpressionComponentMask_17 {class=MaterialExpressionComponentMask, inputs [None {from=MaterialExpressionReflectionVectorWS_1}], properties {b=true, r=true}}
  MaterialExpressionComponentMask_18 {class=MaterialExpressionComponentMask, inputs [None {from=MaterialExpressionNamedRerouteUsage_30}], properties {b=true, r=true}}
  MaterialExpressionComponentMask_2 {class=MaterialExpressionComponentMask, inputs [None {from=MaterialExpressionNamedRerouteUsage_4}], properties {b=true}}
  MaterialExpressionComponentMask_3 {class=MaterialExpressionComponentMask, inputs [None {from=MaterialExpressionNamedRerouteUsage_6}], properties {b=true}}
  MaterialExpressionComponentMask_4 {class=MaterialExpressionComponentMask, inputs [None {from=MaterialExpressionNamedRerouteUsage_7}], properties {g=true}}
  MaterialExpressionComponentMask_5 {class=MaterialExpressionComponentMask, inputs [None {from=MaterialExpressionReflectionVectorWS_0}], properties {b=true, r=true}}
  MaterialExpressionComponentMask_6 {class=MaterialExpressionComponentMask, inputs [None {from=MaterialExpressionNamedRerouteUsage_23}], properties {g=true}}
  MaterialExpressionComponentMask_7 {class=MaterialExpressionComponentMask, inputs [None {from=MaterialExpressionNamedRerouteUsage_15}], properties {g=true}}
  MaterialExpressionComponentMask_8 {class=MaterialExpressionComponentMask, inputs [None {from=MaterialExpressionNamedRerouteUsage_16}], properties {r=true}}
  MaterialExpressionComponentMask_9 {class=MaterialExpressionComponentMask, inputs [None {from=MaterialExpressionNamedRerouteUsage_18}], properties {b=true, r=true}}
  MaterialExpressionConstant2Vector_0 {class=MaterialExpressionConstant2Vector, properties {r=1}}
  MaterialExpressionConstant2Vector_1 {class=MaterialExpressionConstant2Vector, properties {g=-0.25}}
  MaterialExpressionConstant2Vector_2 {class=MaterialExpressionConstant2Vector, properties {g=0.015}}
  MaterialExpressionConstant3Vector_0 {class=MaterialExpressionConstant3Vector, properties {constant=<rgba 0 0 -1 0>}}
  MaterialExpressionConstant3Vector_1 {class=MaterialExpressionConstant3Vector, properties {constant=<rgba 0 0 1 0>}}
  MaterialExpressionConstant3Vector_10 {class=MaterialExpressionConstant3Vector, properties {constant=<rgba 0 0.4582 1 1>}}
  MaterialExpressionConstant3Vector_11 {class=MaterialExpressionConstant3Vector, properties {constant=<rgba 0 0 1 0>}}
  MaterialExpressionConstant3Vector_12 {class=MaterialExpressionConstant3Vector, properties {constant=<rgba 0.2268 0.8609 1 1>}}
  MaterialExpressionConstant3Vector_15 {class=MaterialExpressionConstant3Vector, properties {constant=<rgba 0.5823 1.2 1.1195 1>}}
  MaterialExpressionConstant3Vector_2 {class=MaterialExpressionConstant3Vector, properties {constant=<rgba 0 1 1 0>}}
  MaterialExpressionConstant3Vector_3 {class=MaterialExpressionConstant3Vector, properties {constant=<rgba 0.0208 0.0448 0.0677 1>}}
  MaterialExpressionConstant3Vector_5 {class=MaterialExpressionConstant3Vector, properties {constant=<rgba 1 1 1 0>}}
  MaterialExpressionConstant3Vector_6 {class=MaterialExpressionConstant3Vector, properties {constant=<rgba 0.0058 0.1729 0.3316 1>}}
  MaterialExpressionConstant3Vector_8 {class=MaterialExpressionConstant3Vector, properties {constant=<rgba 0 0 1 0>}}
  MaterialExpressionConstant3Vector_9 {class=MaterialExpressionConstant3Vector, properties {constant=<rgba 0.2268 0.8609 1 1>}}
  MaterialExpressionConstantBiasScale_0 {class=MaterialExpressionConstantBiasScale, inputs [None {from=MaterialExpressionSine_0}], properties {scale=0.1}}
  MaterialExpressionConstant_0 {class=MaterialExpressionConstant}
  MaterialExpressionConstant_1 {class=MaterialExpressionConstant, properties {r=1}}
  MaterialExpressionConstant_10 {class=MaterialExpressionConstant, properties {r=-1}}
  MaterialExpressionConstant_11 {class=MaterialExpressionConstant, properties {r=1}}
  MaterialExpressionConstant_12 {class=MaterialExpressionConstant, properties {r=1}}
  MaterialExpressionConstant_13 {class=MaterialExpressionConstant, properties {r=1.33}}
  MaterialExpressionConstant_14 {class=MaterialExpressionConstant, properties {r=-1.5}}
  MaterialExpressionConstant_15 {class=MaterialExpressionConstant, properties {r=-1}}
  MaterialExpressionConstant_2 {class=MaterialExpressionConstant, properties {r=1}}
  MaterialExpressionConstant_3 {class=MaterialExpressionConstant, properties {r=0.025}}
  MaterialExpressionConstant_4 {class=MaterialExpressionConstant, properties {r=0.15}}
  MaterialExpressionConstant_5 {class=MaterialExpressionConstant, properties {r=0.5}}
  MaterialExpressionConstant_6 {class=MaterialExpressionConstant, properties {r=-1}}
  MaterialExpressionConstant_7 {class=MaterialExpressionConstant, properties {r=-2}}
  MaterialExpressionConstant_8 {class=MaterialExpressionConstant, properties {r=1}}
  MaterialExpressionConstant_9 {class=MaterialExpressionConstant, properties {r=1.33}}
  MaterialExpressionConvert_0 {
   class=MaterialExpressionConvert
   inputs [None {from=MaterialExpressionMaterialFunctionCall_0, output=XY Texture}]
   properties {
    nodeName=Break Float3
    convertInputs [{
      type=Vector3
      expressionInput {
       expression=&self:MaterialExpressionMaterialFunctionCall_0
       inputName=None
       mask=0
       maskA=0
       maskB=0
       maskG=0
       maskR=0
       outputIndex=0
      }
     }]
    convertOutputs [{type=Scalar}, {type=Scalar}, {type=Scalar}]
   }
  }
  MaterialExpressionConvert_2 {
   class=MaterialExpressionConvert
   inputs [None {from=MaterialExpressionDivide_3}]
   properties {
    nodeName=Break Float3
    convertInputs [{
      type=Vector3
      expressionInput {
       expression=&self:MaterialExpressionDivide_3
       inputName=None
       mask=0
       maskA=0
       maskB=0
       maskG=0
       maskR=0
       outputIndex=0
      }
     }]
    convertOutputs [{type=Scalar}, {type=Scalar}, {type=Scalar}]
   }
  }
  MaterialExpressionConvert_3 {
   class=MaterialExpressionConvert
   inputs [
    None {from=MaterialExpressionConvert_2}
    None {from=MaterialExpressionConvert_2}
   ]
   properties {
    nodeName=Make Vector2
    convertInputs [
     {
      type=Scalar
      expressionInput {
       expression=&self:MaterialExpressionConvert_2
       inputName=None
       mask=0
       maskA=0
       maskB=0
       maskG=0
       maskR=0
       outputIndex=0
      }
     }
     {
      type=Scalar
      expressionInput {
       expression=&self:MaterialExpressionConvert_2
       inputName=None
       mask=0
       maskA=0
       maskB=0
       maskG=0
       maskR=0
       outputIndex=1
      }
     }
    ]
    convertOutputs [{type=Vector2}]
   }
  }
  MaterialExpressionCurveAtlasRowParameter_0 {
   class=MaterialExpressionCurveAtlasRowParameter
   inputs [CurveTime {from=MaterialExpressionAdd_16, output=""}]
   properties {
    atlas=/BRCosmetics/Characters/Player/Male/Medium/Bodies/M_MED_FloralMane_Sand/Skins/SuperLevels/Textures/CA_SL_S39
    curve=/BRCosmetics/Characters/Player/Male/Medium/Bodies/M_MED_FloralMane_Sand/Skins/SuperLevels/Textures/C_SL_S39_WaterGradient
    parameterName=PrimaryElement_Gradient
   }
  }
  MaterialExpressionCurveAtlasRowParameter_1 {
   class=MaterialExpressionCurveAtlasRowParameter
   inputs [CurveTime {from=MaterialExpressionConvert_0, output=""}]
   properties {
    atlas=&CA_SL_S39
    curve=/BRCosmetics/Characters/Player/Male/Medium/Bodies/M_MED_FloralMane_Sand/Skins/SuperLevels/Textures/C_SL_S39_CausticsChromatic
    defaultValue=1
    parameterName=PrimaryElement_CausticsChromatic
   }
  }
  MaterialExpressionDesaturation_0 {
   class=MaterialExpressionDesaturation
   inputs [
    None {from=MaterialExpressionGetMaterialAttributes_0, output=BaseColor}
    Fraction {from=MaterialExpressionConstant_1, output=""}
   ]
   properties {}
  }
  MaterialExpressionDesaturation_1 {
   class=MaterialExpressionDesaturation
   inputs [
    None {from=MaterialExpressionGetMaterialAttributes_3, output=BaseColor}
    Fraction {from=MaterialExpressionConstant_11, output=""}
   ]
   properties {}
  }
  MaterialExpressionDivide_0 {
   class=MaterialExpressionDivide
   inputs [
    A {from=MaterialExpressionComponentMask_9}
    B {from=MaterialExpressionScalarParameter_14}
   ]
  }
  MaterialExpressionDivide_1 {
   class=MaterialExpressionDivide
   inputs [
    A {from=MaterialExpressionComponentMask_15}
    B {from=MaterialExpressionScalarParameter_39}
   ]
  }
  MaterialExpressionDivide_2 {
   class=MaterialExpressionDivide
   inputs [
    A {from=MaterialExpressionComponentMask_18}
    B {from=MaterialExpressionScalarParameter_18}
   ]
  }
  MaterialExpressionDivide_3 {
   class=MaterialExpressionDivide
   inputs [
    A {from=MaterialExpressionNamedRerouteUsage_44}
    B {from=""}
   ]
   properties {constB=64}
  }
  MaterialExpressionDivide_4 {
   class=MaterialExpressionDivide
   inputs [
    A {from=MaterialExpressionAdd_31}
    B {from=""}
   ]
   properties {constB=100}
  }
  MaterialExpressionDotProduct_0 {
   class=MaterialExpressionDotProduct
   inputs [
    A {from=MaterialExpressionNamedRerouteUsage_5, output=""}
    B {from=MaterialExpressionConstant3Vector_5, output=""}
   ]
   properties {}
  }
  MaterialExpressionDotProduct_1 {
   class=MaterialExpressionDotProduct
   inputs [
    A {from=MaterialExpressionNamedRerouteUsage_12, output=""}
    B {from=MaterialExpressionConstant3Vector_2, output=""}
   ]
   properties {}
  }
  MaterialExpressionFresnel_0 {
   class=MaterialExpressionFresnel
   inputs [
    ExponentIn {from="", output=""}
    BaseReflectFractionIn {from="", output=""}
    Normal {from="", output=""}
   ]
   properties {baseReflectFraction=0, exponent=1}
  }
  MaterialExpressionFunctionInput_0 {
   class=MaterialExpressionFunctionInput
   inputs [Preview {from="", output=""}]
   properties {inputName=Material, inputType=FunctionInput_MaterialAttributes}
  }
  MaterialExpressionFunctionInput_1 {class=MaterialExpressionFunctionInput, inputs [Preview {from="", output=""}], properties {inputName=M, sortPriority=1}}
  MaterialExpressionFunctionInput_2 {
   class=MaterialExpressionFunctionInput
   inputs [Preview {from=MaterialExpressionShaderStageSwitch_0, output=""}]
   properties {inputName=PreskinnedLocPos, sortPriority=2}
  }
  MaterialExpressionFunctionInput_3 {class=MaterialExpressionFunctionInput, inputs [Preview {from="", output=""}], properties {inputName=Tech Art Mask, sortPriority=3}}
  MaterialExpressionFunctionInput_4 {
   class=MaterialExpressionFunctionInput
   inputs [Preview {from=MaterialExpressionFresnel_0, output=""}]
   properties {inputName=Fresnel, inputType=FunctionInput_Scalar, sortPriority=4}
  }
  MaterialExpressionFunctionOutput_0 {class=MaterialExpressionFunctionOutput, inputs [None {from=MaterialExpressionNamedRerouteUsage_14, output=""}]}
  MaterialExpressionGetMaterialAttributes_0 {
   class=MaterialExpressionGetMaterialAttributes
   inputs [None {from=MaterialExpressionBlendMaterialAttributes_0, output=""}]
   properties {attributeGetTypes [#1, #2]}
  }
  MaterialExpressionGetMaterialAttributes_1 {
   class=MaterialExpressionGetMaterialAttributes
   inputs [None {from=MaterialExpressionNamedRerouteUsage_17, output=""}]
   properties {attributeGetTypes [#3]}
  }
  MaterialExpressionGetMaterialAttributes_2 {
   class=MaterialExpressionGetMaterialAttributes
   inputs [None {from=MaterialExpressionNamedRerouteUsage_1, output=""}]
   properties {attributeGetTypes [#1, #3]}
  }
  MaterialExpressionGetMaterialAttributes_3 {
   class=MaterialExpressionGetMaterialAttributes
   inputs [None {from=MaterialExpressionBlendMaterialAttributes_3, output=""}]
   properties {attributeGetTypes [#2]}
  }
  MaterialExpressionGetMaterialAttributes_4 {
   class=MaterialExpressionGetMaterialAttributes
   inputs [None {from=MaterialExpressionNamedRerouteUsage_29, output=""}]
   properties {attributeGetTypes [#3]}
  }
  MaterialExpressionLength_0 {class=MaterialExpressionLength, inputs [None {from=MaterialExpressionMaterialFunctionCall_6, output=Vector}], properties {}}
  MaterialExpressionLength_1 {class=MaterialExpressionLength, inputs [None {from=MaterialExpressionMaterialFunctionCall_2, output=Vector}], properties {}}
  MaterialExpressionLinearInterpolate_0 {
   class=MaterialExpressionLinearInterpolate
   inputs [
    A {from=MaterialExpressionComponentMask_8}
    B {from=MaterialExpressionMultiply_7}
    Alpha {from=MaterialExpressionScalarParameter_3}
   ]
  }
  MaterialExpressionLinearInterpolate_1 {
   class=MaterialExpressionLinearInterpolate
   inputs [
    A {from=MaterialExpressionMultiply_11}
    B {from=MaterialExpressionMultiply_10}
    Alpha {from=MaterialExpressionPower_0}
   ]
  }
  MaterialExpressionLinearInterpolate_10 {
   class=MaterialExpressionLinearInterpolate
   inputs [
    A {from=MaterialExpressionLinearInterpolate_12}
    B {from=""}
    Alpha {from=MaterialExpressionNamedRerouteUsage_40}
   ]
   properties {constB=0.75}
  }
  MaterialExpressionLinearInterpolate_11 {
   class=MaterialExpressionLinearInterpolate
   inputs [
    A {from=MaterialExpressionScalarParameter_21}
    B {from=""}
    Alpha {from=MaterialExpressionNamedRerouteUsage_9}
   ]
   properties {constB=0.5}
  }
  MaterialExpressionLinearInterpolate_12 {
   class=MaterialExpressionLinearInterpolate
   inputs [
    A {from=""}
    B {from=MaterialExpressionScalarParameter_22}
    Alpha {from=MaterialExpressionNamedRerouteUsage_45}
   ]
   properties {constA=0.8}
  }
  MaterialExpressionLinearInterpolate_13 {
   class=MaterialExpressionLinearInterpolate
   inputs [
    A {from=MaterialExpressionGetMaterialAttributes_2, output=Normal}
    B {from=MaterialExpressionConstant3Vector_1}
    Alpha {from=MaterialExpressionNamedRerouteUsage_47}
   ]
  }
  MaterialExpressionLinearInterpolate_2 {
   class=MaterialExpressionLinearInterpolate
   inputs [
    A {from=MaterialExpressionLinearInterpolate_5}
    B {from=MaterialExpressionSaturate_11}
    Alpha {from=MaterialExpressionMultiply_35}
   ]
  }
  MaterialExpressionLinearInterpolate_3 {
   class=MaterialExpressionLinearInterpolate
   inputs [
    A {from=MaterialExpressionSaturate_2}
    B {from=MaterialExpressionSaturate_13}
    Alpha {from=MaterialExpressionScalarParameter_5}
   ]
  }
  MaterialExpressionLinearInterpolate_4 {
   class=MaterialExpressionLinearInterpolate
   inputs [
    A {from=MaterialExpressionConstant3Vector_3}
    B {from=MaterialExpressionConstant3Vector_6}
    Alpha {from=MaterialExpressionLinearInterpolate_2}
   ]
  }
  MaterialExpressionLinearInterpolate_5 {
   class=MaterialExpressionLinearInterpolate
   inputs [
    A {from=MaterialExpressionOneMinus_4}
    B {from=MaterialExpressionSaturate_3}
    Alpha {from=MaterialExpressionScalarParameter_8}
   ]
  }
  MaterialExpressionLinearInterpolate_7 {
   class=MaterialExpressionLinearInterpolate
   inputs [
    A {from=MaterialExpressionMultiply_11}
    B {from=MaterialExpressionMultiply_10}
    Alpha {from=MaterialExpressionNamedRerouteUsage_37}
   ]
  }
  MaterialExpressionLinearInterpolate_8 {
   class=MaterialExpressionLinearInterpolate
   inputs [
    A {from=MaterialExpressionScalarParameter_20}
    B {from=""}
    Alpha {from=MaterialExpressionNamedRerouteUsage_38}
   ]
   properties {constB=0}
  }
  MaterialExpressionLinearInterpolate_9 {
   class=MaterialExpressionLinearInterpolate
   inputs [
    A {from=MaterialExpressionMultiply_38}
    B {from=""}
    Alpha {from=MaterialExpressionNamedRerouteUsage_39}
   ]
   properties {constB=0}
  }
  MaterialExpressionLocalPosition_0 {class=MaterialExpressionLocalPosition, inputs [], properties {localOrigin=InstancePreSkinning}}
  MaterialExpressionMaterialFunctionCall_0 {
   class=MaterialExpressionMaterialFunctionCall
   inputs [
    TextureObject {from=MaterialExpressionTextureObject_0}
    TextureSize {from=MaterialExpressionScalarParameter_23}
    WorldPosition {from=MaterialExpressionAdd_17}
    FadingGradient {from=""}
    In {from=MaterialExpressionConstant2Vector_2}
    World Space Normal {from=""}
    ProjectionTransitionContrast {from=MaterialExpressionConstant_0}
   ]
   properties {materialFunction=&/Preskinned_WP_Texture_ScalePan_XY}
  }
  MaterialExpressionMaterialFunctionCall_10 {
   class=MaterialExpressionMaterialFunctionCall
   inputs [
    Normal {from=MaterialExpressionTextureSample_1, output=RGB}
    Flatness {from=MaterialExpressionConstant_15}
   ]
   properties {materialFunction=&FlattenNormal}
  }
  MaterialExpressionMaterialFunctionCall_2 {
   class=MaterialExpressionMaterialFunctionCall
   inputs [
    Ray Direction {from=MaterialExpressionTransform_3}
    Surface Normal {from=MaterialExpressionMultiply_48}
    Refractive Index Origin {from=MaterialExpressionConstant_12}
    Refractive Index Target {from=MaterialExpressionConstant_13}
   ]
   properties {materialFunction=/Engine/Functions/Engine_MaterialFunctions01/Vectors/Refract}
  }
  MaterialExpressionMaterialFunctionCall_4 {
   class=MaterialExpressionMaterialFunctionCall
   inputs [
    BaseNormal {from=MaterialExpressionGetMaterialAttributes_2, output=Normal}
    AdditionalNormal {from=MaterialExpressionNamedRerouteUsage_70}
   ]
   properties {materialFunction=&BlendAngleCorrectedNormals}
  }
  MaterialExpressionMaterialFunctionCall_6 {
   class=MaterialExpressionMaterialFunctionCall
   inputs [
    Ray Direction {from=MaterialExpressionTransform_1}
    Surface Normal {from=MaterialExpressionMultiply_20}
    Refractive Index Origin {from=MaterialExpressionConstant_8}
    Refractive Index Target {from=MaterialExpressionConstant_9}
   ]
   properties {materialFunction=&Refract}
  }
  MaterialExpressionMax_0 {
   class=MaterialExpressionMax
   inputs [
    A {from=MaterialExpressionComponentMask_11}
    B {from=""}
   ]
   properties {constB=0.75}
  }
  MaterialExpressionMax_1 {
   class=MaterialExpressionMax
   inputs [
    A {from=MaterialExpressionComponentMask_6}
    B {from=""}
   ]
   properties {constB=0.5}
  }
  MaterialExpressionMin_0 {
   class=MaterialExpressionMin
   inputs [
    A {from=MaterialExpressionGetMaterialAttributes_0, output=Roughness}
    B {from="", output=""}
   ]
   properties {constB=0.45}
  }
  MaterialExpressionMultiply_0 {
   class=MaterialExpressionMultiply
   inputs [
    A {from=MaterialExpressionConstant3Vector_0}
    B {from=MaterialExpressionTime_0}
   ]
  }
  MaterialExpressionMultiply_1 {
   class=MaterialExpressionMultiply
   inputs [
    A {from=MaterialExpressionAdd_14}
    B {from=MaterialExpressionConstant2Vector_0}
   ]
  }
  MaterialExpressionMultiply_10 {
   class=MaterialExpressionMultiply
   inputs [
    A {from=""}
    B {from=MaterialExpressionNamedRerouteDeclaration_18}
   ]
   properties {constA=1.5}
  }
  MaterialExpressionMultiply_11 {
   class=MaterialExpressionMultiply
   inputs [
    A {from=MaterialExpressionNamedRerouteDeclaration_18}
    B {from=""}
   ]
   properties {constB=0.35}
  }
  MaterialExpressionMultiply_12 {
   class=MaterialExpressionMultiply
   inputs [
    A {from=MaterialExpressionAdd_12}
    B {from=MaterialExpressionSaturate_1}
   ]
  }
  MaterialExpressionMultiply_13 {
   class=MaterialExpressionMultiply
   inputs [
    A {from=MaterialExpressionNamedRerouteUsage_43}
    B {from=MaterialExpressionScalarParameter_6}
   ]
   properties {constB=0.1}
  }
  MaterialExpressionMultiply_14 {
   class=MaterialExpressionMultiply
   inputs [
    A {from=MaterialExpressionAdd_4}
    B {from=MaterialExpressionAdd_4}
   ]
  }
  MaterialExpressionMultiply_15 {
   class=MaterialExpressionMultiply
   inputs [
    A {from=MaterialExpressionSphereMask_0}
    B {from=MaterialExpressionSphereMask_0}
   ]
  }
  MaterialExpressionMultiply_16 {
   class=MaterialExpressionMultiply
   inputs [
    A {from=MaterialExpressionNamedRerouteUsage_10}
    B {from=""}
   ]
   properties {constB=0.5}
  }
  MaterialExpressionMultiply_17 {
   class=MaterialExpressionMultiply
   inputs [
    A {from=MaterialExpressionComponentMask_5}
    B {from=MaterialExpressionMultiply_21}
   ]
  }
  MaterialExpressionMultiply_18 {
   class=MaterialExpressionMultiply
   inputs [
    A {from=MaterialExpressionScalarParameter_6}
    B {from=MaterialExpressionScalarParameter_7}
   ]
  }
  MaterialExpressionMultiply_19 {
   class=MaterialExpressionMultiply
   inputs [
    A {from=MaterialExpressionDesaturation_1}
    B {from=""}
   ]
   properties {constB=0.1}
  }
  MaterialExpressionMultiply_2 {
   class=MaterialExpressionMultiply
   inputs [
    A {from=MaterialExpressionMultiply_1}
    B {from=""}
   ]
   properties {constB=0.0175}
  }
  MaterialExpressionMultiply_20 {
   class=MaterialExpressionMultiply
   inputs [
    A {from=MaterialExpressionTransform_0}
    B {from=MaterialExpressionConstant_6}
   ]
  }
  MaterialExpressionMultiply_21 {
   class=MaterialExpressionMultiply
   inputs [
    A {from=MaterialExpressionLength_0}
    B {from=MaterialExpressionScalarParameter_15}
   ]
  }
  MaterialExpressionMultiply_22 {
   class=MaterialExpressionMultiply
   inputs [
    A {from=MaterialExpressionConstant3Vector_9}
    B {from=MaterialExpressionTextureSample_0, output=R}
   ]
  }
  MaterialExpressionMultiply_23 {
   class=MaterialExpressionMultiply
   inputs [
    A {from=MaterialExpressionMultiply_22}
    B {from=""}
   ]
   properties {constB=0.7}
  }
  MaterialExpressionMultiply_24 {
   class=MaterialExpressionMultiply
   inputs [
    A {from=MaterialExpressionNamedRerouteUsage_26}
    B {from=MaterialExpressionPower_3}
   ]
  }
  MaterialExpressionMultiply_25 {
   class=MaterialExpressionMultiply
   inputs [
    A {from=MaterialExpressionNormalize_0}
    B {from=MaterialExpressionScalarParameter_24}
   ]
   properties {constB=12}
  }
  MaterialExpressionMultiply_26 {
   class=MaterialExpressionMultiply
   inputs [
    A {from=MaterialExpressionNamedRerouteUsage_13}
    B {from=""}
   ]
   properties {constB=0.3}
  }
  MaterialExpressionMultiply_27 {
   class=MaterialExpressionMultiply
   inputs [
    A {from=MaterialExpressionMultiply_24}
    B {from=MaterialExpressionConstant3Vector_10}
   ]
  }
  MaterialExpressionMultiply_28 {
   class=MaterialExpressionMultiply
   inputs [
    A {from=MaterialExpressionConvert_0}
    B {from=MaterialExpressionMultiply_34}
   ]
  }
  MaterialExpressionMultiply_29 {
   class=MaterialExpressionMultiply
   inputs [
    A {from=MaterialExpressionNamedRerouteDeclaration_16}
    B {from=MaterialExpressionScalarParameter_25}
   ]
   properties {constB=5}
  }
  MaterialExpressionMultiply_3 {
   class=MaterialExpressionMultiply
   inputs [
    A {from=""}
    B {from=MaterialExpressionNamedRerouteUsage_0}
   ]
   properties {constA=0.01}
  }
  MaterialExpressionMultiply_30 {
   class=MaterialExpressionMultiply
   inputs [
    A {from=MaterialExpressionCurveAtlasRowParameter_1}
    B {from=MaterialExpressionConvert_0}
   ]
  }
  MaterialExpressionMultiply_31 {
   class=MaterialExpressionMultiply
   inputs [
    A {from=MaterialExpressionVectorParameter_1, output=RGB}
    B {from=MaterialExpressionScalarParameter_29}
   ]
   properties {constB=1.5}
  }
  MaterialExpressionMultiply_32 {
   class=MaterialExpressionMultiply
   inputs [
    A {from=MaterialExpressionMax_1}
    B {from=MaterialExpressionMultiply_12}
   ]
  }
  MaterialExpressionMultiply_33 {
   class=MaterialExpressionMultiply
   inputs [
    A {from=MaterialExpressionMax_0}
    B {from=MaterialExpressionLinearInterpolate_9}
   ]
  }
  MaterialExpressionMultiply_34 {
   class=MaterialExpressionMultiply
   inputs [
    A {from=MaterialExpressionConstant3Vector_15}
    B {from=MaterialExpressionNamedRerouteUsage_42}
   ]
  }
  MaterialExpressionMultiply_35 {
   class=MaterialExpressionMultiply
   inputs [
    A {from=MaterialExpressionSaturate_10}
    B {from=MaterialExpressionScalarParameter_4}
   ]
  }
  MaterialExpressionMultiply_36 {
   class=MaterialExpressionMultiply
   inputs [
    A {from=MaterialExpressionCurveAtlasRowParameter_1}
    B {from=MaterialExpressionScalarParameter_0}
   ]
   properties {constB=0.2}
  }
  MaterialExpressionMultiply_37 {
   class=MaterialExpressionMultiply
   inputs [
    A {from=MaterialExpressionLinearInterpolate_4}
    B {from=""}
   ]
   properties {constB=0.25}
  }
  MaterialExpressionMultiply_38 {
   class=MaterialExpressionMultiply
   inputs [
    A {from=MaterialExpressionAdd_21}
    B {from=MaterialExpressionAdd_29}
   ]
  }
  MaterialExpressionMultiply_39 {
   class=MaterialExpressionMultiply
   inputs [
    A {from=MaterialExpressionMultiply_31}
    B {from=MaterialExpressionAdd_32}
   ]
  }
  MaterialExpressionMultiply_4 {
   class=MaterialExpressionMultiply
   inputs [
    A {from=MaterialExpressionMultiply_2}
    B {from=MaterialExpressionTextureSampleParameter2D_4, output=R}
   ]
  }
  MaterialExpressionMultiply_41 {
   class=MaterialExpressionMultiply
   inputs [
    A {from=MaterialExpressionNamedRerouteUsage_0}
    B {from=""}
   ]
   properties {constB=0.1}
  }
  MaterialExpressionMultiply_43 {
   class=MaterialExpressionMultiply
   inputs [
    A {from=""}
    B {from=MaterialExpressionMultiply_0}
   ]
   properties {constA=0.5}
  }
  MaterialExpressionMultiply_44 {
   class=MaterialExpressionMultiply
   inputs [
    A {from=MaterialExpressionNoise_1}
    B {from=""}
   ]
   properties {constB=0.25}
  }
  MaterialExpressionMultiply_45 {
   class=MaterialExpressionMultiply
   inputs [
    A {from=MaterialExpressionVectorParameter_2, output=RGB}
    B {from=MaterialExpressionSaturate_14}
   ]
  }
  MaterialExpressionMultiply_46 {
   class=MaterialExpressionMultiply
   inputs [
    A {from=MaterialExpressionOneMinus_5}
    B {from=MaterialExpressionOneMinus_5}
   ]
  }
  MaterialExpressionMultiply_47 {
   class=MaterialExpressionMultiply
   inputs [
    A {from=MaterialExpressionComponentMask_17}
    B {from=MaterialExpressionMultiply_49}
   ]
  }
  MaterialExpressionMultiply_48 {
   class=MaterialExpressionMultiply
   inputs [
    A {from=MaterialExpressionTransform_2}
    B {from=MaterialExpressionConstant_10}
   ]
  }
  MaterialExpressionMultiply_49 {
   class=MaterialExpressionMultiply
   inputs [
    A {from=MaterialExpressionLength_1}
    B {from=MaterialExpressionScalarParameter_19}
   ]
  }
  MaterialExpressionMultiply_5 {
   class=MaterialExpressionMultiply
   inputs [
    A {from=MaterialExpressionMultiply_0}
    B {from=""}
   ]
   properties {constB=0.232}
  }
  MaterialExpressionMultiply_50 {
   class=MaterialExpressionMultiply
   inputs [
    A {from=MaterialExpressionConstant3Vector_12}
    B {from=MaterialExpressionTextureSample_3, output=R}
   ]
  }
  MaterialExpressionMultiply_51 {
   class=MaterialExpressionMultiply
   inputs [
    A {from=MaterialExpressionMultiply_50}
    B {from=""}
   ]
   properties {constB=0.7}
  }
  MaterialExpressionMultiply_52 {
   class=MaterialExpressionMultiply
   inputs [
    A {from=MaterialExpressionSaturate_16}
    B {from=""}
   ]
   properties {constB=0.6}
  }
  MaterialExpressionMultiply_53 {
   class=MaterialExpressionMultiply
   inputs [
    A {from=MaterialExpressionNamedRerouteUsage_19}
    B {from=""}
   ]
   properties {constB=0.3}
  }
  MaterialExpressionMultiply_54 {
   class=MaterialExpressionMultiply
   inputs [
    A {from=MaterialExpressionLinearInterpolate_7}
    B {from=""}
   ]
   properties {constB=0.7}
  }
  MaterialExpressionMultiply_6 {
   class=MaterialExpressionMultiply
   inputs [
    A {from=MaterialExpressionSaturate_0}
    B {from=MaterialExpressionMultiply_39}
   ]
  }
  MaterialExpressionMultiply_7 {
   class=MaterialExpressionMultiply
   inputs [
    A {from=MaterialExpressionComponentMask_8}
    B {from=MaterialExpressionSaturate_9}
   ]
  }
  MaterialExpressionMultiply_8 {
   class=MaterialExpressionMultiply
   inputs [
    A {from=MaterialExpressionTime_1}
    B {from=""}
   ]
   properties {constB=-0.3333}
  }
  MaterialExpressionMultiply_9 {
   class=MaterialExpressionMultiply
   inputs [
    A {from=MaterialExpressionOneMinus_1}
    B {from=MaterialExpressionOneMinus_2}
   ]
  }
  MaterialExpressionNamedRerouteDeclaration_0 {
   class=MaterialExpressionNamedRerouteDeclaration
   inputs [None {from=MaterialExpressionFunctionInput_2}]
   properties {name=PreskinnedLocalPosition, nodeColor=<rgba 1 0.8672 0 1>}
  }
  MaterialExpressionNamedRerouteDeclaration_1 {
   class=MaterialExpressionNamedRerouteDeclaration
   inputs [None {from=MaterialExpressionFunctionInput_0}]
   properties {name=OriginalMA, nodeColor=<rgba 0.5469 0 1 1>}
  }
  MaterialExpressionNamedRerouteDeclaration_10 {
   class=MaterialExpressionNamedRerouteDeclaration
   inputs [None {from=MaterialExpressionFunctionInput_3}]
   properties {name=InTechArtMask, nodeColor=<rgba 0.1484 0 1 1>}
  }
  MaterialExpressionNamedRerouteDeclaration_11 {
   class=MaterialExpressionNamedRerouteDeclaration
   inputs [None {from=MaterialExpressionGetMaterialAttributes_2, output=Roughness}]
   properties {name=OriginalRoughness, nodeColor=<rgba 0 1 1 1>}
  }
  MaterialExpressionNamedRerouteDeclaration_12 {
   class=MaterialExpressionNamedRerouteDeclaration
   inputs [None {from=MaterialExpressionMultiply_33}]
   properties {name=PrimaryEmissive, nodeColor=<rgba 1 0.3516 0 1>}
  }
  MaterialExpressionNamedRerouteDeclaration_13 {
   class=MaterialExpressionNamedRerouteDeclaration
   inputs [None {from=MaterialExpressionMultiply_32}]
   properties {name=PrimaryBasecolor, nodeColor=<rgba 0 1 0.5781 1>}
  }
  MaterialExpressionNamedRerouteDeclaration_14 {
   class=MaterialExpressionNamedRerouteDeclaration
   inputs [None {from=MaterialExpressionComponentMask_10}]
   properties {name=PreskinnedLocPos.B, nodeColor=<rgba 1 0 0.1172 1>}
  }
  MaterialExpressionNamedRerouteDeclaration_16 {
   class=MaterialExpressionNamedRerouteDeclaration
   inputs [None {from=MaterialExpressionTextureSample_4, output=R}]
   properties {name=CausticsDistortionGradient, nodeColor=<rgba 0 0.1797 1 1>}
  }
  MaterialExpressionNamedRerouteDeclaration_18 {
   class=MaterialExpressionNamedRerouteDeclaration
   inputs [None {from=MaterialExpressionCurveAtlasRowParameter_0}]
   properties {name=Primary_GradientColor, nodeColor=<rgba 0.6406 1 0 1>}
  }
  MaterialExpressionNamedRerouteDeclaration_2 {
   class=MaterialExpressionNamedRerouteDeclaration
   inputs [None {from=MaterialExpressionFunctionInput_1}]
   properties {name=InMasks, nodeColor=<rgba 0 1 0.2266 1>}
  }
  MaterialExpressionNamedRerouteDeclaration_20 {
   class=MaterialExpressionNamedRerouteDeclaration
   inputs [None {from=MaterialExpressionSmoothStep_2}]
   properties {name=PrimaryGradient, nodeColor=<rgba 0 0.1797 1 1>}
  }
  MaterialExpressionNamedRerouteDeclaration_3 {
   class=MaterialExpressionNamedRerouteDeclaration
   inputs [None {from=MaterialExpressionTextureSampleParameter2D_1, output=RGB}]
   properties {name=SuperlevelMasks01, nodeColor=<rgba 1 0.1931 0 1>}
  }
  MaterialExpressionNamedRerouteDeclaration_31 {
   class=MaterialExpressionNamedRerouteDeclaration
   inputs [None {from=MaterialExpressionMaterialFunctionCall_10, output=Result}]
   properties {name=RipplesNormal, nodeColor=<rgba 0.7109 1 0 1>}
  }
  MaterialExpressionNamedRerouteDeclaration_4 {
   class=MaterialExpressionNamedRerouteDeclaration
   inputs [None {from=MaterialExpressionMultiply_15}]
   properties {name=PalmTrees_Outlines, nodeColor=<rgba 0.5 1 0 1>}
  }
  MaterialExpressionNamedRerouteDeclaration_5 {
   class=MaterialExpressionNamedRerouteDeclaration
   inputs [None {from=MaterialExpressionSaturate_12}]
   properties {name=PalmTrees, nodeColor=<rgba 1 0 0.8203 1>}
  }
  MaterialExpressionNamedRerouteDeclaration_6 {
   class=MaterialExpressionNamedRerouteDeclaration
   inputs [None {from=MaterialExpressionFunctionInput_4}]
   properties {name=InFresnel, nodeColor=<rgba 0 1 0.8594 1>}
  }
  MaterialExpressionNamedRerouteDeclaration_7 {
   class=MaterialExpressionNamedRerouteDeclaration
   inputs [None {from=MaterialExpressionBlendMaterialAttributes_6}]
   properties {name=OutMA, nodeColor=<rgba 1 0.5391 0 1>}
  }
  MaterialExpressionNamedRerouteDeclaration_8 {
   class=MaterialExpressionNamedRerouteDeclaration
   inputs [None {from=MaterialExpressionTextureSampleParameter2D_3, output=RGB}]
   properties {name=SuperlevelMasks02, nodeColor=<rgba 0.2188 0 1 1>}
  }
  MaterialExpressionNamedRerouteDeclaration_9 {
   class=MaterialExpressionNamedRerouteDeclaration
   inputs [None {from=MaterialExpressionAdd_20}]
   properties {name=Caustics, nodeColor=<rgba 0 0.5312 1 1>}
  }
  MaterialExpressionNamedRerouteUsage_0 {class=MaterialExpressionNamedRerouteUsage}
  MaterialExpressionNamedRerouteUsage_1 {class=MaterialExpressionNamedRerouteUsage}
  MaterialExpressionNamedRerouteUsage_10 {class=MaterialExpressionNamedRerouteUsage}
  MaterialExpressionNamedRerouteUsage_11 {class=MaterialExpressionNamedRerouteUsage}
  MaterialExpressionNamedRerouteUsage_12 {class=MaterialExpressionNamedRerouteUsage}
  MaterialExpressionNamedRerouteUsage_13 {class=MaterialExpressionNamedRerouteUsage}
  MaterialExpressionNamedRerouteUsage_14 {class=MaterialExpressionNamedRerouteUsage}
  MaterialExpressionNamedRerouteUsage_15 {class=MaterialExpressionNamedRerouteUsage}
  MaterialExpressionNamedRerouteUsage_16 {class=MaterialExpressionNamedRerouteUsage}
  MaterialExpressionNamedRerouteUsage_17 {class=MaterialExpressionNamedRerouteUsage}
  MaterialExpressionNamedRerouteUsage_18 {class=MaterialExpressionNamedRerouteUsage}
  MaterialExpressionNamedRerouteUsage_19 {class=MaterialExpressionNamedRerouteUsage}
  MaterialExpressionNamedRerouteUsage_2 {class=MaterialExpressionNamedRerouteUsage}
  MaterialExpressionNamedRerouteUsage_20 {class=MaterialExpressionNamedRerouteUsage}
  MaterialExpressionNamedRerouteUsage_21 {class=MaterialExpressionNamedRerouteUsage}
  MaterialExpressionNamedRerouteUsage_22 {class=MaterialExpressionNamedRerouteUsage}
  MaterialExpressionNamedRerouteUsage_23 {class=MaterialExpressionNamedRerouteUsage}
  MaterialExpressionNamedRerouteUsage_24 {class=MaterialExpressionNamedRerouteUsage}
  MaterialExpressionNamedRerouteUsage_25 {class=MaterialExpressionNamedRerouteUsage}
  MaterialExpressionNamedRerouteUsage_26 {class=MaterialExpressionNamedRerouteUsage}
  MaterialExpressionNamedRerouteUsage_27 {class=MaterialExpressionNamedRerouteUsage}
  MaterialExpressionNamedRerouteUsage_28 {class=MaterialExpressionNamedRerouteUsage}
  MaterialExpressionNamedRerouteUsage_29 {class=MaterialExpressionNamedRerouteUsage}
  MaterialExpressionNamedRerouteUsage_3 {class=MaterialExpressionNamedRerouteUsage}
  MaterialExpressionNamedRerouteUsage_30 {class=MaterialExpressionNamedRerouteUsage}
  MaterialExpressionNamedRerouteUsage_31 {class=MaterialExpressionNamedRerouteUsage}
  MaterialExpressionNamedRerouteUsage_32 {class=MaterialExpressionNamedRerouteUsage}
  MaterialExpressionNamedRerouteUsage_33 {class=MaterialExpressionNamedRerouteUsage}
  MaterialExpressionNamedRerouteUsage_34 {class=MaterialExpressionNamedRerouteUsage}
  MaterialExpressionNamedRerouteUsage_35 {class=MaterialExpressionNamedRerouteUsage}
  MaterialExpressionNamedRerouteUsage_36 {class=MaterialExpressionNamedRerouteUsage}
  MaterialExpressionNamedRerouteUsage_37 {class=MaterialExpressionNamedRerouteUsage}
  MaterialExpressionNamedRerouteUsage_38 {class=MaterialExpressionNamedRerouteUsage}
  MaterialExpressionNamedRerouteUsage_39 {class=MaterialExpressionNamedRerouteUsage}
  MaterialExpressionNamedRerouteUsage_4 {class=MaterialExpressionNamedRerouteUsage}
  MaterialExpressionNamedRerouteUsage_40 {class=MaterialExpressionNamedRerouteUsage}
  MaterialExpressionNamedRerouteUsage_41 {class=MaterialExpressionNamedRerouteUsage}
  MaterialExpressionNamedRerouteUsage_42 {class=MaterialExpressionNamedRerouteUsage}
  MaterialExpressionNamedRerouteUsage_43 {class=MaterialExpressionNamedRerouteUsage}
  MaterialExpressionNamedRerouteUsage_44 {class=MaterialExpressionNamedRerouteUsage}
  MaterialExpressionNamedRerouteUsage_45 {class=MaterialExpressionNamedRerouteUsage}
  MaterialExpressionNamedRerouteUsage_46 {class=MaterialExpressionNamedRerouteUsage}
  MaterialExpressionNamedRerouteUsage_47 {class=MaterialExpressionNamedRerouteUsage}
  MaterialExpressionNamedRerouteUsage_48 {class=MaterialExpressionNamedRerouteUsage}
  MaterialExpressionNamedRerouteUsage_49 {class=MaterialExpressionNamedRerouteUsage}
  MaterialExpressionNamedRerouteUsage_5 {class=MaterialExpressionNamedRerouteUsage}
  MaterialExpressionNamedRerouteUsage_50 {class=MaterialExpressionNamedRerouteUsage}
  MaterialExpressionNamedRerouteUsage_6 {class=MaterialExpressionNamedRerouteUsage}
  MaterialExpressionNamedRerouteUsage_7 {class=MaterialExpressionNamedRerouteUsage}
  MaterialExpressionNamedRerouteUsage_70 {class=MaterialExpressionNamedRerouteUsage}
  MaterialExpressionNamedRerouteUsage_71 {class=MaterialExpressionNamedRerouteUsage}
  MaterialExpressionNamedRerouteUsage_8 {class=MaterialExpressionNamedRerouteUsage}
  MaterialExpressionNamedRerouteUsage_9 {class=MaterialExpressionNamedRerouteUsage}
  MaterialExpressionNoise_0 {
   class=MaterialExpressionNoise
   inputs [
    World Position {from=MaterialExpressionAdd_0, output=""}
    FilterWidth {from="", output=""}
   ]
   properties {levels=3}
  }
  MaterialExpressionNoise_1 {
   class=MaterialExpressionNoise
   inputs [
    World Position {from=MaterialExpressionAdd_22, output=""}
    FilterWidth {from="", output=""}
   ]
   properties {levels=3}
  }
  MaterialExpressionNormalize_0 {class=MaterialExpressionNormalize, inputs [VectorInput {from=MaterialExpressionSubtract_2, output=""}]}
  MaterialExpressionOneMinus_0 {class=MaterialExpressionOneMinus, inputs [None {from=MaterialExpressionNamedRerouteUsage_11}]}
  MaterialExpressionOneMinus_1 {class=MaterialExpressionOneMinus, inputs [None {from=MaterialExpressionLinearInterpolate_3}]}
  MaterialExpressionOneMinus_2 {class=MaterialExpressionOneMinus, inputs [None {from=MaterialExpressionSaturate_5}]}
  MaterialExpressionOneMinus_3 {class=MaterialExpressionOneMinus, inputs [None {from=MaterialExpressionNamedRerouteUsage_34}]}
  MaterialExpressionOneMinus_4 {class=MaterialExpressionOneMinus, inputs [None {from=MaterialExpressionSaturate_3}]}
  MaterialExpressionOneMinus_5 {class=MaterialExpressionOneMinus, inputs [None {from=MaterialExpressionNamedRerouteUsage_28}]}
  MaterialExpressionOneMinus_6 {class=MaterialExpressionOneMinus, inputs [None {from=MaterialExpressionScalarParameter_28}]}
  MaterialExpressionPanner_0 {
   class=MaterialExpressionPanner
   inputs [
    Coordinate {from=MaterialExpressionDivide_0}
    Time {from=""}
    Speed {from=MaterialExpressionAppendVector_0}
   ]
  }
  MaterialExpressionPanner_1 {
   class=MaterialExpressionPanner
   inputs [
    Coordinate {from=MaterialExpressionDivide_1}
    Time {from=""}
    Speed {from=MaterialExpressionConstant2Vector_1}
   ]
  }
  MaterialExpressionPanner_2 {
   class=MaterialExpressionPanner
   inputs [
    Coordinate {from=MaterialExpressionDivide_2}
    Time {from=""}
    Speed {from=MaterialExpressionAppendVector_1}
   ]
  }
  MaterialExpressionPanner_3 {
   class=MaterialExpressionPanner
   inputs [
    Coordinate {from=MaterialExpressionConvert_3}
    Time {from=""}
    Speed {from=""}
   ]
   properties {speedX=0.053, speedY=-0.1124}
  }
  MaterialExpressionPower_0 {
   class=MaterialExpressionPower
   inputs [
    Base {from=MaterialExpressionComponentMask_4}
    Exp {from=MaterialExpressionScalarParameter_9}
   ]
   properties {constExponent=6}
  }
  MaterialExpressionPower_1 {
   class=MaterialExpressionPower
   inputs [
    Base {from=MaterialExpressionNamedRerouteUsage_50}
    Exp {from=""}
   ]
   properties {constExponent=0.5}
  }
  MaterialExpressionPower_3 {
   class=MaterialExpressionPower
   inputs [
    Base {from=MaterialExpressionNamedRerouteUsage_8}
    Exp {from=""}
   ]
   properties {constExponent=3}
  }
  MaterialExpressionQualitySwitch_0 {
   class=MaterialExpressionQualitySwitch
   inputs [
    Default {from=MaterialExpressionAdd_15, output=""}
    Low {from=MaterialExpressionReroute_0, output=""}
    High {from="", output=""}
    Medium {from="", output=""}
    Epic {from="", output=""}
   ]
   properties {}
  }
  MaterialExpressionReflectionVectorWS_0 {class=MaterialExpressionReflectionVectorWS, inputs [CustomWorldNormal {from="", output=""}], properties {}}
  MaterialExpressionReflectionVectorWS_1 {class=MaterialExpressionReflectionVectorWS, inputs [CustomWorldNormal {from="", output=""}], properties {}}
  MaterialExpressionReroute_0 {class=MaterialExpressionReroute, inputs [None {from=MaterialExpressionNamedRerouteUsage_46}]}
  MaterialExpressionSaturate_0 {class=MaterialExpressionSaturate, inputs [None {from=MaterialExpressionAdd_18}]}
  MaterialExpressionSaturate_1 {class=MaterialExpressionSaturate, inputs [None {from=MaterialExpressionAdd_5}]}
  MaterialExpressionSaturate_10 {class=MaterialExpressionSaturate, inputs [None {from=MaterialExpressionAdd_30}]}
  MaterialExpressionSaturate_11 {class=MaterialExpressionSaturate, inputs [None {from=MaterialExpressionDivide_4}]}
  MaterialExpressionSaturate_12 {class=MaterialExpressionSaturate, inputs [None {from=MaterialExpressionMultiply_14}]}
  MaterialExpressionSaturate_13 {class=MaterialExpressionSaturate, inputs [None {from=MaterialExpressionAdd_23}]}
  MaterialExpressionSaturate_14 {class=MaterialExpressionSaturate, inputs [None {from=MaterialExpressionAdd_24}]}
  MaterialExpressionSaturate_15 {class=MaterialExpressionSaturate, inputs [None {from=MaterialExpressionAdd_26}]}
  MaterialExpressionSaturate_16 {class=MaterialExpressionSaturate, inputs [None {from=MaterialExpressionSubtract_1}]}
  MaterialExpressionSaturate_2 {class=MaterialExpressionSaturate, inputs [None {from=MaterialExpressionDotProduct_0}]}
  MaterialExpressionSaturate_3 {class=MaterialExpressionSaturate, inputs [None {from=MaterialExpressionAdd_8}]}
  MaterialExpressionSaturate_4 {class=MaterialExpressionSaturate, inputs [None {from=MaterialExpressionAdd_13}]}
  MaterialExpressionSaturate_5 {class=MaterialExpressionSaturate, inputs [None {from=MaterialExpressionComponentMask_3}]}
  MaterialExpressionSaturate_6 {class=MaterialExpressionSaturate, inputs [None {from=MaterialExpressionAdd_10}]}
  MaterialExpressionSaturate_7 {class=MaterialExpressionSaturate, inputs [None {from=MaterialExpressionAdd_11}]}
  MaterialExpressionSaturate_8 {class=MaterialExpressionSaturate, inputs [None {from=MaterialExpressionSubtract_0}]}
  MaterialExpressionSaturate_9 {class=MaterialExpressionSaturate, inputs [None {from=MaterialExpressionDotProduct_1}]}
  MaterialExpressionScalarParameter_0 {class=MaterialExpressionScalarParameter, properties {defaultValue=0.02, parameterName=terwt}}
  MaterialExpressionScalarParameter_1 {class=MaterialExpressionScalarParameter, properties {defaultValue=0.6, parameterName=Tertiary_Metallic}}
  MaterialExpressionScalarParameter_10 {class=MaterialExpressionScalarParameter, properties {group=SUPERLEVEL, parameterName=PrimaryElement_RoughnessOffset, sortPriority=10}}
  MaterialExpressionScalarParameter_11 {class=MaterialExpressionScalarParameter, properties {group=SUPERLEVEL, parameterName=SecondaryElement_RoughnessOffset, sortPriority=30}}
  MaterialExpressionScalarParameter_12 {class=MaterialExpressionScalarParameter, properties {group=UV, parameterName=CausticsPannerX}}
  MaterialExpressionScalarParameter_13 {class=MaterialExpressionScalarParameter, properties {defaultValue=-0.015, group=UV, parameterName=CausticsPannerY}}
  MaterialExpressionScalarParameter_14 {class=MaterialExpressionScalarParameter, properties {defaultValue=90, group=UV, parameterName=CuasticsUnit}}
  MaterialExpressionScalarParameter_15 {class=MaterialExpressionScalarParameter, properties {parameterName=DisortIntensity}}
  MaterialExpressionScalarParameter_16 {class=MaterialExpressionScalarParameter, properties {group=UV, parameterName=Caustics02PannerX}}
  MaterialExpressionScalarParameter_17 {class=MaterialExpressionScalarParameter, properties {defaultValue=0.015, group=UV, parameterName=Caustics02PannerY}}
  MaterialExpressionScalarParameter_18 {class=MaterialExpressionScalarParameter, properties {defaultValue=90, group=UV, parameterName=CuasticsUnit}}
  MaterialExpressionScalarParameter_19 {class=MaterialExpressionScalarParameter, properties {parameterName=DisortIntensity}}
  MaterialExpressionScalarParameter_2 {class=MaterialExpressionScalarParameter, properties {parameterName=IsRowLiaisonCast_Body}}
  MaterialExpressionScalarParameter_20 {
   class=MaterialExpressionScalarParameter
   properties {defaultValue=1, group=SUPERLEVEL, parameterName=PrimaryElement_MetallicOverride, sortPriority=10}
  }
  MaterialExpressionScalarParameter_21 {
   class=MaterialExpressionScalarParameter
   properties {defaultValue=0.9, group=SUPERLEVEL, parameterName=PrimaryElement_SpecularOverride, sortPriority=10}
  }
  MaterialExpressionScalarParameter_22 {
   class=MaterialExpressionScalarParameter
   properties {defaultValue=0.15, group=SUPERLEVEL, parameterName=PrimaryElement_RoughnessOverride, sortPriority=10}
  }
  MaterialExpressionScalarParameter_23 {class=MaterialExpressionScalarParameter, properties {defaultValue=100, group=SUPERLEVEL, parameterName=Cuastics_Size}}
  MaterialExpressionScalarParameter_24 {class=MaterialExpressionScalarParameter, properties {defaultValue=15, group=SUPERLEVEL, parameterName=Caustics_ParallaxDepth}}
  MaterialExpressionScalarParameter_25 {class=MaterialExpressionScalarParameter, properties {defaultValue=7, group=SUPERLEVEL, parameterName=Caustics_Distortion}}
  MaterialExpressionScalarParameter_26 {
   class=MaterialExpressionScalarParameter
   properties {defaultValue=30, group=SUPERLEVEL, parameterName=PrimaryElement_GradientStart, sortPriority=11}
  }
  MaterialExpressionScalarParameter_27 {
   class=MaterialExpressionScalarParameter
   properties {defaultValue=170, group=SUPERLEVEL, parameterName=PrimaryElement_GradientEnd, sortPriority=11}
  }
  MaterialExpressionScalarParameter_28 {class=MaterialExpressionScalarParameter, properties {defaultValue=0.5, group=SUPERLEVEL, parameterName=Tertiary_BaseColorContrast}}
  MaterialExpressionScalarParameter_29 {class=MaterialExpressionScalarParameter, properties {defaultValue=1.5, group=SUPERLEVEL, parameterName=Tertiary_ColorBrightness}}
  MaterialExpressionScalarParameter_3 {class=MaterialExpressionScalarParameter, properties {parameterName=IsZenCrashSand}}
  MaterialExpressionScalarParameter_39 {class=MaterialExpressionScalarParameter, properties {defaultValue=70, parameterName=RipplesUnit}}
  MaterialExpressionScalarParameter_4 {class=MaterialExpressionScalarParameter, properties {group=SUPERLEVEL, parameterName=IsMarineCarve}}
  MaterialExpressionScalarParameter_5 {class=MaterialExpressionScalarParameter, properties {group=SUPERLEVEL, parameterName=IsFloralMane}}
  MaterialExpressionScalarParameter_6 {
   class=MaterialExpressionScalarParameter
   properties {defaultValue=0.05, group=SUPERLEVEL, parameterName=SecondaryElement_Hardness, sortPriority=30}
  }
  MaterialExpressionScalarParameter_7 {
   class=MaterialExpressionScalarParameter
   properties {defaultValue=-110, group=SUPERLEVEL, parameterName=SecondaryElement_Height, sortPriority=30}
  }
  MaterialExpressionScalarParameter_8 {class=MaterialExpressionScalarParameter, properties {group=SUPERLEVEL, parameterName=SecondaryElement_InvertMask, sortPriority=31}}
  MaterialExpressionScalarParameter_9 {
   class=MaterialExpressionScalarParameter
   properties {defaultValue=6, group=SUPERLEVEL, parameterName=PrimaryElement_BaseColorExp, sortPriority=11}
  }
  MaterialExpressionSetMaterialAttributes_0 {
   class=MaterialExpressionSetMaterialAttributes
   inputs [
    MaterialAttributes {from=MaterialExpressionGetMaterialAttributes_3, output=MaterialAttributes}
    EmissiveColor {from=MaterialExpressionMultiply_45, output=""}
    BaseColor {from=MaterialExpressionMultiply_19, output=""}
   ]
   properties {attributeSetTypes [#4, #2]}
  }
  MaterialExpressionSetMaterialAttributes_1 {
   class=MaterialExpressionSetMaterialAttributes
   inputs [
    MaterialAttributes {from=MaterialExpressionGetMaterialAttributes_0, output=MaterialAttributes}
    BaseColor {from=MaterialExpressionMultiply_6, output=""}
    Metallic {from=MaterialExpressionScalarParameter_1, output=""}
    Roughness {from=MaterialExpressionMin_0, output=""}
   ]
   properties {attributeSetTypes [#2, #5, #1]}
  }
  MaterialExpressionSetMaterialAttributes_2 {
   class=MaterialExpressionSetMaterialAttributes
   inputs [
    MaterialAttributes {from=MaterialExpressionGetMaterialAttributes_2, output=MaterialAttributes}
    BaseColor {from=MaterialExpressionNamedRerouteUsage_36, output=""}
    Metallic {from=MaterialExpressionLinearInterpolate_8, output=""}
    Roughness {from=MaterialExpressionLinearInterpolate_10, output=""}
    EmissiveColor {from=MaterialExpressionNamedRerouteUsage_35, output=""}
    Normal {from=MaterialExpressionLinearInterpolate_13, output=""}
    Specular {from=MaterialExpressionLinearInterpolate_11, output=""}
   ]
   properties {attributeSetTypes [
     #2
     #5
     #1
     #4
     #3
     #6
    ]}
  }
  MaterialExpressionSetMaterialAttributes_3 {
   class=MaterialExpressionSetMaterialAttributes
   inputs [
    MaterialAttributes {from=MaterialExpressionBlendMaterialAttributes_1, output=""}
    BaseColor {from=MaterialExpressionLinearInterpolate_4, output=""}
    Metallic {from=MaterialExpressionConstant_5, output=""}
    Roughness {from=MaterialExpressionSaturate_7, output=""}
    EmissiveColor {from=MaterialExpressionMultiply_37, output=""}
   ]
   properties {attributeSetTypes [
     #2
     #5
     #1
     #4
    ]}
  }
  MaterialExpressionSetMaterialAttributes_4 {
   class=MaterialExpressionSetMaterialAttributes
   inputs [
    MaterialAttributes {from=MaterialExpressionBlendMaterialAttributes_4, output=""}
    BaseColor {from=MaterialExpressionVectorParameter_0, output=RGB}
    Metallic {from=MaterialExpressionConstant_2, output=""}
   ]
   properties {attributeSetTypes [#2, #5]}
  }
  MaterialExpressionSetMaterialAttributes_5 {
   class=MaterialExpressionSetMaterialAttributes
   inputs [
    MaterialAttributes {from=MaterialExpressionBlendMaterialAttributes_2, output=""}
    BaseColor {from=MaterialExpressionConstant_3, output=""}
   ]
   properties {attributeSetTypes [#2]}
  }
  MaterialExpressionSetMaterialAttributes_6 {
   class=MaterialExpressionSetMaterialAttributes
   inputs [
    MaterialAttributes {from=MaterialExpressionBlendMaterialAttributes_5, output=""}
    OpacityMask {from=MaterialExpressionComponentMask_12, output=""}
   ]
   properties {attributeSetTypes [#7]}
  }
  MaterialExpressionShaderStageSwitch_0 {
   class=MaterialExpressionShaderStageSwitch
   inputs [
    PixelShader {from=MaterialExpressionVertexInterpolator_0, output=PS}
    VertexShader {from=MaterialExpressionLocalPosition_0, output=XYZ}
   ]
   properties {}
  }
  MaterialExpressionShadingPathSwitch_0 {
   class=MaterialExpressionShadingPathSwitch
   inputs [
    Default {from=MaterialExpressionQualitySwitch_0, output=""}
    Deferred {from=MaterialExpressionQualitySwitch_0, output=""}
    Forward {from=MaterialExpressionReroute_0, output=""}
    Mobile {from=MaterialExpressionReroute_0, output=""}
   ]
   properties {}
  }
  MaterialExpressionSine_0 {class=MaterialExpressionSine, inputs [None {from=MaterialExpressionAdd_3}]}
  MaterialExpressionSmoothStep_0 {
   class=MaterialExpressionSmoothStep
   inputs [
    Min {from="", output=""}
    Max {from="", output=""}
    Value {from=MaterialExpressionTextureSampleParameter2D_5, output=B}
   ]
   properties {constMax=0.2}
  }
  MaterialExpressionSmoothStep_1 {
   class=MaterialExpressionSmoothStep
   inputs [
    Min {from="", output=""}
    Max {from="", output=""}
    Value {from=MaterialExpressionTextureSampleParameter2D_0, output=G}
   ]
   properties {constMax=0.3}
  }
  MaterialExpressionSmoothStep_2 {
   class=MaterialExpressionSmoothStep
   inputs [
    Min {from=MaterialExpressionScalarParameter_26, output=""}
    Max {from=MaterialExpressionScalarParameter_27, output=""}
    Value {from=MaterialExpressionNamedRerouteDeclaration_14, output=""}
   ]
   properties {constMax=155, constMin=30}
  }
  MaterialExpressionSphereMask_0 {
   class=MaterialExpressionSphereMask
   inputs [
    A {from=MaterialExpressionAdd_7, output=""}
    B {from=MaterialExpressionConstant_4, output=""}
    Radius {from="", output=""}
    Hardness {from="", output=""}
   ]
   properties {attenuationRadius=0.1, hardnessPercent=1}
  }
  MaterialExpressionSubtract_0 {
   class=MaterialExpressionSubtract
   inputs [
    A {from=MaterialExpressionMultiply_23}
    B {from=MaterialExpressionSaturate_4}
   ]
  }
  MaterialExpressionSubtract_1 {
   class=MaterialExpressionSubtract
   inputs [
    A {from=MaterialExpressionMultiply_51}
    B {from=MaterialExpressionSaturate_15}
   ]
  }
  MaterialExpressionSubtract_2 {
   class=MaterialExpressionSubtract
   inputs [
    A {from=MaterialExpressionNamedRerouteUsage_46}
    B {from=MaterialExpressionTransformPosition_0}
   ]
  }
  MaterialExpressionTextureCoordinate_0 {class=MaterialExpressionTextureCoordinate}
  MaterialExpressionTextureCoordinate_1 {class=MaterialExpressionTextureCoordinate}
  MaterialExpressionTextureCoordinate_2 {class=MaterialExpressionTextureCoordinate}
  MaterialExpressionTextureObject_0 {
   class=MaterialExpressionTextureObject
   inputs []
   properties {
    samplerType=SAMPLERTYPE_Masks
    texture=/BRCosmetics/Characters/Player/Male/Medium/Bodies/M_MED_FloralMane_Sand/Skins/SuperLevels/Textures/T_SL_S39_T1_TA
   }
  }
  MaterialExpressionTextureObject_1 {class=MaterialExpressionTextureObject, inputs [], properties {texture=/Game/Global/Textures/Default/Caustics/T_Caustics_Softer}}
  MaterialExpressionTextureSampleParameter2D_0 {
   class=MaterialExpressionTextureSampleParameter2D
   inputs [
    UVs {from=MaterialExpressionAdd_1, output=""}
    Apply View MipBias {from="", output=""}
   ]
   properties {
    group=SUPERLEVEL
    parameterName=PalmTreeMask
    samplerSource=SSM_Wrap_WorldGroupSettings
    sortPriority=10
    texture=/BRCosmetics/Characters/Player/Male/Medium/Bodies/M_MED_ZenCrash/Skins/Superlevels/T1/Textures/T_ZenCrash_Body_PalmTreeDecals
   }
  }
  MaterialExpressionTextureSampleParameter2D_1 {
   class=MaterialExpressionTextureSampleParameter2D
   inputs [
    UVs {from="", output=""}
    Apply View MipBias {from="", output=""}
   ]
   properties {
    group=SUPERLEVEL
    parameterName=SuperlevelMask_01
    samplerSource=SSM_Wrap_WorldGroupSettings
    samplerType=SAMPLERTYPE_Masks
    sortPriority=0
    texture=/BRCosmetics/Characters/Player/Male/Medium/Bodies/M_MED_ZenCrash/Textures/T_ZenCrash_Body_SL_FX1
   }
  }
  MaterialExpressionTextureSampleParameter2D_3 {
   class=MaterialExpressionTextureSampleParameter2D
   inputs [
    UVs {from="", output=""}
    Apply View MipBias {from="", output=""}
   ]
   properties {
    group=SUPERLEVEL
    parameterName=SuperLevelsMask_02
    samplerType=SAMPLERTYPE_Masks
    sortPriority=0
    texture=/BRCosmetics/Characters/Player/Male/Medium/Bodies/M_MED_ZenCrash/Textures/T_ZenCrash_Body_SL_FX2
   }
  }
  MaterialExpressionTextureSampleParameter2D_4 {
   class=MaterialExpressionTextureSampleParameter2D
   inputs [
    UVs {from="", output=""}
    Apply View MipBias {from="", output=""}
   ]
   properties {
    group=SUPERLEVEL
    parameterName=PalmTreeMask
    samplerSource=SSM_Wrap_WorldGroupSettings
    sortPriority=10
    texture=&T_ZenCrash_Body_PalmTreeDecals
   }
  }
  MaterialExpressionTextureSampleParameter2D_5 {
   class=MaterialExpressionTextureSampleParameter2D
   inputs [
    UVs {from="", output=""}
    Apply View MipBias {from="", output=""}
   ]
   properties {
    group=SUPERLEVEL
    parameterName=PalmTreeMask
    samplerSource=SSM_Wrap_WorldGroupSettings
    sortPriority=10
    texture=&T_ZenCrash_Body_PalmTreeDecals
   }
  }
  MaterialExpressionTextureSample_0 {
   class=MaterialExpressionTextureSample
   inputs [
    UVs {from=MaterialExpressionBumpOffset_0}
    Tex {from=""}
    Apply View MipBias {from=""}
   ]
   properties {texture=&T_Caustics_Softer}
  }
  MaterialExpressionTextureSample_1 {
   class=MaterialExpressionTextureSample
   inputs [
    UVs {from=MaterialExpressionPanner_1}
    Tex {from=""}
    Apply View MipBias {from=""}
   ]
   properties {samplerType=SAMPLERTYPE_Normal, texture=/Game/Global/Textures/Default/Noise/T_VerticalRipples_N}
  }
  MaterialExpressionTextureSample_3 {
   class=MaterialExpressionTextureSample
   inputs [
    UVs {from=MaterialExpressionBumpOffset_1}
    Tex {from=""}
    Apply View MipBias {from=""}
   ]
   properties {texture=&T_Caustics_Softer}
  }
  MaterialExpressionTextureSample_4 {
   class=MaterialExpressionTextureSample
   inputs [
    UVs {from=MaterialExpressionAdd_2}
    Tex {from=""}
    Apply View MipBias {from=""}
   ]
   properties {
    samplerSource=SSM_Wrap_WorldGroupSettings
    samplerType=SAMPLERTYPE_Masks
    texture=&T_SL_S39_T1_TA
   }
  }
  MaterialExpressionTime_0 {class=MaterialExpressionTime}
  MaterialExpressionTime_1 {class=MaterialExpressionTime}
  MaterialExpressionTransformPosition_0 {
   class=MaterialExpressionTransformPosition
   inputs [None {from=MaterialExpressionCameraPositionWS_0, output=""}]
   properties {transformSourceType=TRANSFORMPOSSOURCE_World}
  }
  MaterialExpressionTransform_0 {class=MaterialExpressionTransform, inputs [None {from=MaterialExpressionGetMaterialAttributes_1, output=Normal}]}
  MaterialExpressionTransform_1 {class=MaterialExpressionTransform, inputs [None {from=MaterialExpressionConstant3Vector_8}]}
  MaterialExpressionTransform_2 {class=MaterialExpressionTransform, inputs [None {from=MaterialExpressionGetMaterialAttributes_4, output=Normal}]}
  MaterialExpressionTransform_3 {class=MaterialExpressionTransform, inputs [None {from=MaterialExpressionConstant3Vector_11}]}
  MaterialExpressionVectorParameter_0 {class=MaterialExpressionVectorParameter, properties {parameterName=Accent_Color, defaultValue=<rgba 0.0312 0.1503 0.6059 1>}}
  MaterialExpressionVectorParameter_1 {class=MaterialExpressionVectorParameter, properties {parameterName=Tertiary_Color, defaultValue=<rgba 0.9774 0.2714 0.0288 1>}}
  MaterialExpressionVectorParameter_2 {
   class=MaterialExpressionVectorParameter
   properties {group=SUPERLEVEL, parameterName=AccentEmissive_Color, defaultValue=<rgba 1.2 0.3663 0 1>}
  }
  MaterialExpressionVertexInterpolator_0 {class=MaterialExpressionVertexInterpolator, inputs [VS {from=MaterialExpressionLocalPosition_0, output=XYZ}], properties {}}
 }}
