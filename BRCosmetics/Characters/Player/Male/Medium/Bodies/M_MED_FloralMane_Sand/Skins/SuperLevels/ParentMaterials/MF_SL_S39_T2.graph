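# Material function for the Season 39 tier 2 Super Level style of the FloralMane Sand character body, restyling base color, emissive, metallic, specular and roughness of incoming material attributes.
# Two mask textures split the body into primary, secondary, tertiary and accent regions.
# Primary sand waves use time-scrolled, rotated, distorted position-based UVs, with texture samples rounded into hard color bands.
# Fresnel adds view-dependent shading, and a scalar parameter blends two final attribute sets.
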MaterialFunction /BRCosmetics/Characters/Player/Male/Medium/Bodies/M_MED_FloralMane_Sand/Skins/SuperLevels/ParentMaterials/MF_SL_S39_T2
graph {expressions {
  MaterialExpressionAdd_0 {
   class=MaterialExpressionAdd
   inputs [
    A {from=MaterialExpressionSmoothStep_11}
    B {from=MaterialExpressionConvert_2}
   ]
  }
  MaterialExpressionAdd_1 {
   class=MaterialExpressionAdd
   inputs [
    A {from=MaterialExpressionComponentMask_0}
    B {from=MaterialExpressionComponentMask_2}
   ]
  }
  MaterialExpressionAdd_10 {
   class=MaterialExpressionAdd
   inputs [
    A {from=MaterialExpressionConvert_14}
    B {from=""}
   ]
   properties {constB=-77}
  }
  MaterialExpressionAdd_11 {
   class=MaterialExpressionAdd
   inputs [
    A {from=""}
    B {from=MaterialExpressionTextureSample_2, output=B}
   ]
   properties {constA=0.38, constB=0.3}
  }
  MaterialExpressionAdd_15 {
   class=MaterialExpressionAdd
   inputs [
    A {from=MaterialExpressionReroute_1}
    B {from=MaterialExpressionMultiply_25}
   ]
  }
  MaterialExpressionAdd_16 {
   class=MaterialExpressionAdd
   inputs [
    A {from=MaterialExpressionAdd_7}
    B {from=MaterialExpressionConstant2Vector_2}
   ]
  }
  MaterialExpressionAdd_2 {
   class=MaterialExpressionAdd
   inputs [
    A {from=MaterialExpressionComponentMask_1}
    B {from=""}
   ]
   properties {constB=-100}
  }
  MaterialExpressionAdd_23 {
   class=MaterialExpressionAdd
   inputs [
    A {from=MaterialExpressionComponentMask_16}
    B {from=MaterialExpressionDotProduct_0}
   ]
  }
  MaterialExpressionAdd_3 {
   class=MaterialExpressionAdd
   inputs [
    A {from=""}
    B {from=MaterialExpressionSmoothStep_0}
   ]
   properties {constA=0.15}
  }
  MaterialExpressionAdd_4 {
   class=MaterialExpressionAdd
   inputs [
    A {from=MaterialExpressionFrac_0}
    B {from=MaterialExpressionDivide_0}
   ]
  }
  MaterialExpressionAdd_5 {
   class=MaterialExpressionAdd
   inputs [
    A {from=MaterialExpressionSaturate_0}
    B {from=""}
   ]
   properties {constB=0.2}
  }
  MaterialExpressionAdd_6 {
   class=MaterialExpressionAdd
   inputs [
    A {from=MaterialExpressionTextureSample_5, output=B}
    B {from=""}
   ]
   properties {constB=0.32}
  }
  MaterialExpressionAdd_7 {
   class=MaterialExpressionAdd
   inputs [
    A {from=MaterialExpressionMultiply_15}
    B {from=MaterialExpressionConvert_10}
   ]
  }
  MaterialExpressionAdd_8 {
   class=MaterialExpressionAdd
   inputs [
    A {from=MaterialExpressionMultiply_16}
    B {from=MaterialExpressionConvert_11}
   ]
  }
  MaterialExpressionAdd_9 {
   class=MaterialExpressionAdd
   inputs [
    A {from=MaterialExpressionConvert_14}
    B {from=""}
   ]
   properties {constB=-75}
  }
  MaterialExpressionBlendMaterialAttributes_0 {
   class=MaterialExpressionBlendMaterialAttributes
   inputs [
    A {from=MaterialExpressionSetMaterialAttributes_2, output=""}
    B {from=MaterialExpressionSetMaterialAttributes_6, output=""}
    Alpha {from=MaterialExpressionScalarParameter_16, output=""}
   ]
   properties {}
  }
  MaterialExpressionCameraPositionWS_0 {class=MaterialExpressionCameraPositionWS, inputs [], properties {}}
  MaterialExpressionComponentMask_0 {class=MaterialExpressionComponentMask, inputs [None {from=MaterialExpressionNamedRerouteUsage_16}], properties {b=true, r=true}}
  MaterialExpressionComponentMask_1 {class=MaterialExpressionComponentMask, inputs [None {from=MaterialExpressionNamedRerouteUsage_50}], properties {b=true}}
  MaterialExpressionComponentMask_12 {class=MaterialExpressionComponentMask, inputs [None {from=MaterialExpressionNamedRerouteUsage_49}], properties {r=true}}
  MaterialExpressionComponentMask_16 {class=MaterialExpressionComponentMask, inputs [None {from=MaterialExpressionNamedRerouteDeclaration_5}], properties {b=true}}
  MaterialExpressionComponentMask_2 {class=MaterialExpressionComponentMask, inputs [None {from=MaterialExpressionVectorParameter_2, output=RGB}], properties {g=true, r=true}}
  MaterialExpressionConstant2Vector_2 {class=MaterialExpressionConstant2Vector, properties {desc=Offset a little to cover bleeding, g=-0.01}}
  MaterialExpressionConstant3Vector_0 {class=MaterialExpressionConstant3Vector, properties {constant=<rgba 0.025 0.025 0.025 0>}}
  MaterialExpressionConstant3Vector_1 {class=MaterialExpressionConstant3Vector, properties {constant=<rgba 0.3864 0.0048 0.3467 1>}}
  MaterialExpressionConstant3Vector_2 {class=MaterialExpressionConstant3Vector, properties {constant=<rgba 0.025 0.025 0.025 0>}}
  MaterialExpressionConstant3Vector_3 {class=MaterialExpressionConstant3Vector, properties {constant=<rgba 0.0469 0.036 0.1441 1>}}
  MaterialExpressionConstant3Vector_4 {class=MaterialExpressionConstant3Vector, properties {constant=<rgba 0 1 1 0>}}
  MaterialExpressionConstant3Vector_5 {class=MaterialExpressionConstant3Vector, properties {constant=<rgba 1 1 1 0>}}
  MaterialExpressionConstant_0 {class=MaterialExpressionConstant}
  MaterialExpressionConstant_3 {class=MaterialExpressionConstant, properties {r=5}}
  MaterialExpressionConstant_5 {class=MaterialExpressionConstant, properties {r=1}}
  MaterialExpressionConvert_0 {
   class=MaterialExpressionConvert
   inputs [None {from=MaterialExpressionFunctionInput_3}]
   properties {
    nodeName=Break Float3
    convertInputs [{
      type=Vector3
      expressionInput {
       expression=&self:MaterialExpressionFunctionInput_3
       inputName=None
       mask=0
       maskA=0
       maskB=0
       maskG=0
       maskR=0
       outputIndex=0
      }
     }]
    convertOutputs [{type=Scalar}, {type=Scalar}, {type=Scalar}]
   }
  }
  MaterialExpressionConvert_10 {
   class=MaterialExpressionConvert
   inputs [
    None {from=MaterialExpressionMultiply_12}
    None {from=MaterialExpressionMultiply_12}
   ]
   properties {
    nodeName=Make Vector2
    convertInputs [
     {
      type=Scalar
      expressionInput {
       expression=&self:MaterialExpressionMultiply_12
       inputName=None
       mask=0
       maskA=0
       maskB=0
       maskG=0
       maskR=0
       outputIndex=0
      }
     }
     {
      type=Scalar
      expressionInput {
       expression=&self:MaterialExpressionMultiply_12
       inputName=None
       mask=0
       maskA=0
       maskB=0
       maskG=0
       maskR=0
       outputIndex=0
      }
     }
    ]
    convertOutputs [{type=Vector2}]
   }
  }
  MaterialExpressionConvert_11 {
   class=MaterialExpressionConvert
   inputs [
    None {from=MaterialExpressionConvert_9}
    None {from=MaterialExpressionConvert_9}
   ]
   properties {
    nodeName=Make Vector2
    convertInputs [
     {
      type=Scalar
      expressionInput {
       expression=&self:MaterialExpressionConvert_9
       inputName=None
       mask=0
       maskA=0
       maskB=0
       maskG=0
       maskR=0
       outputIndex=0
      }
     }
     {
      type=Scalar
      expressionInput {
       expression=&self:MaterialExpressionConvert_9
       inputName=None
       mask=0
       maskA=0
       maskB=0
       maskG=0
       maskR=0
       outputIndex=0
      }
     }
    ]
    convertOutputs [{type=Vector2}]
   }
  }
  MaterialExpressionConvert_12 {
   class=MaterialExpressionConvert
   inputs [None {from=MaterialExpressionNamedRerouteUsage_3}]
   properties {
    nodeName=Break Float3
    convertInputs [{
      type=Vector3
      expressionInput {
       expression=&self:MaterialExpressionNamedRerouteUsage_3
       inputName=None
       mask=0
       maskA=0
       maskB=0
       maskG=0
       maskR=0
       outputIndex=0
      }
     }]
    convertOutputs [{type=Scalar}, {type=Scalar}, {type=Scalar}]
   }
  }
  MaterialExpressionConvert_13 {
   class=MaterialExpressionConvert
   inputs [None {from=MaterialExpressionNamedRerouteUsage_19}]
   properties {
    nodeName=Break Float3
    convertInputs [{
      type=Vector3
      expressionInput {
       expression=&self:MaterialExpressionNamedRerouteUsage_19
       inputName=None
       mask=0
       maskA=0
       maskB=0
       maskG=0
       maskR=0
       outputIndex=0
      }
     }]
    convertOutputs [{type=Scalar}, {type=Scalar}, {type=Scalar}]
   }
  }
  MaterialExpressionConvert_14 {
   class=MaterialExpressionConvert
   inputs [None {from=MaterialExpressionNamedRerouteUsage_23}]
   properties {
    nodeName=Break Float3
    convertInputs [{
      type=Vector3
      expressionInput {
       expression=&self:MaterialExpressionNamedRerouteUsage_23
       inputName=None
       mask=0
       maskA=0
       maskB=0
       maskG=0
       maskR=0
       outputIndex=0
      }
     }]
    convertOutputs [{type=Scalar}, {type=Scalar}, {type=Scalar}]
   }
  }
  MaterialExpressionConvert_2 {
   class=MaterialExpressionConvert
   inputs [None {from=MaterialExpressionMaterialFunctionCall_0, output=XY Texture}]
   properties {
    nodeName=Break Float3
    convertInputs [{
      type=Vector3
      expressionInput {
       expression=&self:MaterialExpressionMaterialFunctionCall_0
       inputName=None
       mask=0
       maskA=0
       maskB=0
       maskG=0
       maskR=0
       outputIndex=0
      }
     }]
    convertOutputs [{type=Scalar}, {type=Scalar}, {type=Scalar}]
   }
  }
  MaterialExpressionConvert_8 {
   class=MaterialExpressionConvert
   inputs [
    None {from=MaterialExpressionMaterialFunctionCall_3, output=R}
    None {from=MaterialExpressionMaterialFunctionCall_3, output=R}
   ]
   properties {
    nodeName=Make Vector2
    convertInputs [
     {
      type=Scalar
      expressionInput {
       expression=&self:MaterialExpressionMaterialFunctionCall_3
       inputName=None
       mask=0
       maskA=0
       maskB=0
       maskG=0
       maskR=0
       outputIndex=1
      }
     }
     {
      type=Scalar
      expressionInput {
       expression=&self:MaterialExpressionMaterialFunctionCall_3
       inputName=None
       mask=0
       maskA=0
       maskB=0
       maskG=0
       maskR=0
       outputIndex=3
      }
     }
    ]
    convertOutputs [{type=Vector2}]
   }
  }
  MaterialExpressionConvert_9 {
   class=MaterialExpressionConvert
   inputs [None {from=MaterialExpressionRotator_1}]
   properties {
    nodeName=Break Float2
    convertInputs [{
      type=Vector2
      expressionInput {
       expression=&self:MaterialExpressionRotator_1
       inputName=None
       mask=0
       maskA=0
       maskB=0
       maskG=0
       maskR=0
       outputIndex=0
      }
     }]
    convertOutputs [{type=Scalar}, {type=Scalar}]
   }
  }
  MaterialExpressionDesaturation_0 {
   class=MaterialExpressionDesaturation
   inputs [
    None {from=MaterialExpressionNamedRerouteUsage_27, output=""}
    Fraction {from=MaterialExpressionConstant_5, output=""}
   ]
   properties {}
  }
  MaterialExpressionDivide_0 {
   class=MaterialExpressionDivide
   inputs [
    A {from=MaterialExpressionAdd_1}
    B {from=""}
   ]
   properties {constB=128}
  }
  MaterialExpressionDivide_2 {
   class=MaterialExpressionDivide
   inputs [
    A {from=MaterialExpressionAdd_10}
    B {from=""}
   ]
   properties {constB=100}
  }
  MaterialExpressionDotProduct_0 {
   class=MaterialExpressionDotProduct
   inputs [
    A {from=MaterialExpressionNamedRerouteUsage_34, output=""}
    B {from=MaterialExpressionConstant3Vector_5, output=""}
   ]
   properties {}
  }
  MaterialExpressionDotProduct_1 {
   class=MaterialExpressionDotProduct
   inputs [
    A {from=MaterialExpressionNamedRerouteUsage_48, output=""}
    B {from=MaterialExpressionConstant3Vector_4, output=""}
   ]
   properties {}
  }
  MaterialExpressionFrac_0 {class=MaterialExpressionFrac, inputs [None {from=MaterialExpressionMultiply_13, output=""}], properties {}}
  MaterialExpressionFresnel_0 {
   class=MaterialExpressionFresnel
   inputs [
    ExponentIn {from="", output=""}
    BaseReflectFractionIn {from="", output=""}
    Normal {from="", output=""}
   ]
   properties {baseReflectFraction=0, exponent=1.5}
  }
  MaterialExpressionFresnel_1 {
   class=MaterialExpressionFresnel
   inputs [
    ExponentIn {from="", output=""}
    BaseReflectFractionIn {from="", output=""}
    Normal {from="", output=""}
   ]
   properties {baseReflectFraction=0, exponent=1}
  }
  MaterialExpressionFunctionInput_0 {
   class=MaterialExpressionFunctionInput
   inputs [Preview {from="", output=""}]
   properties {inputName=Material, inputType=FunctionInput_MaterialAttributes}
  }
  MaterialExpressionFunctionInput_1 {
   class=MaterialExpressionFunctionInput
   inputs [Preview {from=MaterialExpressionVertexInterpolator_1, output=PS}]
   properties {inputName=PreskinnedLocPos, sortPriority=1}
  }
  MaterialExpressionFunctionInput_2 {
   class=MaterialExpressionFunctionInput
   inputs [Preview {from=MaterialExpressionFresnel_1, output=""}]
   properties {inputName=Fresnel, inputType=FunctionInput_Scalar, sortPriority=2}
  }
  MaterialExpressionFunctionInput_3 {class=MaterialExpressionFunctionInput, inputs [Preview {from="", output=""}], properties {inputName=M, sortPriority=3}}
  MaterialExpressionFunctionInput_4 {class=MaterialExpressionFunctionInput, inputs [Preview {from="", output=""}], properties {inputName=TechArtMask, sortPriority=4}}
  MaterialExpressionFunctionOutput_0 {class=MaterialExpressionFunctionOutput, inputs [None {from=MaterialExpressionBlendMaterialAttributes_0, output=""}]}
  MaterialExpressionGetMaterialAttributes_0 {
   class=MaterialExpressionGetMaterialAttributes
   inputs [None {from=MaterialExpressionFunctionInput_0, output=""}]
   properties {attributeGetTypes [
     #1
     #2
     #3
     #4
     #5
    ]}
  }
  MaterialExpressionLinearInterpolate_0 {
   class=MaterialExpressionLinearInterpolate
   inputs [
    A {from=MaterialExpressionScalarParameter_6}
    B {from=MaterialExpressionScalarParameter_14}
    Alpha {from=MaterialExpressionNamedRerouteUsage_0}
   ]
   properties {constB=0.9}
  }
  MaterialExpressionLinearInterpolate_1 {
   class=MaterialExpressionLinearInterpolate
   inputs [
    A {from=""}
    B {from=MaterialExpressionScalarParameter_24}
    Alpha {from=MaterialExpressionNamedRerouteUsage_32}
   ]
   properties {constB=5}
  }
  MaterialExpressionLinearInterpolate_10 {
   class=MaterialExpressionLinearInterpolate
   inputs [
    A {from=MaterialExpressionLinearInterpolate_9}
    B {from=MaterialExpressionLinearInterpolate_11}
    Alpha {from=MaterialExpressionRound_0}
   ]
  }
  MaterialExpressionLinearInterpolate_11 {
   class=MaterialExpressionLinearInterpolate
   inputs [
    A {from=MaterialExpressionVectorParameter_7, output=RGB}
    B {from=MaterialExpressionVectorParameter_10, output=RGB}
    Alpha {from=MaterialExpressionSmoothStep_9}
   ]
  }
  MaterialExpressionLinearInterpolate_12 {
   class=MaterialExpressionLinearInterpolate
   inputs [
    A {from=MaterialExpressionVectorParameter_8, output=RGB}
    B {from=MaterialExpressionVectorParameter_11, output=RGB}
    Alpha {from=MaterialExpressionNamedRerouteUsage_6}
   ]
   properties {desc=Blue sand}
  }
  MaterialExpressionLinearInterpolate_13 {
   class=MaterialExpressionLinearInterpolate
   inputs [
    A {from=MaterialExpressionNamedRerouteUsage_35}
    B {from=""}
    Alpha {from=MaterialExpressionNamedRerouteUsage_36}
   ]
   properties {constB=0.5}
  }
  MaterialExpressionLinearInterpolate_14 {
   class=MaterialExpressionLinearInterpolate
   inputs [
    A {from=MaterialExpressionLinearInterpolate_6}
    B {from=MaterialExpressionScalarParameter_10}
    Alpha {from=MaterialExpressionNamedRerouteUsage_20}
   ]
  }
  MaterialExpressionLinearInterpolate_15 {
   class=MaterialExpressionLinearInterpolate
   inputs [
    A {from=MaterialExpressionLinearInterpolate_8}
    B {from=MaterialExpressionNamedRerouteUsage_22}
    Alpha {from=MaterialExpressionNamedRerouteUsage_21}
   ]
  }
  MaterialExpressionLinearInterpolate_16 {
   class=MaterialExpressionLinearInterpolate
   inputs [
    A {from=MaterialExpressionConstant3Vector_0}
    B {from=MaterialExpressionConstant3Vector_1}
    Alpha {from=MaterialExpressionLinearInterpolate_24}
   ]
  }
  MaterialExpressionLinearInterpolate_17 {
   class=MaterialExpressionLinearInterpolate
   inputs [
    A {from=MaterialExpressionLinearInterpolate_5}
    B {from=MaterialExpressionNamedRerouteUsage_25}
    Alpha {from=MaterialExpressionNamedRerouteUsage_14}
   ]
  }
  MaterialExpressionLinearInterpolate_18 {
   class=MaterialExpressionLinearInterpolate
   inputs [
    A {from=MaterialExpressionNamedRerouteUsage_40}
    B {from=MaterialExpressionLinearInterpolate_19}
    Alpha {from=MaterialExpressionNamedRerouteUsage_39}
   ]
   properties {constB=0.5}
  }
  MaterialExpressionLinearInterpolate_19 {
   class=MaterialExpressionLinearInterpolate
   inputs [
    A {from=MaterialExpressionScalarParameter_9}
    B {from=""}
    Alpha {from=MaterialExpressionNamedRerouteUsage_42}
   ]
   properties {constA=1, constB=0}
  }
  MaterialExpressionLinearInterpolate_2 {
   class=MaterialExpressionLinearInterpolate
   inputs [
    A {from=MaterialExpressionVectorParameter_6, output=RGB}
    B {from=MaterialExpressionVectorParameter_0, output=RGB}
    Alpha {from=MaterialExpressionSmoothStep_2}
   ]
  }
  MaterialExpressionLinearInterpolate_20 {
   class=MaterialExpressionLinearInterpolate
   inputs [
    A {from=MaterialExpressionTextureSample_3, output=R}
    B {from=""}
    Alpha {from=""}
   ]
  }
  MaterialExpressionLinearInterpolate_21 {
   class=MaterialExpressionLinearInterpolate
   inputs [
    A {from=MaterialExpressionLinearInterpolate_7}
    B {from=MaterialExpressionNamedRerouteUsage_44}
    Alpha {from=MaterialExpressionNamedRerouteUsage_45}
   ]
  }
  MaterialExpressionLinearInterpolate_22 {
   class=MaterialExpressionLinearInterpolate
   inputs [
    A {from=MaterialExpressionLinearInterpolate_15}
    B {from=MaterialExpressionConstant3Vector_3}
    Alpha {from=MaterialExpressionNamedRerouteUsage_47}
   ]
   properties {constB=0.05}
  }
  MaterialExpressionLinearInterpolate_23 {
   class=MaterialExpressionLinearInterpolate
   inputs [
    A {from=MaterialExpressionConvert_13}
    B {from=MaterialExpressionMultiply_10}
    Alpha {from=MaterialExpressionScalarParameter_17}
   ]
  }
  MaterialExpressionLinearInterpolate_24 {
   class=MaterialExpressionLinearInterpolate
   inputs [
    A {from=MaterialExpressionSmoothStep_12}
    B {from=MaterialExpressionSaturate_4}
    Alpha {from=MaterialExpressionMultiply_17}
   ]
  }
  MaterialExpressionLinearInterpolate_25 {
   class=MaterialExpressionLinearInterpolate
   inputs [
    A {from=MaterialExpressionSaturate_2}
    B {from=MaterialExpressionSaturate_13}
    Alpha {from=MaterialExpressionScalarParameter_19}
   ]
  }
  MaterialExpressionLinearInterpolate_3 {
   class=MaterialExpressionLinearInterpolate
   inputs [
    A {from=MaterialExpressionNamedRerouteUsage_9}
    B {from=MaterialExpressionConstant3Vector_2}
    Alpha {from=MaterialExpressionNamedRerouteUsage_33}
   ]
  }
  MaterialExpressionLinearInterpolate_4 {
   class=MaterialExpressionLinearInterpolate
   inputs [
    A {from=MaterialExpressionNamedRerouteUsage_4}
    B {from=MaterialExpressionMultiply_2}
    Alpha {from=MaterialExpressionNamedRerouteUsage_2}
   ]
  }
  MaterialExpressionLinearInterpolate_5 {
   class=MaterialExpressionLinearInterpolate
   inputs [
    A {from=MaterialExpressionLinearInterpolate_3}
    B {from=MaterialExpressionMultiply_7}
    Alpha {from=MaterialExpressionNamedRerouteUsage_7}
   ]
  }
  MaterialExpressionLinearInterpolate_6 {
   class=MaterialExpressionLinearInterpolate
   inputs [
    A {from=MaterialExpressionNamedRerouteUsage_12}
    B {from=MaterialExpressionLinearInterpolate_0}
    Alpha {from=MaterialExpressionNamedRerouteUsage_11}
   ]
  }
  MaterialExpressionLinearInterpolate_7 {
   class=MaterialExpressionLinearInterpolate
   inputs [
    A {from=MaterialExpressionLinearInterpolate_4}
    B {from=MaterialExpressionNamedRerouteUsage_15}
    Alpha {from=MaterialExpressionNamedRerouteUsage_24}
   ]
  }
  MaterialExpressionLinearInterpolate_8 {
   class=MaterialExpressionLinearInterpolate
   inputs [
    A {from=MaterialExpressionLinearInterpolate_17}
    B {from=MaterialExpressionNamedRerouteUsage_28}
    Alpha {from=MaterialExpressionNamedRerouteUsage_26}
   ]
  }
  MaterialExpressionLinearInterpolate_9 {
   class=MaterialExpressionLinearInterpolate
   inputs [
    A {from=MaterialExpressionLinearInterpolate_12}
    B {from=MaterialExpressionVectorParameter_9, output=RGB}
    Alpha {from=MaterialExpressionNamedRerouteDeclaration_2}
   ]
  }
  MaterialExpressionLocalPosition_1 {class=MaterialExpressionLocalPosition, inputs [], properties {localOrigin=InstancePreSkinning}}
  MaterialExpressionMaterialFunctionCall_0 {
   class=MaterialExpressionMaterialFunctionCall
   inputs [
    TextureObject {from=MaterialExpressionTextureObject_0}
    TextureSize {from=MaterialExpressionConstant_3}
    WorldPosition {from=MaterialExpressionShadingPathSwitch_0}
    FadingGradient {from=""}
    In {from=""}
    World Space Normal {from=""}
    ProjectionTransitionContrast {from=""}
   ]
   properties {materialFunction=&/Preskinned_WP_Texture_ScalePan_XY}
  }
  MaterialExpressionMaterialFunctionCall_3 {
   class=MaterialExpressionMaterialFunctionCall
   inputs []
   properties {materialFunction=&BoundingBoxBased_0-1_UVW}
  }
  MaterialExpressionMax_0 {
   class=MaterialExpressionMax
   inputs [
    A {from=MaterialExpressionNamedRerouteUsage_37}
    B {from=""}
   ]
   properties {constB=0.5}
  }
  MaterialExpressionMax_1 {
   class=MaterialExpressionMax
   inputs [
    A {from=MaterialExpressionRound_0}
    B {from=MaterialExpressionRound_1}
   ]
  }
  MaterialExpressionMultiply_0 {
   class=MaterialExpressionMultiply
   inputs [
    A {from=MaterialExpressionNamedRerouteDeclaration_2}
    B {from=MaterialExpressionMultiply_3}
   ]
  }
  MaterialExpressionMultiply_1 {
   class=MaterialExpressionMultiply
   inputs [
    A {from=MaterialExpressionOneMinus_1}
    B {from=MaterialExpressionOneMinus_2}
   ]
  }
  MaterialExpressionMultiply_10 {
   class=MaterialExpressionMultiply
   inputs [
    A {from=MaterialExpressionConvert_13}
    B {from=MaterialExpressionMultiply_14}
   ]
  }
  MaterialExpressionMultiply_11 {
   class=MaterialExpressionMultiply
   inputs [
    A {from=MaterialExpressionVectorParameter_5, output=RGB}
    B {from=MaterialExpressionAdd_3}
   ]
  }
  MaterialExpressionMultiply_12 {
   class=MaterialExpressionMultiply
   inputs [
    A {from=MaterialExpressionConvert_9}
    B {from=MaterialExpressionScalarParameter_8}
   ]
   properties {constB=2.5}
  }
  MaterialExpressionMultiply_13 {
   class=MaterialExpressionMultiply
   inputs [
    A {from=MaterialExpressionTime_2}
    B {from=MaterialExpressionScalarParameter_7}
   ]
  }
  MaterialExpressionMultiply_14 {
   class=MaterialExpressionMultiply
   inputs [
    A {from=MaterialExpressionSaturate_1}
    B {from=MaterialExpressionSaturate_9}
   ]
   properties {constA=0.1}
  }
  MaterialExpressionMultiply_15 {
   class=MaterialExpressionMultiply
   inputs [
    A {from=MaterialExpressionNamedRerouteDeclaration_3}
    B {from=MaterialExpressionScalarParameter_15}
   ]
   properties {constB=0.2}
  }
  MaterialExpressionMultiply_16 {
   class=MaterialExpressionMultiply
   inputs [
    A {from=MaterialExpressionTextureSample_3, output=R}
    B {from=""}
   ]
   properties {constB=0.3}
  }
  MaterialExpressionMultiply_17 {
   class=MaterialExpressionMultiply
   inputs [
    A {from=MaterialExpressionSaturate_3}
    B {from=MaterialExpressionScalarParameter_18}
   ]
  }
  MaterialExpressionMultiply_2 {
   class=MaterialExpressionMultiply
   inputs [
    A {from=MaterialExpressionNamedRerouteUsage_29}
    B {from=MaterialExpressionScalarParameter_4}
   ]
  }
  MaterialExpressionMultiply_20 {
   class=MaterialExpressionMultiply
   inputs [
    A {from=MaterialExpressionMultiply_3}
    B {from=MaterialExpressionOneMinus_0}
   ]
  }
  MaterialExpressionMultiply_25 {
   class=MaterialExpressionMultiply
   inputs [
    A {from=MaterialExpressionNormalize_0}
    B {from=MaterialExpressionLinearInterpolate_1}
   ]
   properties {constB=12}
  }
  MaterialExpressionMultiply_3 {
   class=MaterialExpressionMultiply
   inputs [
    A {from=MaterialExpressionLinearInterpolate_10}
    B {from=MaterialExpressionMax_0}
   ]
  }
  MaterialExpressionMultiply_4 {
   class=MaterialExpressionMultiply
   inputs [
    A {from=MaterialExpressionPower_0}
    B {from=MaterialExpressionTextureSample_3, output=R}
   ]
  }
  MaterialExpressionMultiply_5 {
   class=MaterialExpressionMultiply
   inputs [
    A {from=MaterialExpressionSmoothStep_5}
    B {from=MaterialExpressionScalarParameter_3}
   ]
   properties {constB=0.5}
  }
  MaterialExpressionMultiply_6 {
   class=MaterialExpressionMultiply
   inputs [
    A {from=MaterialExpressionMultiply_9}
    B {from=MaterialExpressionMultiply_8}
   ]
  }
  MaterialExpressionMultiply_7 {
   class=MaterialExpressionMultiply
   inputs [
    A {from=MaterialExpressionNamedRerouteUsage_8}
    B {from=MaterialExpressionScalarParameter_5}
   ]
  }
  MaterialExpressionMultiply_8 {
   class=MaterialExpressionMultiply
   inputs [
    A {from=MaterialExpressionOneMinus_3}
    B {from=MaterialExpressionOneMinus_3}
   ]
  }
  MaterialExpressionMultiply_9 {
   class=MaterialExpressionMultiply
   inputs [
    A {from=MaterialExpressionVectorParameter_3, output=RGB}
    B {from=""}
   ]
   properties {constB=3.2}
  }
  MaterialExpressionNamedRerouteDeclaration_0 {
   class=MaterialExpressionNamedRerouteDeclaration
   inputs [None {from=MaterialExpressionMultiply_20}]
   properties {name=PrimaryBaseColor, nodeColor=<rgba 0 1 0.3203 1>}
  }
  MaterialExpressionNamedRerouteDeclaration_1 {
   class=MaterialExpressionNamedRerouteDeclaration
   inputs [None {from=MaterialExpressionReroute_0}]
   properties {name=Debug, nodeColor=<rgba 0.1719 0 1 1>}
  }
  MaterialExpressionNamedRerouteDeclaration_10 {
   class=MaterialExpressionNamedRerouteDeclaration
   inputs [None {from=MaterialExpressionTextureSampleParameter2D_3, output=RGB}]
   properties {name=SuperlevelMasks02, nodeColor=<rgba 0.2188 0 1 1>}
  }
  MaterialExpressionNamedRerouteDeclaration_11 {
   class=MaterialExpressionNamedRerouteDeclaration
   inputs [None {from=MaterialExpressionSmoothStep_4}]
   properties {name=BlueSandMask, nodeColor=<rgba 0.0005 0.86 1 1>}
  }
  MaterialExpressionNamedRerouteDeclaration_12 {
   class=MaterialExpressionNamedRerouteDeclaration
   inputs [None {from=MaterialExpressionLinearInterpolate_21}]
   properties {name=Emissive_OUT, nodeColor=<rgba 0.2422 1 0 1>}
  }
  MaterialExpressionNamedRerouteDeclaration_13 {
   class=MaterialExpressionNamedRerouteDeclaration
   inputs [None {from=MaterialExpressionGetMaterialAttributes_0, output=EmissiveColor}]
   properties {name=Emissive_IN, nodeColor=<rgba 0.0781 0 1 1>}
  }
  MaterialExpressionNamedRerouteDeclaration_14 {
   class=MaterialExpressionNamedRerouteDeclaration
   inputs [None {from=MaterialExpressionLinearInterpolate_22}]
   properties {name=Basecolor_OUT, nodeColor=<rgba 0 1 0.8594 1>}
  }
  MaterialExpressionNamedRerouteDeclaration_15 {
   class=MaterialExpressionNamedRerouteDeclaration
   inputs [None {from=MaterialExpressionGetMaterialAttributes_0, output=BaseColor}]
   properties {name=Basecolor_IN, nodeColor=<rgba 1 0 0.8203 1>}
  }
  MaterialExpressionNamedRerouteDeclaration_16 {
   class=MaterialExpressionNamedRerouteDeclaration
   inputs [None {from=MaterialExpressionLinearInterpolate_14}]
   properties {name=Roughness_OUT, nodeColor=<rgba 0 1 0.8594 1>}
  }
  MaterialExpressionNamedRerouteDeclaration_17 {
   class=MaterialExpressionNamedRerouteDeclaration
   inputs [None {from=MaterialExpressionMultiply_11}]
   properties {name=TertiaryEmissiveColor, nodeColor=<rgba 0.0809 0.3312 1 1>}
  }
  MaterialExpressionNamedRerouteDeclaration_18 {
   class=MaterialExpressionNamedRerouteDeclaration
   inputs [None {from=MaterialExpressionFresnel_0}]
   properties {name=TertiaryFresnel, nodeColor=<rgba 0 0.8594 1 1>}
  }
  MaterialExpressionNamedRerouteDeclaration_19 {
   class=MaterialExpressionNamedRerouteDeclaration
   inputs [None {from=MaterialExpressionGetMaterialAttributes_0, output=Roughness}]
   properties {name=Roughness_IN, nodeColor=<rgba 0 1 0.9297 1>}
  }
  MaterialExpressionNamedRerouteDeclaration_2 {
   class=MaterialExpressionNamedRerouteDeclaration
   inputs [None {from=MaterialExpressionMax_1}]
   properties {name=PrimaryWaves, nodeColor=<rgba 0 1 0.6953 1>}
  }
  MaterialExpressionNamedRerouteDeclaration_20 {
   class=MaterialExpressionNamedRerouteDeclaration
   inputs [None {from=MaterialExpressionLinearInterpolate_23}]
   properties {name=EmissiveAccentMask, nodeColor=<rgba 1 0.0439 0.1233 1>}
  }
  MaterialExpressionNamedRerouteDeclaration_21 {
   class=MaterialExpressionNamedRerouteDeclaration
   inputs [None {from=MaterialExpressionConvert_13}]
   properties {name=AccentMask, nodeColor=<rgba 0.1217 1 0 1>}
  }
  MaterialExpressionNamedRerouteDeclaration_22 {
   class=MaterialExpressionNamedRerouteDeclaration
   inputs [None {from=MaterialExpressionConvert_13}]
   properties {name=TertiaryMask_1, nodeColor=<rgba 0.1304 0 0.9609 1>}
  }
  MaterialExpressionNamedRerouteDeclaration_23 {
   class=MaterialExpressionNamedRerouteDeclaration
   inputs [None {from=MaterialExpressionVectorParameter_13, output=RGB}]
   properties {name=AccentColor, nodeColor=<rgba 1 0.4453 0 1>}
  }
  MaterialExpressionNamedRerouteDeclaration_24 {
   class=MaterialExpressionNamedRerouteDeclaration
   inputs [None {from=MaterialExpressionLinearInterpolate_13}]
   properties {name=Specular_OUT, nodeColor=<rgba 0.5234 1 0 1>}
  }
  MaterialExpressionNamedRerouteDeclaration_25 {
   class=MaterialExpressionNamedRerouteDeclaration
   inputs [None {from=MaterialExpressionFunctionInput_1}]
   properties {name=PS-LocPos, nodeColor=<rgba 0 1 0.1797 1>}
  }
  MaterialExpressionNamedRerouteDeclaration_26 {
   class=MaterialExpressionNamedRerouteDeclaration
   inputs [None {from=MaterialExpressionLinearInterpolate_16}]
   properties {name=SecondaryColor, nodeColor=<rgba 0.4062 0 1 1>}
  }
  MaterialExpressionNamedRerouteDeclaration_27 {
   class=MaterialExpressionNamedRerouteDeclaration
   inputs [None {from=MaterialExpressionLinearInterpolate_2}]
   properties {name=TertiaryBasecolor, nodeColor=<rgba 0 1 0.5313 1>}
  }
  MaterialExpressionNamedRerouteDeclaration_28 {
   class=MaterialExpressionNamedRerouteDeclaration
   inputs [None {from=MaterialExpressionMultiply_0}]
   properties {name=PrimaryEmissiveColor, nodeColor=<rgba 0.1953 1 0 1>}
  }
  MaterialExpressionNamedRerouteDeclaration_29 {
   class=MaterialExpressionNamedRerouteDeclaration
   inputs [None {from=MaterialExpressionFunctionInput_2}]
   properties {name=Fresnel_IN, nodeColor=<rgba 1 0.4219 0 1>}
  }
  MaterialExpressionNamedRerouteDeclaration_3 {
   class=MaterialExpressionNamedRerouteDeclaration
   inputs [None {from=MaterialExpressionMultiply_4}]
   properties {name=PrimaryDistortion, nodeColor=<rgba 1 0 0.3047 1>}
  }
  MaterialExpressionNamedRerouteDeclaration_30 {
   class=MaterialExpressionNamedRerouteDeclaration
   inputs [None {from=MaterialExpressionGetMaterialAttributes_0, output=Specular}]
   properties {name=Specular_IN, nodeColor=<rgba 0.7578 1 0 1>}
  }
  MaterialExpressionNamedRerouteDeclaration_31 {
   class=MaterialExpressionNamedRerouteDeclaration
   inputs [None {from=MaterialExpressionConvert_0}]
   properties {name=Gmap, nodeColor=<rgba 0.9922 0 1 1>}
  }
  MaterialExpressionNamedRerouteDeclaration_32 {
   class=MaterialExpressionNamedRerouteDeclaration
   inputs [None {from=MaterialExpressionLinearInterpolate_18}]
   properties {name=Metallic_OUT, nodeColor=<rgba 0.1719 0 1 1>}
  }
  MaterialExpressionNamedRerouteDeclaration_33 {
   class=MaterialExpressionNamedRerouteDeclaration
   inputs [None {from=MaterialExpressionGetMaterialAttributes_0, output=Metallic}]
   properties {name=Metallic_IN, nodeColor=<rgba 0.1484 1 0 1>}
  }
  MaterialExpressionNamedRerouteDeclaration_34 {
   class=MaterialExpressionNamedRerouteDeclaration
   inputs [None {from=MaterialExpressionMultiply_6}]
   properties {name=EmissiveAccentColor, nodeColor=<rgba 0.0809 0.3312 1 1>}
  }
  MaterialExpressionNamedRerouteDeclaration_4 {
   class=MaterialExpressionNamedRerouteDeclaration
   inputs [None {from=MaterialExpressionMultiply_1}]
   properties {name=NonSL_Mask, nodeColor=<rgba 1 0.7266 0 1>}
  }
  MaterialExpressionNamedRerouteDeclaration_5 {
   class=MaterialExpressionNamedRerouteDeclaration
   inputs [None {from=MaterialExpressionFunctionInput_4}]
   properties {name=TechArtMask, nodeColor=<rgba 0 0.9531 1 1>}
  }
  MaterialExpressionNamedRerouteDeclaration_6 {
   class=MaterialExpressionNamedRerouteDeclaration
   inputs [None {from=MaterialExpressionConvert_12}]
   properties {name=PrimaryMask, nodeColor=<rgba 1 0.0439 0.1233 1>}
  }
  MaterialExpressionNamedRerouteDeclaration_7 {
   class=MaterialExpressionNamedRerouteDeclaration
   inputs [None {from=MaterialExpressionConvert_12}]
   properties {name=SecondaryMask, nodeColor=<rgba 0.1217 1 0 1>}
  }
  MaterialExpressionNamedRerouteDeclaration_8 {
   class=MaterialExpressionNamedRerouteDeclaration
   inputs [None {from=MaterialExpressionConvert_12}]
   properties {name=TertiaryMask, nodeColor=<rgba 0.1304 0 0.9609 1>}
  }
  MaterialExpressionNamedRerouteDeclaration_9 {
   class=MaterialExpressionNamedRerouteDeclaration
   inputs [None {from=MaterialExpressionTextureSampleParameter2D_1, output=RGB}]
   properties {name=SuperlevelMasks01, nodeColor=<rgba 1 0.1931 0 1>}
  }
  MaterialExpressionNamedRerouteUsage_0 {class=MaterialExpressionNamedRerouteUsage}
  MaterialExpressionNamedRerouteUsage_1 {class=MaterialExpressionNamedRerouteUsage}
  MaterialExpressionNamedRerouteUsage_10 {class=MaterialExpressionNamedRerouteUsage}
  MaterialExpressionNamedRerouteUsage_11 {class=MaterialExpressionNamedRerouteUsage}
  MaterialExpressionNamedRerouteUsage_12 {class=MaterialExpressionNamedRerouteUsage}
  MaterialExpressionNamedRerouteUsage_13 {class=MaterialExpressionNamedRerouteUsage}
  MaterialExpressionNamedRerouteUsage_14 {class=MaterialExpressionNamedRerouteUsage}
  MaterialExpressionNamedRerouteUsage_15 {class=MaterialExpressionNamedRerouteUsage}
  MaterialExpressionNamedRerouteUsage_16 {class=MaterialExpressionNamedRerouteUsage}
  MaterialExpressionNamedRerouteUsage_17 {class=MaterialExpressionNamedRerouteUsage}
  MaterialExpressionNamedRerouteUsage_18 {class=MaterialExpressionNamedRerouteUsage}
  MaterialExpressionNamedRerouteUsage_19 {class=MaterialExpressionNamedRerouteUsage}
  MaterialExpressionNamedRerouteUsage_2 {class=MaterialExpressionNamedRerouteUsage}
  MaterialExpressionNamedRerouteUsage_20 {class=MaterialExpressionNamedRerouteUsage}
  MaterialExpressionNamedRerouteUsage_21 {class=MaterialExpressionNamedRerouteUsage}
  MaterialExpressionNamedRerouteUsage_22 {class=MaterialExpressionNamedRerouteUsage}
  MaterialExpressionNamedRerouteUsage_23 {class=MaterialExpressionNamedRerouteUsage}
  MaterialExpressionNamedRerouteUsage_24 {class=MaterialExpressionNamedRerouteUsage}
  MaterialExpressionNamedRerouteUsage_25 {class=MaterialExpressionNamedRerouteUsage}
  MaterialExpressionNamedRerouteUsage_26 {class=MaterialExpressionNamedRerouteUsage}
  MaterialExpressionNamedRerouteUsage_27 {class=MaterialExpressionNamedRerouteUsage}
  MaterialExpressionNamedRerouteUsage_28 {class=MaterialExpressionNamedRerouteUsage}
  MaterialExpressionNamedRerouteUsage_29 {class=MaterialExpressionNamedRerouteUsage}
  MaterialExpressionNamedRerouteUsage_3 {class=MaterialExpressionNamedRerouteUsage}
  MaterialExpressionNamedRerouteUsage_30 {class=MaterialExpressionNamedRerouteUsage}
  MaterialExpressionNamedRerouteUsage_31 {class=MaterialExpressionNamedRerouteUsage}
  MaterialExpressionNamedRerouteUsage_32 {class=MaterialExpressionNamedRerouteUsage}
  MaterialExpressionNamedRerouteUsage_33 {class=MaterialExpressionNamedRerouteUsage}
  MaterialExpressionNamedRerouteUsage_34 {class=MaterialExpressionNamedRerouteUsage}
  MaterialExpressionNamedRerouteUsage_35 {class=MaterialExpressionNamedRerouteUsage}
  MaterialExpressionNamedRerouteUsage_36 {class=MaterialExpressionNamedRerouteUsage}
  MaterialExpressionNamedRerouteUsage_37 {class=MaterialExpressionNamedRerouteUsage}
  MaterialExpressionNamedRerouteUsage_38 {class=MaterialExpressionNamedRerouteUsage}
  MaterialExpressionNamedRerouteUsage_39 {class=MaterialExpressionNamedRerouteUsage}
  MaterialExpressionNamedRerouteUsage_4 {class=MaterialExpressionNamedRerouteUsage}
  MaterialExpressionNamedRerouteUsage_40 {class=MaterialExpressionNamedRerouteUsage}
  MaterialExpressionNamedRerouteUsage_41 {class=MaterialExpressionNamedRerouteUsage}
  MaterialExpressionNamedRerouteUsage_42 {class=MaterialExpressionNamedRerouteUsage}
  MaterialExpressionNamedRerouteUsage_43 {class=MaterialExpressionNamedRerouteUsage}
  MaterialExpressionNamedRerouteUsage_44 {class=MaterialExpressionNamedRerouteUsage}
  MaterialExpressionNamedRerouteUsage_45 {class=MaterialExpressionNamedRerouteUsage}
  MaterialExpressionNamedRerouteUsage_46 {class=MaterialExpressionNamedRerouteUsage}
  MaterialExpressionNamedRerouteUsage_47 {class=MaterialExpressionNamedRerouteUsage}
  MaterialExpressionNamedRerouteUsage_48 {class=MaterialExpressionNamedRerouteUsage}
  MaterialExpressionNamedRerouteUsage_49 {class=MaterialExpressionNamedRerouteUsage}
  MaterialExpressionNamedRerouteUsage_5 {class=MaterialExpressionNamedRerouteUsage}
  MaterialExpressionNamedRerouteUsage_50 {class=MaterialExpressionNamedRerouteUsage}
  MaterialExpressionNamedRerouteUsage_6 {class=MaterialExpressionNamedRerouteUsage}
  MaterialExpressionNamedRerouteUsage_7 {class=MaterialExpressionNamedRerouteUsage}
  MaterialExpressionNamedRerouteUsage_8 {class=MaterialExpressionNamedRerouteUsage}
  MaterialExpressionNamedRerouteUsage_9 {class=MaterialExpressionNamedRerouteUsage}
  MaterialExpressionNormalize_0 {class=MaterialExpressionNormalize, inputs [VectorInput {from=MaterialExpressionSubtract_2, output=""}]}
  MaterialExpressionOneMinus_0 {class=MaterialExpressionOneMinus, inputs [None {from=MaterialExpressionNamedRerouteUsage_30}]}
  MaterialExpressionOneMinus_1 {class=MaterialExpressionOneMinus, inputs [None {from=MaterialExpressionConvert_0}]}
  MaterialExpressionOneMinus_2 {class=MaterialExpressionOneMinus, inputs [None {from=MaterialExpressionLinearInterpolate_25}]}
  MaterialExpressionOneMinus_3 {class=MaterialExpressionOneMinus, inputs [None {from=MaterialExpressionNamedRerouteUsage_46}]}
  MaterialExpressionOneMinus_4 {class=MaterialExpressionOneMinus, inputs [None {from=MaterialExpressionSphereMask_0}]}
  MaterialExpressionPower_0 {
   class=MaterialExpressionPower
   inputs [
    Base {from=MaterialExpressionOneMinus_4}
    Exp {from=""}
   ]
  }
  MaterialExpressionQualitySwitch_0 {
   class=MaterialExpressionQualitySwitch
   inputs [
    Default {from=MaterialExpressionAdd_15, output=""}
    Low {from=MaterialExpressionReroute_1, output=""}
    High {from="", output=""}
    Medium {from="", output=""}
    Epic {from="", output=""}
   ]
   properties {}
  }
  MaterialExpressionReroute_0 {class=MaterialExpressionReroute, inputs [None {from=MaterialExpressionTextureSample_2, output=B}]}
  MaterialExpressionReroute_1 {class=MaterialExpressionReroute, inputs [None {from=MaterialExpressionNamedRerouteUsage_1}]}
  MaterialExpressionRotator_1 {
   class=MaterialExpressionRotator
   inputs [
    Coordinate {from=MaterialExpressionDivide_0, output=""}
    Time {from=MaterialExpressionScalarParameter_0, output=""}
   ]
   properties {centerX=0, speed=1}
  }
  MaterialExpressionRound_0 {class=MaterialExpressionRound, inputs [None {from=MaterialExpressionAdd_6, output=""}], properties {}}
  MaterialExpressionRound_1 {class=MaterialExpressionRound, inputs [None {from=MaterialExpressionAdd_11, output=""}], properties {}}
  MaterialExpressionSaturate_0 {class=MaterialExpressionSaturate, inputs [None {from=MaterialExpressionAdd_2}]}
  MaterialExpressionSaturate_1 {class=MaterialExpressionSaturate, inputs [None {from=MaterialExpressionAdd_5}]}
  MaterialExpressionSaturate_13 {class=MaterialExpressionSaturate, inputs [None {from=MaterialExpressionAdd_23}]}
  MaterialExpressionSaturate_2 {class=MaterialExpressionSaturate, inputs [None {from=MaterialExpressionDotProduct_0}]}
  MaterialExpressionSaturate_3 {class=MaterialExpressionSaturate, inputs [None {from=MaterialExpressionAdd_9}]}
  MaterialExpressionSaturate_4 {class=MaterialExpressionSaturate, inputs [None {from=MaterialExpressionDivide_2}]}
  MaterialExpressionSaturate_9 {class=MaterialExpressionSaturate, inputs [None {from=MaterialExpressionDotProduct_1}]}
  MaterialExpressionScalarParameter_0 {class=MaterialExpressionScalarParameter, properties {defaultValue=0.4, parameterName=Primary_Rotation}}
  MaterialExpressionScalarParameter_1 {class=MaterialExpressionScalarParameter, properties {defaultValue=0.1, parameterName=MINN}}
  MaterialExpressionScalarParameter_10 {class=MaterialExpressionScalarParameter, properties {defaultValue=0.4, parameterName=Accent_Roughness}}
  MaterialExpressionScalarParameter_11 {class=MaterialExpressionScalarParameter, properties {parameterName=Secondary_GradientStart}}
  MaterialExpressionScalarParameter_12 {class=MaterialExpressionScalarParameter, properties {defaultValue=110, parameterName=Secondary_GradientEnd}}
  MaterialExpressionScalarParameter_13 {class=MaterialExpressionScalarParameter, properties {defaultValue=0.4, parameterName=SandMax}}
  MaterialExpressionScalarParameter_14 {class=MaterialExpressionScalarParameter, properties {defaultValue=0.7, parameterName=Primary_Roughness2}}
  MaterialExpressionScalarParameter_15 {class=MaterialExpressionScalarParameter, properties {defaultValue=0.2, parameterName=Primary_DistortionIntensity}}
  MaterialExpressionScalarParameter_16 {class=MaterialExpressionScalarParameter, properties {parameterName=IsRowLiaison_Cast_Body}}
  MaterialExpressionScalarParameter_17 {class=MaterialExpressionScalarParameter, properties {parameterName=IsZenCrashSand}}
  MaterialExpressionScalarParameter_18 {class=MaterialExpressionScalarParameter, properties {group=SUPERLEVEL, parameterName=IsMarineCarve}}
  MaterialExpressionScalarParameter_19 {class=MaterialExpressionScalarParameter, properties {group=SUPERLEVEL, parameterName=IsFloralMane}}
  MaterialExpressionScalarParameter_2 {class=MaterialExpressionScalarParameter, properties {defaultValue=0.3, parameterName=MAXX}}
  MaterialExpressionScalarParameter_20 {class=MaterialExpressionScalarParameter, properties {group=SUPERLEVEL, parameterName=Tertiary_SmoothStep_Min}}
  MaterialExpressionScalarParameter_21 {class=MaterialExpressionScalarParameter, properties {defaultValue=1, group=SUPERLEVEL, parameterName=Tertiary_SmoothStep_Max}}
  MaterialExpressionScalarParameter_24 {class=MaterialExpressionScalarParameter, properties {group=SUPERLEVEL, parameterName=Caustics_ParallaxDepth}}
  MaterialExpressionScalarParameter_3 {class=MaterialExpressionScalarParameter, properties {defaultValue=0.2, parameterName=SandSubstract}}
  MaterialExpressionScalarParameter_4 {class=MaterialExpressionScalarParameter, properties {defaultValue=0.3, parameterName=Primary_EmissiveIntensity}}
  MaterialExpressionScalarParameter_5 {class=MaterialExpressionScalarParameter, properties {defaultValue=0.8, parameterName=Primary_BasecolorIntensity}}
  MaterialExpressionScalarParameter_6 {class=MaterialExpressionScalarParameter, properties {defaultValue=0.45, parameterName=Primary_Roughness}}
  MaterialExpressionScalarParameter_7 {class=MaterialExpressionScalarParameter, properties {defaultValue=-0.045, parameterName=Primary_Speed}}
  MaterialExpressionScalarParameter_8 {class=MaterialExpressionScalarParameter, properties {defaultValue=2, parameterName=Stripes_Scale}}
  MaterialExpressionScalarParameter_9 {class=MaterialExpressionScalarParameter, properties {defaultValue=1, parameterName=Primary_WaveMetallic}}
  MaterialExpressionSetMaterialAttributes_2 {
   class=MaterialExpressionSetMaterialAttributes
   inputs [
    MaterialAttributes {from=MaterialExpressionGetMaterialAttributes_0, output=MaterialAttributes}
    EmissiveColor {from=MaterialExpressionNamedRerouteUsage_10, output=""}
    BaseColor {from=MaterialExpressionNamedRerouteUsage_13, output=""}
    Metallic {from=MaterialExpressionNamedRerouteUsage_38, output=""}
    Specular {from=MaterialExpressionNamedRerouteUsage_17, output=""}
    Roughness {from=MaterialExpressionNamedRerouteUsage_18, output=""}
   ]
   properties {attributeSetTypes [
     #1
     #2
     #3
     #4
     #5
    ]}
  }
  MaterialExpressionSetMaterialAttributes_6 {
   class=MaterialExpressionSetMaterialAttributes
   inputs [
    MaterialAttributes {from=MaterialExpressionSetMaterialAttributes_2, output=""}
    OpacityMask {from=MaterialExpressionComponentMask_12, output=""}
   ]
   properties {attributeSetTypes [#6]}
  }
  MaterialExpressionShadingPathSwitch_0 {
   class=MaterialExpressionShadingPathSwitch
   inputs [
    Default {from=MaterialExpressionQualitySwitch_0, output=""}
    Deferred {from=MaterialExpressionQualitySwitch_0, output=""}
    Forward {from=MaterialExpressionReroute_1, output=""}
    Mobile {from=MaterialExpressionReroute_1, output=""}
   ]
   properties {}
  }
  MaterialExpressionSmoothStep_0 {
   class=MaterialExpressionSmoothStep
   inputs [
    Min {from="", output=""}
    Max {from="", output=""}
    Value {from=MaterialExpressionNamedRerouteDeclaration_18, output=""}
   ]
   properties {constMax=0.5}
  }
  MaterialExpressionSmoothStep_1 {
   class=MaterialExpressionSmoothStep
   inputs [
    Min {from="", output=""}
    Max {from="", output=""}
    Value {from=MaterialExpressionSphereMask_0, output=""}
   ]
   properties {constMax=0.8}
  }
  MaterialExpressionSmoothStep_11 {
   class=MaterialExpressionSmoothStep
   inputs [
    Min {from="", output=""}
    Max {from=MaterialExpressionScalarParameter_13, output=""}
    Value {from=MaterialExpressionReroute_0, output=""}
   ]
   properties {constMax=0.2}
  }
  MaterialExpressionSmoothStep_12 {
   class=MaterialExpressionSmoothStep
   inputs [
    Min {from=MaterialExpressionScalarParameter_12, output=""}
    Max {from=MaterialExpressionScalarParameter_11, output=""}
    Value {from=MaterialExpressionConvert_14, output=""}
   ]
   properties {}
  }
  MaterialExpressionSmoothStep_2 {
   class=MaterialExpressionSmoothStep
   inputs [
    Min {from=MaterialExpressionScalarParameter_20, output=""}
    Max {from=MaterialExpressionScalarParameter_21, output=""}
    Value {from=MaterialExpressionDesaturation_0, output=""}
   ]
   properties {}
  }
  MaterialExpressionSmoothStep_4 {
   class=MaterialExpressionSmoothStep
   inputs [
    Min {from=MaterialExpressionScalarParameter_1, output=""}
    Max {from=MaterialExpressionScalarParameter_2, output=""}
    Value {from=MaterialExpressionSubtract_0, output=""}
   ]
   properties {constMax=0.26, constMin=0.1}
  }
  MaterialExpressionSmoothStep_5 {
   class=MaterialExpressionSmoothStep
   inputs [
    Min {from="", output=""}
    Max {from="", output=""}
    Value {from=MaterialExpressionReroute_0, output=""}
   ]
   properties {constMax=0, constMin=0.01}
  }
  MaterialExpressionSmoothStep_9 {
   class=MaterialExpressionSmoothStep
   inputs [
    Min {from="", output=""}
    Max {from="", output=""}
    Value {from=MaterialExpressionTextureSample_1, output=B}
   ]
   properties {constMax=0.29, constMin=0.28}
  }
  MaterialExpressionSphereMask_0 {
   class=MaterialExpressionSphereMask
   inputs [
    A {from=MaterialExpressionNamedRerouteUsage_43, output=""}
    B {from=MaterialExpressionVectorParameter_1, output=RGB}
    Radius {from="", output=""}
    Hardness {from="", output=""}
   ]
   properties {attenuationRadius=120, hardnessPercent=2}
  }
  MaterialExpressionSubtract_0 {
   class=MaterialExpressionSubtract
   inputs [
    A {from=MaterialExpressionAdd_0}
    B {from=MaterialExpressionMultiply_5}
   ]
   properties {constB=0.2}
  }
  MaterialExpressionSubtract_1 {
   class=MaterialExpressionSubtract
   inputs [
    A {from=MaterialExpressionTextureSample_3, output=R}
    B {from=MaterialExpressionSmoothStep_1}
   ]
  }
  MaterialExpressionSubtract_2 {
   class=MaterialExpressionSubtract
   inputs [
    A {from=MaterialExpressionNamedRerouteUsage_1}
    B {from=MaterialExpressionTransformPosition_0}
   ]
  }
  MaterialExpressionTextureObject_0 {
   class=MaterialExpressionTextureObject
   inputs []
   properties {
    samplerType=SAMPLERTYPE_LinearColor
    texture=/BRCosmetics/Characters/Player/Male/Medium/Bodies/M_MED_FloralMane_Sand/Skins/SuperLevels/Textures/T_SL_S39_T2_TA
   }
  }
  MaterialExpressionTextureSampleParameter2D_1 {
   class=MaterialExpressionTextureSampleParameter2D
   inputs [
    UVs {from="", output=""}
    Apply View MipBias {from="", output=""}
   ]
   properties {
    group=SUPERLEVEL
    parameterName=SuperlevelMask_01
    samplerSource=SSM_Wrap_WorldGroupSettings
    samplerType=SAMPLERTYPE_Masks
    sortPriority=0
    texture=/BRCosmetics/Characters/Player/Male/Medium/Bodies/M_MED_ZenCrash/Textures/T_ZenCrash_Body_SL_FX1
   }
  }
  MaterialExpressionTextureSampleParameter2D_3 {
   class=MaterialExpressionTextureSampleParameter2D
   inputs [
    UVs {from="", output=""}
    Apply View MipBias {from="", output=""}
   ]
   properties {
    group=SUPERLEVEL
    parameterName=SuperLevelsMask_02
    samplerType=SAMPLERTYPE_Masks
    sortPriority=0
    texture=/BRCosmetics/Characters/Player/Male/Medium/Bodies/M_MED_ZenCrash/Textures/T_ZenCrash_Body_SL_FX2
   }
  }
  MaterialExpressionTextureSample_1 {
   class=MaterialExpressionTextureSample
   inputs [
    UVs {from=MaterialExpressionAdd_8}
    Tex {from=MaterialExpressionTextureObject_0}
    Apply View MipBias {from=""}
   ]
   properties {samplerSource=SSM_Wrap_WorldGroupSettings, samplerType=SAMPLERTYPE_LinearColor}
  }
  MaterialExpressionTextureSample_2 {
   class=MaterialExpressionTextureSample
   inputs [
    UVs {from=MaterialExpressionAdd_16}
    Tex {from=MaterialExpressionTextureObject_0}
    Apply View MipBias {from=""}
   ]
   properties {samplerSource=SSM_Wrap_WorldGroupSettings, samplerType=SAMPLERTYPE_LinearColor}
  }
  MaterialExpressionTextureSample_3 {
   class=MaterialExpressionTextureSample
   inputs [
    UVs {from=MaterialExpressionAdd_4}
    Tex {from=MaterialExpressionTextureObject_0}
    Apply View MipBias {from=""}
   ]
   properties {desc=Distortion, samplerSource=SSM_Wrap_WorldGroupSettings, samplerType=SAMPLERTYPE_LinearColor}
  }
  MaterialExpressionTextureSample_5 {
   class=MaterialExpressionTextureSample
   inputs [
    UVs {from=MaterialExpressionAdd_7}
    Tex {from=MaterialExpressionTextureObject_0}
    Apply View MipBias {from=""}
   ]
   properties {samplerSource=SSM_Wrap_WorldGroupSettings, samplerType=SAMPLERTYPE_LinearColor}
  }
  MaterialExpressionTime_2 {class=MaterialExpressionTime}
  MaterialExpressionTransformPosition_0 {
   class=MaterialExpressionTransformPosition
   inputs [None {from=MaterialExpressionCameraPositionWS_0, output=""}]
   properties {transformSourceType=TRANSFORMPOSSOURCE_World}
  }
  MaterialExpressionVectorParameter_0 {
   class=MaterialExpressionVectorParameter
   properties {parameterName=Tertiary_BasecolorOverride_Dark, defaultValue=<rgba 0.1968 0.0715 0.9062 1>}
  }
  MaterialExpressionVectorParameter_1 {class=MaterialExpressionVectorParameter, properties {parameterName=Primary_DistortionFadeOutCenter, defaultValue=<rgba -50 0 500 1>}}
  MaterialExpressionVectorParameter_10 {class=MaterialExpressionVectorParameter, properties {parameterName=DarkOrange, defaultValue=<rgba 0.7708 0.1725 0 1>}}
  MaterialExpressionVectorParameter_11 {class=MaterialExpressionVectorParameter, properties {parameterName=Blue, defaultValue=<rgba 0.0869 1.4108 2 1>}}
  MaterialExpressionVectorParameter_13 {class=MaterialExpressionVectorParameter, properties {parameterName=Accent_Color, defaultValue=<rgba 0.7011 0.0048 0.2016 1>}}
  MaterialExpressionVectorParameter_2 {class=MaterialExpressionVectorParameter, properties {parameterName=Primary_UVoffset, defaultValue=<rgba 0 0 0 1>}}
  MaterialExpressionVectorParameter_3 {class=MaterialExpressionVectorParameter, properties {parameterName=EmissiveAccentColor, defaultValue=<rgba 0.6939 0.008 0.1981 1>}}
  MaterialExpressionVectorParameter_5 {class=MaterialExpressionVectorParameter, properties {parameterName=Tertiary_Color, defaultValue=<rgba 0.1635 0.0445 0.8958 1>}}
  MaterialExpressionVectorParameter_6 {class=MaterialExpressionVectorParameter, properties {parameterName=Tertiary_BasecolorOverride, defaultValue=<rgba 0.093 0.0346 0.2656 1>}}
  MaterialExpressionVectorParameter_7 {class=MaterialExpressionVectorParameter, properties {parameterName=Pink, defaultValue=<rgba 0.807 0.0194 0.3095 1>}}
  MaterialExpressionVectorParameter_8 {class=MaterialExpressionVectorParameter, properties {parameterName=Black, defaultValue=<rgba 0.0242 0.0437 0.0482 1>}}
  MaterialExpressionVectorParameter_9 {class=MaterialExpressionVectorParameter, properties {parameterName=Orange, defaultValue=<rgba 1 0.3185 0.016 1>}}
  MaterialExpressionVertexInterpolator_1 {class=MaterialExpressionVertexInterpolator, inputs [VS {from=MaterialExpressionLocalPosition_1, output=XYZ}], properties {}}
 }}
