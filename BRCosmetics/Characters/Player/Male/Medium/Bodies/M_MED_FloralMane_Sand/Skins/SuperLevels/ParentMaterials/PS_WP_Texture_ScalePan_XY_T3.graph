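# Material function that projects a scale-panned texture from two sides (XZ and YZ) for animated ribbons on a character skin.
# Pre-skinned local position and normal are the defaults, keeping the projection fixed to the animating mesh; contrast-adjusted normal X blends the planes.
# A second ribbon layer with its own tiling, speed and opacity is added, then saturated and dimmed by FadingGradient with a floor.
# Quantity-threshold, rotator and sine branches are unconnected.
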MaterialFunction /BRCosmetics/Characters/Player/Male/Medium/Bodies/M_MED_FloralMane_Sand/Skins/SuperLevels/ParentMaterials/PS_WP_Texture_ScalePan_XY_T3
graph {
 properties {libraryCategoriesText []}
 nodes [
  ConstantBiasScale(None=Sine(None=Time()))
  Constant() {r=6.2832}
  n1 = ScalarParameter() {defaultValue=1.5, parameterName=AdditionalRibbons_Tiling}
  n2 = LocalPosition() {localOrigin=InstancePreSkinning}
  n3 = FunctionInput() {bUsePreviewValueAsDefault=true, description=Size of texture in worldspace units, inputName=TextureSize, previewValue=<64 64 64 6>}
  n4 = Divide(
   A=FunctionInput(Preview=VertexInterpolator(VS=@n2.XYZ).PS) {bUsePreviewValueAsDefault=true, inputName=WorldPosition}
   B=Multiply(A=Abs(None=@n3)) {constB=-1}
  )
  n5 = ComponentMask(None=@n4) {b=true, r=true}
  n6 = Time()
  n7 = FunctionInput(Preview=Constant2Vector()) {bUsePreviewValueAsDefault=true, inputType=FunctionInput_Vector2}
  n8 = Multiply(A=ScalarParameter() {defaultValue=0.1, parameterName=AdditionalRibbons_SpeedMult}, B=@n7) {constA=0.5}
  n9 = Panner(Coordinate=Multiply(A=@n1, B=@n5), Time=@n6, Speed=@n8) {bFractionalPart=true}
  n10 = FunctionInput(Preview=TextureObject() {texture=/Engine/EngineMaterials/DefaultDiffuse}) {inputName=TextureObject, inputType=FunctionInput_Texture2D}
  n11 = Constant()
  n12 = ComponentMask(None=@n4) {b=true, g=true}
  n13 = Panner(Coordinate=Multiply(A=@n1, B=@n12), Time=@n6, Speed=@n8) {bFractionalPart=true}
  n14 = VertexInterpolator(VS=PreSkinnedNormal())
  n15 = Abs(None=ComponentMask(None=FunctionInput(Preview=@n14.PS) {bUsePreviewValueAsDefault=true, inputName=World Space Normal, previewValue=<0 0 1 1>}) {r=true})
  n16 = FunctionInput() {
   bUsePreviewValueAsDefault=true
   inputName=ProjectionTransitionContrast
   inputType=FunctionInput_Scalar
   sortPriority=10
   previewValue=<1 0 0 1>
  }
  n17 = OneMinus(None=ComponentMask(None=MaterialFunctionCall(In=@n15, Contrast=@n16) {materialFunction=&CheapContrast}.Result) {r=true})
  n18 = LinearInterpolate(
   A=TextureSample!(UVs=@n9, Tex=@n10, Level=@n11, Apply View MipBias) {mipValueMode=TMVM_MipLevel, samplerSource=SSM_Wrap_WorldGroupSettings}.RGB
   B=TextureSample(UVs=@n13, Tex=@n10, Level=@n11) {mipValueMode=TMVM_MipLevel, samplerSource=SSM_Wrap_WorldGroupSettings}.RGB
   Alpha=@n17
  )
  n19 = LinearInterpolate(
   A=TextureSample(UVs=Panner(Coordinate=@n5, Time=@n6, Speed=@n7) {bFractionalPart=true}, Tex=@n10, Level=@n11) {mipValueMode=TMVM_MipLevel, samplerSource=SSM_Wrap_WorldGroupSettings}.RGB
   B=TextureSample(UVs=Panner(Coordinate=@n12, Time=@n6, Speed=@n7) {bFractionalPart=true}, Tex=@n10, Level=@n11) {mipValueMode=TMVM_MipLevel, samplerSource=SSM_Wrap_WorldGroupSettings}.RGB
   Alpha=@n17
  ) {desc=Float3}
  n20 = Add(A=Multiply(A=@n18, B=ScalarParameter() {defaultValue=0.4, parameterName=AdditionalRibbons_Opacity}), B=@n19)
  n21 = OneMinus(None=FunctionInput() {bUsePreviewValueAsDefault=true, inputName=FadingGradient, inputType=FunctionInput_Scalar})
  FunctionOutput(None=Multiply(A=Saturate(None=@n20), B=Max(A=@n21) {constB=0.65})) {outputName=XY Texture, sortPriority=1}
  n22 = Add(
   A=ComponentMask(None=@n18) {b=true}
   B=ConstantBiasScale(None=ScalarParameter() {defaultValue=0.5, parameterName=AdditionalRibbons_Quantity}) {bias=-0.5, scale=2}
  )
  Multiply(A=Round(None=Saturate(None=@n22)), B=@n18)
  n23 = Add(
   A=ComponentMask(None=@n19) {b=true}
   B=ConstantBiasScale(None=ScalarParameter() {defaultValue=1, parameterName=Ribbons_Quantity}) {bias=-0.5, scale=2}
  )
  Multiply(A=@n19, B=Round(None=Saturate(None=@n23)))
  n24 = Multiply(A=Time()) {constB=0.001}
  Rotator(Coordinate=@n13, Time=@n24)
  Rotator(Coordinate=@n9, Time=@n24)
 ]
}
