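# Tier 3 super-level (S39) skin material function for the FloralMane Sand player body, restyling an incoming material with animated effects.
# It chains material-attribute blends driven by mask channels and scalar parameters, adding car clear coat and blended normals.
# Effects draw on pre-skinned position gradients, sphere masks, time-driven panning XY-projected textures and noise, Fresnel and sky-light environment sampling.
# Bump-offset tiled UVs add parallax detail.
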MaterialFunction /BRCosmetics/Characters/Player/Male/Medium/Bodies/M_MED_FloralMane_Sand/Skins/SuperLevels/ParentMaterials/MF_SL_S39_T3
graph {expressions {
  MaterialExpressionAbs_0 {class=MaterialExpressionAbs, inputs [None {from=MaterialExpressionComponentMask_17, output=""}], properties {}}
  MaterialExpressionAdd_0 {
   class=MaterialExpressionAdd
   inputs [
    A {from=MaterialExpressionDivide_6}
    B {from=MaterialExpressionMultiply_0}
   ]
  }
  MaterialExpressionAdd_1 {
   class=MaterialExpressionAdd
   inputs [
    A {from=MaterialExpressionDesaturation_0}
    B {from=MaterialExpressionOneMinus_0}
   ]
   properties {constB=0.3}
  }
  MaterialExpressionAdd_10 {
   class=MaterialExpressionAdd
   inputs [
    A {from=MaterialExpressionNamedRerouteUsage_22}
    B {from=MaterialExpressionScalarParameter_10}
   ]
  }
  MaterialExpressionAdd_11 {
   class=MaterialExpressionAdd
   inputs [
    A {from=MaterialExpressionNamedRerouteUsage_24}
    B {from=MaterialExpressionScalarParameter_11}
   ]
  }
  MaterialExpressionAdd_12 {
   class=MaterialExpressionAdd
   inputs [
    A {from=MaterialExpressionNamedRerouteDeclaration_12}
    B {from=MaterialExpressionVertexNormalWS_2}
   ]
  }
  MaterialExpressionAdd_13 {
   class=MaterialExpressionAdd
   inputs [
    A {from=MaterialExpressionNamedRerouteUsage_30}
    B {from=MaterialExpressionVertexNormalWS_3}
   ]
  }
  MaterialExpressionAdd_14 {
   class=MaterialExpressionAdd
   inputs [
    A {from=MaterialExpressionScalarParameter_37}
    B {from=MaterialExpressionScalarParameter_40}
   ]
  }
  MaterialExpressionAdd_17 {
   class=MaterialExpressionAdd
   inputs [
    A {from=MaterialExpressionDivide_0}
    B {from=MaterialExpressionDivide_1}
   ]
  }
  MaterialExpressionAdd_18 {
   class=MaterialExpressionAdd
   inputs [
    A {from=MaterialExpressionGetMaterialAttributes_5, output=EmissiveColor}
    B {from=MaterialExpressionMultiply_24}
   ]
  }
  MaterialExpressionAdd_19 {
   class=MaterialExpressionAdd
   inputs [
    A {from=MaterialExpressionSphereMask_1}
    B {from=MaterialExpressionLinearInterpolate_1}
   ]
  }
  MaterialExpressionAdd_2 {
   class=MaterialExpressionAdd
   inputs [
    A {from=MaterialExpressionComponentMask_4}
    B {from=""}
   ]
   properties {constB=-0.65}
  }
  MaterialExpressionAdd_21 {
   class=MaterialExpressionAdd
   inputs [
    A {from=MaterialExpressionComponentMask_20}
    B {from=MaterialExpressionMultiply_14}
   ]
  }
  MaterialExpressionAdd_22 {
   class=MaterialExpressionAdd
   inputs [
    A {from=MaterialExpressionMultiply_16}
    B {from=""}
   ]
   properties {constB=0.5}
  }
  MaterialExpressionAdd_23 {
   class=MaterialExpressionAdd
   inputs [
    A {from=MaterialExpressionComponentMask_16}
    B {from=MaterialExpressionDotProduct_0}
   ]
  }
  MaterialExpressionAdd_25 {
   class=MaterialExpressionAdd
   inputs [
    A {from=MaterialExpressionReroute_1}
    B {from=MaterialExpressionMultiply_42}
   ]
  }
  MaterialExpressionAdd_26 {
   class=MaterialExpressionAdd
   inputs [
    A {from=MaterialExpressionShadingPathSwitch_0}
    B {from=MaterialExpressionMultiply_39}
   ]
  }
  MaterialExpressionAdd_27 {
   class=MaterialExpressionAdd
   inputs [
    A {from=MaterialExpressionComponentMask_21}
    B {from=MaterialExpressionTime_2}
   ]
  }
  MaterialExpressionAdd_28 {
   class=MaterialExpressionAdd
   inputs [
    A {from=MaterialExpressionComponentMask_6}
    B {from=MaterialExpressionMultiply_43}
   ]
  }
  MaterialExpressionAdd_29 {
   class=MaterialExpressionAdd
   inputs [
    A {from=MaterialExpressionComponentMask_22}
    B {from=MaterialExpressionMultiply_44}
   ]
  }
  MaterialExpressionAdd_30 {
   class=MaterialExpressionAdd
   inputs [
    A {from=MaterialExpressionAdd_31}
    B {from=MaterialExpressionMultiply_48}
   ]
  }
  MaterialExpressionAdd_31 {
   class=MaterialExpressionAdd
   inputs [
    A {from=MaterialExpressionMultiply_47}
    B {from=MaterialExpressionMultiply_46}
   ]
  }
  MaterialExpressionAdd_32 {
   class=MaterialExpressionAdd
   inputs [
    A {from=MaterialExpressionVectorParameter_2, output=RGB}
    B {from=MaterialExpressionAdd_35}
   ]
  }
  MaterialExpressionAdd_33 {
   class=MaterialExpressionAdd
   inputs [
    A {from=MaterialExpressionMultiply_49}
    B {from=MaterialExpressionSphereMask_3}
   ]
  }
  MaterialExpressionAdd_34 {
   class=MaterialExpressionAdd
   inputs [
    A {from=MaterialExpressionAdd_33}
    B {from=MaterialExpressionMultiply_50}
   ]
  }
  MaterialExpressionAdd_35 {
   class=MaterialExpressionAdd
   inputs [
    A {from=MaterialExpressionNamedRerouteUsage_50}
    B {from=MaterialExpressionMultiply_51}
   ]
  }
  MaterialExpressionAdd_36 {
   class=MaterialExpressionAdd
   inputs [
    A {from=MaterialExpressionAdd_50}
    B {from=MaterialExpressionMultiply_60}
   ]
  }
  MaterialExpressionAdd_37 {
   class=MaterialExpressionAdd
   inputs [
    A {from=MaterialExpressionMultiply_58}
    B {from=MaterialExpressionMultiply_63}
   ]
  }
  MaterialExpressionAdd_38 {
   class=MaterialExpressionAdd
   inputs [
    A {from=MaterialExpressionComponentMask_6}
    B {from=""}
   ]
   properties {constB=-75}
  }
  MaterialExpressionAdd_39 {
   class=MaterialExpressionAdd
   inputs [
    A {from=MaterialExpressionMultiply_54}
    B {from=MaterialExpressionMultiply_55}
   ]
  }
  MaterialExpressionAdd_4 {
   class=MaterialExpressionAdd
   inputs [
    A {from=MaterialExpressionOneMinus_3}
    B {from=""}
   ]
   properties {constB=-0.2}
  }
  MaterialExpressionAdd_40 {
   class=MaterialExpressionAdd
   inputs [
    A {from=MaterialExpressionComponentMask_6}
    B {from=""}
   ]
   properties {constB=-77}
  }
  MaterialExpressionAdd_41 {
   class=MaterialExpressionAdd
   inputs [
    A {from=MaterialExpressionMultiply_64}
    B {from=MaterialExpressionMultiply_62}
   ]
  }
  MaterialExpressionAdd_42 {
   class=MaterialExpressionAdd
   inputs [
    A {from=MaterialExpressionMultiply_67}
    B {from=""}
   ]
   properties {constB=0.5}
  }
  MaterialExpressionAdd_43 {
   class=MaterialExpressionAdd
   inputs [
    A {from=MaterialExpressionPower_8}
    B {from=""}
   ]
   properties {constB=0.1}
  }
  MaterialExpressionAdd_44 {
   class=MaterialExpressionAdd
   inputs [
    A {from=MaterialExpressionMultiply_81}
    B {from=MaterialExpressionTextureSampleParameter2D_0, output=G}
   ]
  }
  MaterialExpressionAdd_45 {
   class=MaterialExpressionAdd
   inputs [
    A {from=MaterialExpressionMultiply_52}
    B {from=MaterialExpressionMultiply_70}
   ]
  }
  MaterialExpressionAdd_46 {
   class=MaterialExpressionAdd
   inputs [
    A {from=MaterialExpressionMultiply_72}
    B {from=MaterialExpressionNamedRerouteUsage_31}
   ]
  }
  MaterialExpressionAdd_47 {
   class=MaterialExpressionAdd
   inputs [
    A {from=MaterialExpressionSaturate_20}
    B {from=MaterialExpressionSaturate_19}
   ]
  }
  MaterialExpressionAdd_48 {
   class=MaterialExpressionAdd
   inputs [
    A {from=MaterialExpressionNamedRerouteUsage_19}
    B {from=MaterialExpressionNamedRerouteUsage_59}
   ]
  }
  MaterialExpressionAdd_50 {
   class=MaterialExpressionAdd
   inputs [
    A {from=MaterialExpressionMultiply_83}
    B {from=MaterialExpressionMultiply_84}
   ]
  }
  MaterialExpressionAdd_51 {
   class=MaterialExpressionAdd
   inputs [
    A {from=MaterialExpressionMultiply_85}
    B {from=MaterialExpressionComponentMask_29}
   ]
  }
  MaterialExpressionAdd_52 {
   class=MaterialExpressionAdd
   inputs [
    A {from=MaterialExpressionMultiply_87}
    B {from=MaterialExpressionMultiply_86}
   ]
  }
  MaterialExpressionAdd_6 {
   class=MaterialExpressionAdd
   inputs [
    A {from=MaterialExpressionMultiply_5}
    B {from=MaterialExpressionMultiply_6}
   ]
   properties {constB=-12}
  }
  MaterialExpressionAdd_8 {
   class=MaterialExpressionAdd
   inputs [
    A {from=MaterialExpressionMultiply_13}
    B {from=MaterialExpressionMultiply_18}
   ]
   properties {constB=-12}
  }
  MaterialExpressionAdd_9 {
   class=MaterialExpressionAdd
   inputs [
    A {from=MaterialExpressionMultiply_29}
    B {from=""}
   ]
   properties {constB=-1}
  }
  MaterialExpressionAtmosphericLightVector_0 {class=MaterialExpressionAtmosphericLightVector, inputs [], properties {}}
  MaterialExpressionBlendMaterialAttributes_0 {
   class=MaterialExpressionBlendMaterialAttributes
   inputs [
    A {from=MaterialExpressionNamedRerouteUsage_1, output=""}
    B {from=MaterialExpressionSetMaterialAttributes_2, output=""}
    Alpha {from=MaterialExpressionComponentMask_0, output=""}
   ]
   properties {}
  }
  MaterialExpressionBlendMaterialAttributes_1 {
   class=MaterialExpressionBlendMaterialAttributes
   inputs [
    A {from=MaterialExpressionBlendMaterialAttributes_0, output=""}
    B {from=MaterialExpressionSetMaterialAttributes_1, output=""}
    Alpha {from=MaterialExpressionComponentMask_1, output=""}
   ]
   properties {}
  }
  MaterialExpressionBlendMaterialAttributes_2 {
   class=MaterialExpressionBlendMaterialAttributes
   inputs [
    A {from=MaterialExpressionBlendMaterialAttributes_1, output=""}
    B {from=MaterialExpressionSetMaterialAttributes_3, output=""}
    Alpha {from=MaterialExpressionComponentMask_2, output=""}
   ]
   properties {}
  }
  MaterialExpressionBlendMaterialAttributes_3 {
   class=MaterialExpressionBlendMaterialAttributes
   inputs [
    A {from=MaterialExpressionBlendMaterialAttributes_4, output=""}
    B {from=MaterialExpressionSetMaterialAttributes_4, output=""}
    Alpha {from=MaterialExpressionComponentMask_7, output=""}
   ]
   properties {}
  }
  MaterialExpressionBlendMaterialAttributes_4 {
   class=MaterialExpressionBlendMaterialAttributes
   inputs [
    A {from=MaterialExpressionBlendMaterialAttributes_2, output=""}
    B {from=MaterialExpressionSetMaterialAttributes_5, output=""}
    Alpha {from=MaterialExpressionMultiply_9, output=""}
   ]
   properties {}
  }
  MaterialExpressionBlendMaterialAttributes_5 {
   class=MaterialExpressionBlendMaterialAttributes
   inputs [
    A {from=MaterialExpressionBlendMaterialAttributes_3, output=""}
    B {from=MaterialExpressionSetMaterialAttributes_0, output=""}
    Alpha {from=MaterialExpressionLinearInterpolate_17, output=""}
   ]
   properties {}
  }
  MaterialExpressionBlendMaterialAttributes_6 {
   class=MaterialExpressionBlendMaterialAttributes
   inputs [
    A {from=MaterialExpressionBlendMaterialAttributes_5, output=""}
    B {from=MaterialExpressionSetMaterialAttributes_6, output=""}
    Alpha {from=MaterialExpressionScalarParameter_55, output=""}
   ]
   properties {}
  }
  MaterialExpressionBumpOffset_2 {
   class=MaterialExpressionBumpOffset
   inputs [
    Coordinate {from=MaterialExpressionMultiply_11, output=""}
    Height {from=MaterialExpressionScalarParameter_38, output=""}
    HeightRatioInput {from="", output=""}
   ]
   properties {}
  }
  MaterialExpressionCameraPositionWS_0 {class=MaterialExpressionCameraPositionWS, inputs [], properties {}}
  MaterialExpressionCameraVectorWS_0 {class=MaterialExpressionCameraVectorWS, inputs [], properties {}}
  MaterialExpressionComponentMask_0 {class=MaterialExpressionComponentMask, inputs [None {from=MaterialExpressionNamedRerouteUsage_2}], properties {r=true}}
  MaterialExpressionComponentMask_1 {class=MaterialExpressionComponentMask, inputs [None {from=MaterialExpressionNamedRerouteUsage_3}], properties {g=true}}
  MaterialExpressionComponentMask_10 {class=MaterialExpressionComponentMask, inputs [None {from=MaterialExpressionTextureCoordinate_0}], properties {r=true}}
  MaterialExpressionComponentMask_12 {class=MaterialExpressionComponentMask, inputs [None {from=MaterialExpressionNamedRerouteUsage_21}], properties {r=true}}
  MaterialExpressionComponentMask_13 {class=MaterialExpressionComponentMask, inputs [None {from=MaterialExpressionNamedRerouteUsage_12}], properties {b=true}}
  MaterialExpressionComponentMask_14 {class=MaterialExpressionComponentMask, inputs [None {from=MaterialExpressionVectorParameter_0, output=RGB}], properties {b=true}}
  MaterialExpressionComponentMask_15 {class=MaterialExpressionComponentMask, inputs [None {from=MaterialExpressionNamedRerouteUsage_39}], properties {b=true, r=true}}
  MaterialExpressionComponentMask_16 {class=MaterialExpressionComponentMask, inputs [None {from=MaterialExpressionNamedRerouteUsage_29}], properties {b=true}}
  MaterialExpressionComponentMask_17 {class=MaterialExpressionComponentMask, inputs [None {from=MaterialExpressionNamedRerouteDeclaration_9}], properties {g=true}}
  MaterialExpressionComponentMask_18 {class=MaterialExpressionComponentMask, inputs [None {from=MaterialExpressionMaterialFunctionCall_4, output=XY Texture}], properties {g=true}}
  MaterialExpressionComponentMask_19 {class=MaterialExpressionComponentMask, inputs [None {from=MaterialExpressionReroute_6}], properties {r=true}}
  MaterialExpressionComponentMask_2 {class=MaterialExpressionComponentMask, inputs [None {from=MaterialExpressionNamedRerouteUsage_4}], properties {b=true}}
  MaterialExpressionComponentMask_20 {class=MaterialExpressionComponentMask, inputs [None {from=MaterialExpressionReroute_6}], properties {g=true}}
  MaterialExpressionComponentMask_21 {class=MaterialExpressionComponentMask, inputs [None {from=MaterialExpressionReroute_6}], properties {b=true}}
  MaterialExpressionComponentMask_22 {class=MaterialExpressionComponentMask, inputs [None {from=MaterialExpressionNamedRerouteUsage_9}], properties {r=true}}
  MaterialExpressionComponentMask_23 {class=MaterialExpressionComponentMask, inputs [None {from=MaterialExpressionNoise_1}], properties {g=true}}
  MaterialExpressionComponentMask_24 {class=MaterialExpressionComponentMask, inputs [None {from=MaterialExpressionConstant3Vector_9}], properties {b=true}}
  MaterialExpressionComponentMask_25 {class=MaterialExpressionComponentMask, inputs [None {from=MaterialExpressionLocalPosition_2, output=XYZ}], properties {b=true}}
  MaterialExpressionComponentMask_26 {class=MaterialExpressionComponentMask, inputs [None {from=MaterialExpressionDivide_4}], properties {b=true, r=true}}
  MaterialExpressionComponentMask_27 {class=MaterialExpressionComponentMask, inputs [None {from=MaterialExpressionRotator_0}], properties {g=true}}
  MaterialExpressionComponentMask_28 {class=MaterialExpressionComponentMask, inputs [None {from=MaterialExpressionNamedRerouteUsage_64}], properties {g=true}}
  MaterialExpressionComponentMask_29 {class=MaterialExpressionComponentMask, inputs [None {from=MaterialExpressionNamedRerouteUsage_9}], properties {b=true}}
  MaterialExpressionComponentMask_3 {class=MaterialExpressionComponentMask, inputs [None {from=MaterialExpressionNamedRerouteUsage_6}], properties {b=true}}
  MaterialExpressionComponentMask_30 {class=MaterialExpressionComponentMask, inputs [None {from=MaterialExpressionMaterialFunctionCall_4, output=XY Texture}], properties {r=true}}
  MaterialExpressionComponentMask_4 {class=MaterialExpressionComponentMask, inputs [None {from=MaterialExpressionTextureCoordinate_0}], properties {g=true}}
  MaterialExpressionComponentMask_5 {class=MaterialExpressionComponentMask, inputs [None {from=MaterialExpressionNamedRerouteUsage_8}], properties {r=true}}
  MaterialExpressionComponentMask_6 {class=MaterialExpressionComponentMask, inputs [None {from=MaterialExpressionNamedRerouteUsage_9}], properties {b=true}}
  MaterialExpressionComponentMask_7 {class=MaterialExpressionComponentMask, inputs [None {from=MaterialExpressionNamedRerouteUsage_15}], properties {g=true}}
  MaterialExpressionComponentMask_8 {class=MaterialExpressionComponentMask, inputs [None {from=MaterialExpressionNamedRerouteUsage_16}], properties {r=true}}
  MaterialExpressionComponentMask_9 {
   class=MaterialExpressionComponentMask
   inputs [None {from=MaterialExpressionPreSkinnedLocalBounds_2, output=Half Extents}]
   properties {b=true}
  }
  MaterialExpressionConstant2Vector_1 {class=MaterialExpressionConstant2Vector, properties {g=-0.2, r=0.094}}
  MaterialExpressionConstant3Vector_0 {class=MaterialExpressionConstant3Vector, properties {constant=<rgba 0 0 -0.1 0>}}
  MaterialExpressionConstant3Vector_1 {class=MaterialExpressionConstant3Vector, properties {constant=<rgba 0.3646 0.3646 0.3646 0>}}
  MaterialExpressionConstant3Vector_10 {class=MaterialExpressionConstant3Vector, properties {constant=<rgba 0 1 1 0>}}
  MaterialExpressionConstant3Vector_11 {class=MaterialExpressionConstant3Vector, properties {constant=<rgba 1 1 1 1>}}
  MaterialExpressionConstant3Vector_12 {class=MaterialExpressionConstant3Vector, properties {constant=<rgba 0.9479 1 0.4713 1>}}
  MaterialExpressionConstant3Vector_13 {class=MaterialExpressionConstant3Vector, properties {constant=<rgba 0 0 -1 0>}}
  MaterialExpressionConstant3Vector_2 {class=MaterialExpressionConstant3Vector, properties {constant=<rgba 0.06 0 -0.25 0>}}
  MaterialExpressionConstant3Vector_3 {class=MaterialExpressionConstant3Vector, properties {constant=<rgba 0.025 0.025 0.025 1>}}
  MaterialExpressionConstant3Vector_5 {class=MaterialExpressionConstant3Vector, properties {constant=<rgba 1 1 1 0>}}
  MaterialExpressionConstant3Vector_7 {class=MaterialExpressionConstant3Vector, properties {constant=<rgba 0.7934 0.7934 0.7934 1>}}
  MaterialExpressionConstant3Vector_8 {class=MaterialExpressionConstant3Vector, properties {constant=<rgba 1 0 -3 0>}}
  MaterialExpressionConstant3Vector_9 {class=MaterialExpressionConstant3Vector}
  MaterialExpressionConstantBiasScale_1 {class=MaterialExpressionConstantBiasScale, inputs [None {from=MaterialExpressionSine_3}]}
  MaterialExpressionConstantBiasScale_3 {class=MaterialExpressionConstantBiasScale, inputs [None {from=MaterialExpressionSine_5}]}
  MaterialExpressionConstant_0 {class=MaterialExpressionConstant, properties {r=0.3}}
  MaterialExpressionConstant_1 {class=MaterialExpressionConstant, properties {r=1}}
  MaterialExpressionConstant_10 {class=MaterialExpressionConstant}
  MaterialExpressionConstant_11 {class=MaterialExpressionConstant}
  MaterialExpressionConstant_12 {class=MaterialExpressionConstant, properties {r=0.04}}
  MaterialExpressionConstant_13 {class=MaterialExpressionConstant, properties {r=1}}
  MaterialExpressionConstant_14 {class=MaterialExpressionConstant, properties {r=0.4}}
  MaterialExpressionConstant_15 {class=MaterialExpressionConstant, properties {r=1}}
  MaterialExpressionConstant_16 {class=MaterialExpressionConstant, properties {r=3.1416}}
  MaterialExpressionConstant_17 {class=MaterialExpressionConstant, properties {r=0.3}}
  MaterialExpressionConstant_18 {class=MaterialExpressionConstant}
  MaterialExpressionConstant_2 {class=MaterialExpressionConstant, properties {r=1}}
  MaterialExpressionConstant_20 {class=MaterialExpressionConstant}
  MaterialExpressionConstant_3 {class=MaterialExpressionConstant, properties {r=0.03}}
  MaterialExpressionConstant_5 {class=MaterialExpressionConstant, properties {r=0.5}}
  MaterialExpressionConstant_6 {class=MaterialExpressionConstant, properties {r=0.02}}
  MaterialExpressionConstant_7 {class=MaterialExpressionConstant, properties {r=1}}
  MaterialExpressionConstant_8 {class=MaterialExpressionConstant, properties {r=1}}
  MaterialExpressionConstant_9 {class=MaterialExpressionConstant}
  MaterialExpressionConvert_1 {
   class=MaterialExpressionConvert
   inputs [
    None {from=MaterialExpressionVectorParameter_1, output=R}
    None {from=MaterialExpressionVectorParameter_1, output=R}
   ]
   properties {
    nodeName=Make Vector2
    convertInputs [
     {
      type=Scalar
      expressionInput {
       expression=&self:MaterialExpressionVectorParameter_1
       inputName=None
       mask=1
       maskA=0
       maskB=0
       maskG=0
       maskR=1
       outputIndex=1
      }
     }
     {
      type=Scalar
      expressionInput {
       expression=&self:MaterialExpressionVectorParameter_1
       inputName=None
       mask=1
       maskA=0
       maskB=0
       maskG=1
       maskR=0
       outputIndex=2
      }
     }
    ]
    convertOutputs [{type=Vector2}]
   }
  }
  MaterialExpressionDesaturation_0 {
   class=MaterialExpressionDesaturation
   inputs [
    None {from=MaterialExpressionGetMaterialAttributes_0, output=BaseColor}
    Fraction {from=MaterialExpressionConstant_7, output=""}
   ]
   properties {}
  }
  MaterialExpressionDesaturation_2 {
   class=MaterialExpressionDesaturation
   inputs [
    None {from=MaterialExpressionGetMaterialAttributes_5, output=EmissiveColor}
    Fraction {from="", output=""}
   ]
   properties {}
  }
  MaterialExpressionDesaturation_4 {
   class=MaterialExpressionDesaturation
   inputs [
    None {from=MaterialExpressionSkyLightEnvMapSample_0, output=""}
    Fraction {from=MaterialExpressionScalarParameter_39, output=""}
   ]
   properties {}
  }
  MaterialExpressionDesaturation_5 {
   class=MaterialExpressionDesaturation
   inputs [
    None {from=MaterialExpressionGetMaterialAttributes_6, output=EmissiveColor}
    Fraction {from="", output=""}
   ]
   properties {}
  }
  MaterialExpressionDivide_0 {
   class=MaterialExpressionDivide
   inputs [
    A {from=MaterialExpressionOneMinus_5}
    B {from=MaterialExpressionScalarParameter_13}
   ]
  }
  MaterialExpressionDivide_1 {
   class=MaterialExpressionDivide
   inputs [
    A {from=MaterialExpressionComponentMask_9}
    B {from=""}
   ]
   properties {constB=100}
  }
  MaterialExpressionDivide_2 {
   class=MaterialExpressionDivide
   inputs [
    A {from=MaterialExpressionComponentMask_15}
    B {from=""}
   ]
   properties {constB=75}
  }
  MaterialExpressionDivide_3 {
   class=MaterialExpressionDivide
   inputs [
    A {from=MaterialExpressionSubtract_0}
    B {from=MaterialExpressionScalarParameter_36}
   ]
  }
  MaterialExpressionDivide_4 {
   class=MaterialExpressionDivide
   inputs [
    A {from=MaterialExpressionNamedRerouteUsage_55}
    B {from=""}
   ]
   properties {constB=80}
  }
  MaterialExpressionDivide_5 {
   class=MaterialExpressionDivide
   inputs [
    A {from=MaterialExpressionAdd_40}
    B {from=""}
   ]
   properties {constB=100}
  }
  MaterialExpressionDivide_6 {
   class=MaterialExpressionDivide
   inputs [
    A {from=MaterialExpressionNamedRerouteUsage_39}
    B {from=""}
   ]
   properties {constB=75}
  }
  MaterialExpressionDotProduct_0 {
   class=MaterialExpressionDotProduct
   inputs [
    A {from=MaterialExpressionNamedRerouteUsage_5, output=""}
    B {from=MaterialExpressionConstant3Vector_5, output=""}
   ]
   properties {}
  }
  MaterialExpressionDotProduct_1 {
   class=MaterialExpressionDotProduct
   inputs [
    A {from=MaterialExpressionNormalize_2, output=""}
    B {from=MaterialExpressionCameraVectorWS_0, output=""}
   ]
   properties {}
  }
  MaterialExpressionDotProduct_2 {
   class=MaterialExpressionDotProduct
   inputs [
    A {from=MaterialExpressionVertexNormalWS_1, output=""}
    B {from=MaterialExpressionAtmosphericLightVector_0, output=""}
   ]
   properties {}
  }
  MaterialExpressionDotProduct_3 {
   class=MaterialExpressionDotProduct
   inputs [
    A {from=MaterialExpressionNormalize_1, output=""}
    B {from=MaterialExpressionNormalize_4, output=""}
   ]
   properties {}
  }
  MaterialExpressionDotProduct_4 {
   class=MaterialExpressionDotProduct
   inputs [
    A {from=MaterialExpressionNamedRerouteUsage_53, output=""}
    B {from=MaterialExpressionConstant3Vector_10, output=""}
   ]
   properties {}
  }
  MaterialExpressionFrac_0 {class=MaterialExpressionFrac, inputs [None {from=MaterialExpressionAdd_44, output=""}], properties {}}
  MaterialExpressionFrac_1 {class=MaterialExpressionFrac, inputs [None {from=MaterialExpressionMultiply_45, output=""}], properties {}}
  MaterialExpressionFrac_3 {class=MaterialExpressionFrac, inputs [None {from=MaterialExpressionAdd_21, output=""}], properties {}}
  MaterialExpressionFrac_5 {class=MaterialExpressionFrac, inputs [None {from=MaterialExpressionMultiply_11, output=""}], properties {}}
  MaterialExpressionFresnel_0 {
   class=MaterialExpressionFresnel
   inputs [
    ExponentIn {from="", output=""}
    BaseReflectFractionIn {from="", output=""}
    Normal {from="", output=""}
   ]
   properties {baseReflectFraction=0, exponent=1}
  }
  MaterialExpressionFresnel_1 {
   class=MaterialExpressionFresnel
   inputs [
    ExponentIn {from="", output=""}
    BaseReflectFractionIn {from="", output=""}
    Normal {from="", output=""}
   ]
   properties {baseReflectFraction=0, exponent=1}
  }
  MaterialExpressionFunctionInput_0 {
   class=MaterialExpressionFunctionInput
   inputs [Preview {from="", output=""}]
   properties {inputName=Material, inputType=FunctionInput_MaterialAttributes}
  }
  MaterialExpressionFunctionInput_1 {class=MaterialExpressionFunctionInput, inputs [Preview {from="", output=""}], properties {inputName=M, sortPriority=1}}
  MaterialExpressionFunctionInput_2 {
   class=MaterialExpressionFunctionInput
   inputs [Preview {from=MaterialExpressionShaderStageSwitch_1, output=""}]
   properties {inputName=PreskinnedLocPos, sortPriority=2}
  }
  MaterialExpressionFunctionInput_3 {class=MaterialExpressionFunctionInput, inputs [Preview {from="", output=""}], properties {inputName=Tech Art Mask, sortPriority=3}}
  MaterialExpressionFunctionInput_4 {
   class=MaterialExpressionFunctionInput
   inputs [Preview {from=MaterialExpressionFresnel_0, output=""}]
   properties {inputName=Fresnel, inputType=FunctionInput_Scalar, sortPriority=4}
  }
  MaterialExpressionFunctionOutput_0 {class=MaterialExpressionFunctionOutput, inputs [None {from=MaterialExpressionNamedRerouteUsage_14, output=""}]}
  MaterialExpressionGetMaterialAttributes_0 {
   class=MaterialExpressionGetMaterialAttributes
   inputs [None {from=MaterialExpressionBlendMaterialAttributes_0, output=""}]
   properties {attributeGetTypes [#1, #2, #3]}
  }
  MaterialExpressionGetMaterialAttributes_1 {
   class=MaterialExpressionGetMaterialAttributes
   inputs [None {from=MaterialExpressionMaterialFunctionCall_5, output=Result}]
   properties {attributeGetTypes [#1, #2, #4]}
  }
  MaterialExpressionGetMaterialAttributes_2 {
   class=MaterialExpressionGetMaterialAttributes
   inputs [None {from=MaterialExpressionNamedRerouteUsage_1, output=""}]
   properties {attributeGetTypes [#1, #3]}
  }
  MaterialExpressionGetMaterialAttributes_3 {
   class=MaterialExpressionGetMaterialAttributes
   inputs [None {from=MaterialExpressionBlendMaterialAttributes_3, output=""}]
   properties {attributeGetTypes [#3]}
  }
  MaterialExpressionGetMaterialAttributes_5 {
   class=MaterialExpressionGetMaterialAttributes
   inputs [None {from=MaterialExpressionNamedRerouteUsage_42, output=""}]
   properties {attributeGetTypes [#5]}
  }
  MaterialExpressionGetMaterialAttributes_6 {
   class=MaterialExpressionGetMaterialAttributes
   inputs [None {from="", output=""}]
   properties {attributeGetTypes [#5, #3]}
  }
  MaterialExpressionGetMaterialAttributes_7 {
   class=MaterialExpressionGetMaterialAttributes
   inputs [None {from=MaterialExpressionNamedRerouteUsage_52, output=""}]
   properties {attributeGetTypes [#2]}
  }
  MaterialExpressionLinearInterpolate_0 {
   class=MaterialExpressionLinearInterpolate
   inputs [
    A {from=MaterialExpressionScalarParameter_3}
    B {from=""}
    Alpha {from=MaterialExpressionMultiply_2}
   ]
   properties {constB=3}
  }
  MaterialExpressionLinearInterpolate_1 {
   class=MaterialExpressionLinearInterpolate
   inputs [
    A {from=MaterialExpressionOneMinus_6}
    B {from=MaterialExpressionSaturate_8}
    Alpha {from=MaterialExpressionScalarParameter_9}
   ]
  }
  MaterialExpressionLinearInterpolate_10 {
   class=MaterialExpressionLinearInterpolate
   inputs [
    A {from=MaterialExpressionScalarParameter_2}
    B {from=""}
    Alpha {from=MaterialExpressionMultiply_2}
   ]
   properties {constB=1.23}
  }
  MaterialExpressionLinearInterpolate_11 {
   class=MaterialExpressionLinearInterpolate
   inputs [
    A {from=MaterialExpressionScalarParameter_23}
    B {from=""}
    Alpha {from=MaterialExpressionNamedRerouteUsage_33}
   ]
   properties {constB=0.35}
  }
  MaterialExpressionLinearInterpolate_12 {
   class=MaterialExpressionLinearInterpolate
   inputs [
    A {from=MaterialExpressionScalarParameter_22}
    B {from=""}
    Alpha {from=MaterialExpressionNamedRerouteUsage_46}
   ]
  }
  MaterialExpressionLinearInterpolate_13 {
   class=MaterialExpressionLinearInterpolate
   inputs [
    A {from=MaterialExpressionScalarParameter_26}
    B {from=""}
    Alpha {from=MaterialExpressionNamedRerouteUsage_38}
   ]
   properties {constB=0}
  }
  MaterialExpressionLinearInterpolate_14 {
   class=MaterialExpressionLinearInterpolate
   inputs [
    A {from=""}
    B {from=MaterialExpressionTextureSample_2, output=R}
    Alpha {from=MaterialExpressionSmoothStep_1}
   ]
   properties {constA=0.5, constB=0}
  }
  MaterialExpressionLinearInterpolate_16 {
   class=MaterialExpressionLinearInterpolate
   inputs [
    A {from=MaterialExpressionConstant3Vector_11}
    B {from=MaterialExpressionConstant3Vector_12}
    Alpha {from=MaterialExpressionFrac_0}
   ]
  }
  MaterialExpressionLinearInterpolate_17 {
   class=MaterialExpressionLinearInterpolate
   inputs [
    A {from=MaterialExpressionComponentMask_8}
    B {from=MaterialExpressionMultiply_56}
    Alpha {from=MaterialExpressionScalarParameter_56}
   ]
  }
  MaterialExpressionLinearInterpolate_18 {
   class=MaterialExpressionLinearInterpolate
   inputs [
    A {from=MaterialExpressionScalarParameter_58}
    B {from=MaterialExpressionMultiply_69}
    Alpha {from=MaterialExpressionNamedRerouteUsage_58}
   ]
   properties {constB=10}
  }
  MaterialExpressionLinearInterpolate_19 {
   class=MaterialExpressionLinearInterpolate
   inputs [
    A {from=MaterialExpressionScalarParameter_57}
    B {from=MaterialExpressionMultiply_68}
    Alpha {from=MaterialExpressionNamedRerouteUsage_58}
   ]
  }
  MaterialExpressionLinearInterpolate_2 {
   class=MaterialExpressionLinearInterpolate
   inputs [
    A {from=MaterialExpressionConstant_6}
    B {from=""}
    Alpha {from=MaterialExpressionNamedRerouteUsage_13}
   ]
   properties {constA=0.1, constB=0.15}
  }
  MaterialExpressionLinearInterpolate_20 {
   class=MaterialExpressionLinearInterpolate
   inputs [
    A {from=MaterialExpressionScalarParameter_65}
    B {from=MaterialExpressionMultiply_75}
    Alpha {from=MaterialExpressionNamedRerouteUsage_58}
   ]
   properties {constB=10}
  }
  MaterialExpressionLinearInterpolate_21 {
   class=MaterialExpressionLinearInterpolate
   inputs [
    A {from=MaterialExpressionOneMinus_14}
    B {from=MaterialExpressionSaturate_21}
    Alpha {from=MaterialExpressionScalarParameter_68}
   ]
  }
  MaterialExpressionLinearInterpolate_22 {
   class=MaterialExpressionLinearInterpolate
   inputs [
    A {from=MaterialExpressionSaturate_2}
    B {from=MaterialExpressionSaturate_14}
    Alpha {from=MaterialExpressionScalarParameter_70}
   ]
  }
  MaterialExpressionLinearInterpolate_23 {
   class=MaterialExpressionLinearInterpolate
   inputs [
    A {from=MaterialExpressionSmoothStep_3}
    B {from=MaterialExpressionSaturate_23}
    Alpha {from=MaterialExpressionMultiply_88}
   ]
  }
  MaterialExpressionLinearInterpolate_3 {
   class=MaterialExpressionLinearInterpolate
   inputs [
    A {from=MaterialExpressionLinearInterpolate_2}
    B {from=MaterialExpressionConstant3Vector_1}
    Alpha {from=MaterialExpressionNamedRerouteUsage_20}
   ]
  }
  MaterialExpressionLinearInterpolate_4 {
   class=MaterialExpressionLinearInterpolate
   inputs [
    A {from=MaterialExpressionConstant3Vector_3}
    B {from=MaterialExpressionConstant_17}
    Alpha {from=MaterialExpressionLinearInterpolate_23}
   ]
  }
  MaterialExpressionLinearInterpolate_5 {
   class=MaterialExpressionLinearInterpolate
   inputs [
    A {from=MaterialExpressionOneMinus_4}
    B {from=MaterialExpressionSaturate_3}
    Alpha {from=MaterialExpressionScalarParameter_8}
   ]
  }
  MaterialExpressionLinearInterpolate_6 {
   class=MaterialExpressionLinearInterpolate
   inputs [
    A {from=MaterialExpressionGetMaterialAttributes_1, output=Roughness}
    B {from=MaterialExpressionScalarParameter_12}
    Alpha {from=MaterialExpressionConstant_8}
   ]
  }
  MaterialExpressionLinearInterpolate_7 {
   class=MaterialExpressionLinearInterpolate
   inputs [
    A {from=MaterialExpressionGetMaterialAttributes_1, output=Normal}
    B {from=MaterialExpressionMaterialFunctionCall_6, output=Result}
    Alpha {from=MaterialExpressionConstant_8}
   ]
  }
  MaterialExpressionLinearInterpolate_8 {
   class=MaterialExpressionLinearInterpolate
   inputs [
    A {from=MaterialExpressionMultiply_26}
    B {from=""}
    Alpha {from=MaterialExpressionNamedRerouteUsage_41}
   ]
   properties {constB=0}
  }
  MaterialExpressionLinearInterpolate_9 {
   class=MaterialExpressionLinearInterpolate
   inputs [
    A {from=MaterialExpressionSaturate_1}
    B {from=""}
    Alpha {from=MaterialExpressionNamedRerouteUsage_23}
   ]
  }
  MaterialExpressionLocalPosition_0 {class=MaterialExpressionLocalPosition, inputs [], properties {}}
  MaterialExpressionLocalPosition_1 {class=MaterialExpressionLocalPosition, inputs [], properties {localOrigin=InstancePreSkinning}}
  MaterialExpressionLocalPosition_2 {class=MaterialExpressionLocalPosition, inputs [], properties {}}
  MaterialExpressionMaterialFunctionCall_1 {
   class=MaterialExpressionMaterialFunctionCall
   inputs [Float3 {from=MaterialExpressionReroute_5}]
   properties {materialFunction=&BreakOutFloat3Components}
  }
  MaterialExpressionMaterialFunctionCall_10 {
   class=MaterialExpressionMaterialFunctionCall
   inputs [UV Channel {from=MaterialExpressionFrac_3}]
   properties {materialFunction=&LinearGradient}
  }
  MaterialExpressionMaterialFunctionCall_11 {
   class=MaterialExpressionMaterialFunctionCall
   inputs [
    X {from=MaterialExpressionMultiply_12}
    Y {from=MaterialExpressionTextureSampleParameter2D_0, output=G}
    Z {from=MaterialExpressionDivide_3}
   ]
   properties {desc=could be removed if no blending is needed, materialFunction=&MakeFloat3}
  }
  MaterialExpressionMaterialFunctionCall_12 {
   class=MaterialExpressionMaterialFunctionCall
   inputs [
    Normal {from=MaterialExpressionNamedRerouteUsage_48}
    Flatness {from=MaterialExpressionConstant_14}
   ]
   properties {materialFunction=&FlattenNormal}
  }
  MaterialExpressionMaterialFunctionCall_2 {class=MaterialExpressionMaterialFunctionCall, inputs []}
  MaterialExpressionMaterialFunctionCall_3 {
   class=MaterialExpressionMaterialFunctionCall
   inputs [
    R {from=MaterialExpressionConstant_15}
    G {from=MaterialExpressionConstant_15}
    B {from=MaterialExpressionScalarParameter_59}
    A {from=""}
   ]
   properties {materialFunction=&AppendMany}
  }
  MaterialExpressionMaterialFunctionCall_4 {
   class=MaterialExpressionMaterialFunctionCall
   inputs [
    TextureObject {from=MaterialExpressionTextureObject_3}
    TextureSize {from=MaterialExpressionScalarParameter_24}
    WorldPosition {from=MaterialExpressionAdd_26}
    FadingGradient {from=MaterialExpressionNamedRerouteUsage_57}
    In {from=MaterialExpressionConvert_1}
    World Space Normal {from=MaterialExpressionNamedRerouteUsage_45}
    ProjectionTransitionContrast {from=MaterialExpressionScalarParameter_29}
   ]
   properties {materialFunction=&/PS_WP_Texture_ScalePan_XY_T3}
  }
  MaterialExpressionMaterialFunctionCall_5 {
   class=MaterialExpressionMaterialFunctionCall
   inputs [
    Mask {from=MaterialExpressionConstant_9}
    Fresnel {from=MaterialExpressionNamedRerouteUsage_26}
    In {from=MaterialExpressionNamedRerouteUsage_27}
   ]
   properties {materialFunction=/Game/Global/MaterialFunctions/TechArt/MF_Advanced_CarClearCoat_v2}
  }
  MaterialExpressionMaterialFunctionCall_6 {
   class=MaterialExpressionMaterialFunctionCall
   inputs [
    BaseNormal {from=MaterialExpressionGetMaterialAttributes_1, output=Normal}
    AdditionalNormal {from=MaterialExpressionMaterialFunctionCall_5, output=ClearCoatNormal}
   ]
   properties {materialFunction=&BlendAngleCorrectedNormals}
  }
  MaterialExpressionMaterialFunctionCall_7 {
   class=MaterialExpressionMaterialFunctionCall
   inputs [
    BaseNormal {from=MaterialExpressionGetMaterialAttributes_0, output=Normal}
    AdditionalNormal {from=MaterialExpressionMaterialFunctionCall_12, output=Result}
   ]
   properties {materialFunction=&BlendAngleCorrectedNormals}
  }
  MaterialExpressionMaterialFunctionCall_8 {class=MaterialExpressionMaterialFunctionCall, inputs []}
  MaterialExpressionMax_0 {
   class=MaterialExpressionMax
   inputs [
    A {from=MaterialExpressionConstantBiasScale_1}
    B {from=""}
   ]
   properties {constB=0.2}
  }
  MaterialExpressionMax_2 {
   class=MaterialExpressionMax
   inputs [
    A {from=MaterialExpressionConstantBiasScale_3}
    B {from=""}
   ]
   properties {constB=0.3}
  }
  MaterialExpressionMax_3 {
   class=MaterialExpressionMax
   inputs [
    A {from=MaterialExpressionComponentMask_28}
    B {from=""}
   ]
   properties {constB=0.3}
  }
  MaterialExpressionMax_4 {
   class=MaterialExpressionMax
   inputs [
    A {from=MaterialExpressionNamedRerouteUsage_40}
    B {from=MaterialExpressionLinearInterpolate_8}
   ]
  }
  MaterialExpressionMin_0 {
   class=MaterialExpressionMin
   inputs [
    A {from=MaterialExpressionGetMaterialAttributes_0, output=Roughness}
    B {from="", output=""}
   ]
   properties {constB=0.45}
  }
  MaterialExpressionMultiply_0 {
   class=MaterialExpressionMultiply
   inputs [
    A {from=MaterialExpressionTime_10}
    B {from=MaterialExpressionConstant3Vector_0}
   ]
  }
  MaterialExpressionMultiply_1 {
   class=MaterialExpressionMultiply
   inputs [
    A {from=MaterialExpressionSaturate_0}
    B {from=MaterialExpressionSaturate_11}
   ]
  }
  MaterialExpressionMultiply_10 {
   class=MaterialExpressionMultiply
   inputs [
    A {from=MaterialExpressionTextureCoordinate_3}
    B {from=MaterialExpressionScalarParameter_36}
   ]
   properties {constB=3}
  }
  MaterialExpressionMultiply_11 {
   class=MaterialExpressionMultiply
   inputs [
    A {from=MaterialExpressionMultiply_10}
    B {from=MaterialExpressionScalarParameter_41}
   ]
  }
  MaterialExpressionMultiply_12 {
   class=MaterialExpressionMultiply
   inputs [
    A {from=MaterialExpressionTextureSampleParameter2D_2, output=R}
    B {from=MaterialExpressionTextureSampleParameter2D_0, output=R}
   ]
  }
  MaterialExpressionMultiply_13 {
   class=MaterialExpressionMultiply
   inputs [
    A {from=MaterialExpressionAdd_28}
    B {from=MaterialExpressionScalarParameter_6}
   ]
   properties {constB=0.1}
  }
  MaterialExpressionMultiply_14 {
   class=MaterialExpressionMultiply
   inputs [
    A {from=MaterialExpressionAdd_27}
    B {from=MaterialExpressionScalarParameter_44}
   ]
   properties {constB=0.5}
  }
  MaterialExpressionMultiply_15 {
   class=MaterialExpressionMultiply
   inputs [
    A {from=MaterialExpressionPower_0}
    B {from=MaterialExpressionAdd_22}
   ]
  }
  MaterialExpressionMultiply_16 {
   class=MaterialExpressionMultiply
   inputs [
    A {from=MaterialExpressionSine_1}
    B {from=""}
   ]
   properties {constB=0.5}
  }
  MaterialExpressionMultiply_17 {
   class=MaterialExpressionMultiply
   inputs [
    A {from=MaterialExpressionGetMaterialAttributes_6, output=BaseColor}
    B {from=MaterialExpressionDesaturation_5}
   ]
  }
  MaterialExpressionMultiply_18 {
   class=MaterialExpressionMultiply
   inputs [
    A {from=MaterialExpressionScalarParameter_6}
    B {from=MaterialExpressionScalarParameter_7}
   ]
  }
  MaterialExpressionMultiply_19 {
   class=MaterialExpressionMultiply
   inputs [
    A {from=MaterialExpressionGetMaterialAttributes_3, output=BaseColor}
    B {from=""}
   ]
   properties {constB=0.1}
  }
  MaterialExpressionMultiply_2 {
   class=MaterialExpressionMultiply
   inputs [
    A {from=MaterialExpressionScalarParameter_21}
    B {from=MaterialExpressionNamedRerouteUsage_0}
   ]
  }
  MaterialExpressionMultiply_20 {
   class=MaterialExpressionMultiply
   inputs [
    A {from=MaterialExpressionMultiply_28}
    B {from=MaterialExpressionScalarParameter_15}
   ]
  }
  MaterialExpressionMultiply_21 {
   class=MaterialExpressionMultiply
   inputs [
    A {from=MaterialExpressionSaturate_9}
    B {from=MaterialExpressionNamedRerouteUsage_35}
   ]
  }
  MaterialExpressionMultiply_22 {
   class=MaterialExpressionMultiply
   inputs [
    A {from=MaterialExpressionPower_3}
    B {from=""}
   ]
   properties {constB=4}
  }
  MaterialExpressionMultiply_24 {
   class=MaterialExpressionMultiply
   inputs [
    A {from=MaterialExpressionMultiply_25}
    B {from=MaterialExpressionScalarParameter_19}
   ]
  }
  MaterialExpressionMultiply_25 {
   class=MaterialExpressionMultiply
   inputs [
    A {from=MaterialExpressionGetMaterialAttributes_5, output=EmissiveColor}
    B {from=MaterialExpressionPower_4}
   ]
  }
  MaterialExpressionMultiply_26 {
   class=MaterialExpressionMultiply
   inputs [
    A {from=MaterialExpressionAdd_18}
    B {from=MaterialExpressionScalarParameter_16}
   ]
  }
  MaterialExpressionMultiply_27 {
   class=MaterialExpressionMultiply
   inputs [
    A {from=MaterialExpressionSaturate_13}
    B {from=MaterialExpressionScalarParameter_46}
   ]
  }
  MaterialExpressionMultiply_28 {
   class=MaterialExpressionMultiply
   inputs [
    A {from=MaterialExpressionMultiply_17}
    B {from=MaterialExpressionVectorParameter_7, output=RGB}
   ]
  }
  MaterialExpressionMultiply_29 {
   class=MaterialExpressionMultiply
   inputs [
    A {from=MaterialExpressionTextureSample_1, output=RGB}
    B {from=""}
   ]
   properties {constB=2}
  }
  MaterialExpressionMultiply_30 {
   class=MaterialExpressionMultiply
   inputs [
    A {from=MaterialExpressionAdd_9}
    B {from=MaterialExpressionScalarParameter_34}
   ]
   properties {constB=0.5}
  }
  MaterialExpressionMultiply_31 {
   class=MaterialExpressionMultiply
   inputs [
    A {from=MaterialExpressionTextureCoordinate_1}
    B {from=MaterialExpressionScalarParameter_35}
   ]
  }
  MaterialExpressionMultiply_32 {
   class=MaterialExpressionMultiply
   inputs [
    A {from=MaterialExpressionMultiply_27}
    B {from=MaterialExpressionPower_1}
   ]
  }
  MaterialExpressionMultiply_33 {
   class=MaterialExpressionMultiply
   inputs [
    A {from=MaterialExpressionMultiply_34}
    B {from=MaterialExpressionScalarParameter_14}
   ]
  }
  MaterialExpressionMultiply_34 {
   class=MaterialExpressionMultiply
   inputs [
    A {from=MaterialExpressionNamedRerouteUsage_34}
    B {from=MaterialExpressionDesaturation_4}
   ]
  }
  MaterialExpressionMultiply_35 {
   class=MaterialExpressionMultiply
   inputs [
    A {from=MaterialExpressionConstant3Vector_7}
    B {from=MaterialExpressionPower_2}
   ]
  }
  MaterialExpressionMultiply_36 {
   class=MaterialExpressionMultiply
   inputs [
    A {from=MaterialExpressionSaturate_15}
    B {from=MaterialExpressionLinearInterpolate_0}
   ]
   properties {constB=1.5}
  }
  MaterialExpressionMultiply_37 {
   class=MaterialExpressionMultiply
   inputs [
    A {from=MaterialExpressionNamedRerouteUsage_49}
    B {from=MaterialExpressionLinearInterpolate_5}
   ]
  }
  MaterialExpressionMultiply_38 {
   class=MaterialExpressionMultiply
   inputs [
    A {from=MaterialExpressionAdd_29}
    B {from=""}
   ]
   properties {constB=0.05}
  }
  MaterialExpressionMultiply_39 {
   class=MaterialExpressionMultiply
   inputs [
    A {from=MaterialExpressionLinearInterpolate_14}
    B {from=MaterialExpressionScalarParameter_26}
   ]
   properties {constB=20}
  }
  MaterialExpressionMultiply_40 {
   class=MaterialExpressionMultiply
   inputs [
    A {from=MaterialExpressionScalarParameter_25}
    B {from=MaterialExpressionConstant2Vector_1}
   ]
  }
  MaterialExpressionMultiply_41 {
   class=MaterialExpressionMultiply
   inputs [
    A {from=MaterialExpressionAdd_45}
    B {from=MaterialExpressionScalarParameter_28}
   ]
  }
  MaterialExpressionMultiply_42 {
   class=MaterialExpressionMultiply
   inputs [
    A {from=MaterialExpressionNormalize_0}
    B {from=MaterialExpressionScalarParameter_30}
   ]
   properties {constB=12}
  }
  MaterialExpressionMultiply_43 {
   class=MaterialExpressionMultiply
   inputs [
    A {from=MaterialExpressionSine_2}
    B {from=""}
   ]
   properties {constB=2.8}
  }
  MaterialExpressionMultiply_44 {
   class=MaterialExpressionMultiply
   inputs [
    A {from=MaterialExpressionTime_3}
    B {from=""}
   ]
   properties {constB=-7}
  }
  MaterialExpressionMultiply_45 {
   class=MaterialExpressionMultiply
   inputs [
    A {from=MaterialExpressionAdd_30}
    B {from=MaterialExpressionScalarParameter_63}
   ]
  }
  MaterialExpressionMultiply_46 {
   class=MaterialExpressionMultiply
   inputs [
    A {from=MaterialExpressionTime_4}
    B {from=MaterialExpressionScalarParameter_47}
   ]
  }
  MaterialExpressionMultiply_47 {
   class=MaterialExpressionMultiply
   inputs [
    A {from=MaterialExpressionReroute_4}
    B {from=MaterialExpressionScalarParameter_48}
   ]
  }
  MaterialExpressionMultiply_48 {
   class=MaterialExpressionMultiply
   inputs [
    A {from=MaterialExpressionComponentMask_23}
    B {from=MaterialExpressionScalarParameter_49}
   ]
  }
  MaterialExpressionMultiply_49 {
   class=MaterialExpressionMultiply
   inputs [
    A {from=MaterialExpressionSphereMask_2}
    B {from=MaterialExpressionScalarParameter_51}
   ]
  }
  MaterialExpressionMultiply_5 {
   class=MaterialExpressionMultiply
   inputs [
    A {from=MaterialExpressionComponentMask_13}
    B {from=MaterialExpressionScalarParameter_4}
   ]
   properties {constB=0.1}
  }
  MaterialExpressionMultiply_50 {
   class=MaterialExpressionMultiply
   inputs [
    A {from=MaterialExpressionConstant_13}
    B {from=MaterialExpressionScalarParameter_54}
   ]
  }
  MaterialExpressionMultiply_51 {
   class=MaterialExpressionMultiply
   inputs [
    A {from=MaterialExpressionTime_5}
    B {from=MaterialExpressionConstant3Vector_8}
   ]
  }
  MaterialExpressionMultiply_52 {
   class=MaterialExpressionMultiply
   inputs [
    A {from=MaterialExpressionComponentMask_18}
    B {from=""}
   ]
  }
  MaterialExpressionMultiply_53 {
   class=MaterialExpressionMultiply
   inputs [
    A {from=MaterialExpressionAdd_47}
    B {from=MaterialExpressionMax_0}
   ]
  }
  MaterialExpressionMultiply_54 {
   class=MaterialExpressionMultiply
   inputs [
    A {from=MaterialExpressionComponentMask_25}
    B {from=""}
   ]
   properties {constB=0.01}
  }
  MaterialExpressionMultiply_55 {
   class=MaterialExpressionMultiply
   inputs [
    A {from=MaterialExpressionTime_6}
    B {from=""}
   ]
   properties {constB=-0.2}
  }
  MaterialExpressionMultiply_56 {
   class=MaterialExpressionMultiply
   inputs [
    A {from=MaterialExpressionComponentMask_8}
    B {from=MaterialExpressionSaturate_16}
   ]
  }
  MaterialExpressionMultiply_57 {
   class=MaterialExpressionMultiply
   inputs [
    A {from=MaterialExpressionPower_7}
    B {from=MaterialExpressionLinearInterpolate_19}
   ]
   properties {constB=2}
  }
  MaterialExpressionMultiply_58 {
   class=MaterialExpressionMultiply
   inputs [
    A {from=MaterialExpressionReroute_5}
    B {from=MaterialExpressionMaterialFunctionCall_3, output=RGB}
   ]
  }
  MaterialExpressionMultiply_6 {
   class=MaterialExpressionMultiply
   inputs [
    A {from=MaterialExpressionScalarParameter_4}
    B {from=MaterialExpressionScalarParameter_5}
   ]
  }
  MaterialExpressionMultiply_60 {
   class=MaterialExpressionMultiply
   inputs [
    A {from=MaterialExpressionNamedRerouteUsage_62}
    B {from=MaterialExpressionScalarParameter_69}
   ]
  }
  MaterialExpressionMultiply_61 {
   class=MaterialExpressionMultiply
   inputs [
    A {from=MaterialExpressionMultiply_82}
    B {from=MaterialExpressionMultiply_32}
   ]
  }
  MaterialExpressionMultiply_62 {
   class=MaterialExpressionMultiply
   inputs [
    A {from=MaterialExpressionTime_7}
    B {from=MaterialExpressionConstant3Vector_2}
   ]
  }
  MaterialExpressionMultiply_63 {
   class=MaterialExpressionMultiply
   inputs [
    A {from=MaterialExpressionNoise_0}
    B {from=MaterialExpressionScalarParameter_60}
   ]
  }
  MaterialExpressionMultiply_64 {
   class=MaterialExpressionMultiply
   inputs [
    A {from=MaterialExpressionScalarParameter_61}
    B {from=MaterialExpressionNamedRerouteUsage_54}
   ]
  }
  MaterialExpressionMultiply_66 {
   class=MaterialExpressionMultiply
   inputs [
    A {from=MaterialExpressionTime_8}
    B {from=""}
   ]
   properties {constB=0.176}
  }
  MaterialExpressionMultiply_67 {
   class=MaterialExpressionMultiply
   inputs [
    A {from=MaterialExpressionSine_4}
    B {from=""}
   ]
   properties {constB=0.5}
  }
  MaterialExpressionMultiply_68 {
   class=MaterialExpressionMultiply
   inputs [
    A {from=MaterialExpressionScalarParameter_57}
    B {from=""}
   ]
   properties {constB=1.5}
  }
  MaterialExpressionMultiply_69 {
   class=MaterialExpressionMultiply
   inputs [
    A {from=MaterialExpressionScalarParameter_58}
    B {from=""}
   ]
   properties {constB=1.5}
  }
  MaterialExpressionMultiply_70 {
   class=MaterialExpressionMultiply
   inputs [
    A {from=MaterialExpressionScalarParameter_64}
    B {from=MaterialExpressionSaturate_18}
   ]
   properties {constA=0.2}
  }
  MaterialExpressionMultiply_71 {
   class=MaterialExpressionMultiply
   inputs [
    A {from=MaterialExpressionAdd_48}
    B {from=MaterialExpressionMax_3}
   ]
  }
  MaterialExpressionMultiply_72 {
   class=MaterialExpressionMultiply
   inputs [
    A {from=MaterialExpressionMultiply_73}
    B {from=MaterialExpressionTime_10}
   ]
  }
  MaterialExpressionMultiply_73 {
   class=MaterialExpressionMultiply
   inputs [
    A {from=MaterialExpressionScalarParameter_62}
    B {from=MaterialExpressionConstant3Vector_13}
   ]
  }
  MaterialExpressionMultiply_74 {
   class=MaterialExpressionMultiply
   inputs [
    A {from=MaterialExpressionPower_9}
    B {from=""}
   ]
  }
  MaterialExpressionMultiply_75 {
   class=MaterialExpressionMultiply
   inputs [
    A {from=MaterialExpressionScalarParameter_65}
    B {from=""}
   ]
   properties {constB=1.5}
  }
  MaterialExpressionMultiply_76 {
   class=MaterialExpressionMultiply
   inputs [
    A {from=MaterialExpressionPower_10}
    B {from=MaterialExpressionTextureSample_12, output=B}
   ]
  }
  MaterialExpressionMultiply_77 {
   class=MaterialExpressionMultiply
   inputs [
    A {from=MaterialExpressionMultiply_76}
    B {from=""}
   ]
   properties {constB=0.03}
  }
  MaterialExpressionMultiply_78 {
   class=MaterialExpressionMultiply
   inputs [
    A {from=MaterialExpressionTextureCoordinate_2}
    B {from=""}
   ]
   properties {constB=5}
  }
  MaterialExpressionMultiply_79 {
   class=MaterialExpressionMultiply
   inputs [
    A {from=MaterialExpressionNamedRerouteUsage_63}
    B {from=MaterialExpressionNamedRerouteUsage_61}
   ]
  }
  MaterialExpressionMultiply_8 {
   class=MaterialExpressionMultiply
   inputs [
    A {from=MaterialExpressionSphereMask_0}
    B {from=MaterialExpressionScalarParameter_20}
   ]
  }
  MaterialExpressionMultiply_80 {
   class=MaterialExpressionMultiply
   inputs [
    A {from=MaterialExpressionTextureSample_15, output=R}
    B {from=MaterialExpressionMax_2}
   ]
  }
  MaterialExpressionMultiply_81 {
   class=MaterialExpressionMultiply
   inputs [
    A {from=MaterialExpressionTime_9}
    B {from=""}
   ]
   properties {constB=0.31}
  }
  MaterialExpressionMultiply_82 {
   class=MaterialExpressionMultiply
   inputs [
    A {from=MaterialExpressionMultiply_80}
    B {from=""}
   ]
   properties {constB=5}
  }
  MaterialExpressionMultiply_83 {
   class=MaterialExpressionMultiply
   inputs [
    A {from=MaterialExpressionMultiply_79}
    B {from=""}
   ]
   properties {constB=0.75}
  }
  MaterialExpressionMultiply_84 {
   class=MaterialExpressionMultiply
   inputs [
    A {from=MaterialExpressionNamedRerouteUsage_65}
    B {from=""}
   ]
   properties {constB=0.25}
  }
  MaterialExpressionMultiply_85 {
   class=MaterialExpressionMultiply
   inputs [
    A {from=MaterialExpressionSine_2}
    B {from=""}
   ]
   properties {constB=3}
  }
  MaterialExpressionMultiply_86 {
   class=MaterialExpressionMultiply
   inputs [
    A {from=MaterialExpressionNamedRerouteUsage_66}
    B {from=MaterialExpressionScalarParameter_0}
   ]
   properties {constB=0.6}
  }
  MaterialExpressionMultiply_87 {
   class=MaterialExpressionMultiply
   inputs [
    A {from=MaterialExpressionMultiply_37}
    B {from=""}
   ]
  }
  MaterialExpressionMultiply_88 {
   class=MaterialExpressionMultiply
   inputs [
    A {from=MaterialExpressionSaturate_22}
    B {from=MaterialExpressionScalarParameter_71}
   ]
  }
  MaterialExpressionMultiply_9 {
   class=MaterialExpressionMultiply
   inputs [
    A {from=MaterialExpressionOneMinus_1}
    B {from=MaterialExpressionOneMinus_2}
   ]
  }
  MaterialExpressionNamedRerouteDeclaration_0 {
   class=MaterialExpressionNamedRerouteDeclaration
   inputs [None {from=MaterialExpressionFunctionInput_2}]
   properties {name=PreskinnedLocalPosition, nodeColor=<rgba 1 0.8672 0 1>}
  }
  MaterialExpressionNamedRerouteDeclaration_1 {
   class=MaterialExpressionNamedRerouteDeclaration
   inputs [None {from=MaterialExpressionFunctionInput_0}]
   properties {name=OriginalMA, nodeColor=<rgba 0.5469 0 1 1>}
  }
  MaterialExpressionNamedRerouteDeclaration_10 {
   class=MaterialExpressionNamedRerouteDeclaration
   inputs [None {from=MaterialExpressionFunctionInput_3}]
   properties {name=InTechArtMask, nodeColor=<rgba 0.1484 0 1 1>}
  }
  MaterialExpressionNamedRerouteDeclaration_11 {
   class=MaterialExpressionNamedRerouteDeclaration
   inputs [None {from=MaterialExpressionGetMaterialAttributes_2, output=Roughness}]
   properties {name=OriginalRoughness, nodeColor=<rgba 0 1 1 1>}
  }
  MaterialExpressionNamedRerouteDeclaration_12 {
   class=MaterialExpressionNamedRerouteDeclaration
   inputs [None {from=MaterialExpressionMultiply_30}]
   properties {name=Random Normal Dir, nodeColor=<rgba 0 0.3501 1 1>}
  }
  MaterialExpressionNamedRerouteDeclaration_13 {
   class=MaterialExpressionNamedRerouteDeclaration
   inputs [None {from=MaterialExpressionTransform_3}]
   properties {name=Glitter Normal, nodeColor=<rgba 0.7812 1 0 1>}
  }
  MaterialExpressionNamedRerouteDeclaration_14 {
   class=MaterialExpressionNamedRerouteDeclaration
   inputs [None {from=MaterialExpressionMultiply_21}]
   properties {name=LDotN, nodeColor=<rgba 0 0.2031 1 1>}
  }
  MaterialExpressionNamedRerouteDeclaration_15 {
   class=MaterialExpressionNamedRerouteDeclaration
   inputs [None {from=MaterialExpressionSaturate_4}]
   properties {name=Random Glitter, nodeColor=<rgba 0 1 0.4141 1>}
  }
  MaterialExpressionNamedRerouteDeclaration_16 {
   class=MaterialExpressionNamedRerouteDeclaration
   inputs [None {from=MaterialExpressionSetMaterialAttributes_8}]
   properties {name=Glitter FX, nodeColor=<rgba 1 0.0938 0 1>}
  }
  MaterialExpressionNamedRerouteDeclaration_17 {
   class=MaterialExpressionNamedRerouteDeclaration
   inputs [None {from=MaterialExpressionSmoothStep_2}]
   properties {name=Reflection Glitter - Mask, nodeColor=<rgba 0 0.2266 1 1>}
  }
  MaterialExpressionNamedRerouteDeclaration_18 {
   class=MaterialExpressionNamedRerouteDeclaration
   inputs [None {from=MaterialExpressionMultiply_33}]
   properties {name=Reflection Glitter, nodeColor=<rgba 0.5469 1 0 1>}
  }
  MaterialExpressionNamedRerouteDeclaration_19 {
   class=MaterialExpressionNamedRerouteDeclaration
   inputs [None {from=MaterialExpressionSaturate_10}]
   properties {name=Glitter - Top Down Fade, nodeColor=<rgba 1 0 0.8672 1>}
  }
  MaterialExpressionNamedRerouteDeclaration_2 {
   class=MaterialExpressionNamedRerouteDeclaration
   inputs [None {from=MaterialExpressionFunctionInput_1}]
   properties {name=InMasks, nodeColor=<rgba 0 1 0.2266 1>}
  }
  MaterialExpressionNamedRerouteDeclaration_20 {
   class=MaterialExpressionNamedRerouteDeclaration
   inputs [None {from=MaterialExpressionSetMaterialAttributes_9}]
   properties {name=Reactive Sparkle, nodeColor=<rgba 0 1 0.8125 1>}
  }
  MaterialExpressionNamedRerouteDeclaration_21 {
   class=MaterialExpressionNamedRerouteDeclaration
   inputs [None {from=MaterialExpressionAdd_36}]
   properties {name=PrimaryEmissive, nodeColor=<rgba 0.9219 1 0 1>}
  }
  MaterialExpressionNamedRerouteDeclaration_22 {
   class=MaterialExpressionNamedRerouteDeclaration
   inputs [None {from=MaterialExpressionNamedRerouteDeclaration_24}]
   properties {name=Debug, nodeColor=<rgba 0 1 1 1>}
  }
  MaterialExpressionNamedRerouteDeclaration_23 {
   class=MaterialExpressionNamedRerouteDeclaration
   inputs [None {from=MaterialExpressionMultiply_53}]
   properties {name=NewRibbons, nodeColor=<rgba 0 1 0.6719 1>}
  }
  MaterialExpressionNamedRerouteDeclaration_24 {
   class=MaterialExpressionNamedRerouteDeclaration
   inputs [None {from=MaterialExpressionMultiply_61}]
   properties {name=Glitter Animated Emissive, nodeColor=<rgba 1 0.2874 0 1>}
  }
  MaterialExpressionNamedRerouteDeclaration_25 {
   class=MaterialExpressionNamedRerouteDeclaration
   inputs [None {from=MaterialExpressionSaturate_17}]
   properties {name=BodyPanningGradient, nodeColor=<rgba 0.5234 0 1 1>}
  }
  MaterialExpressionNamedRerouteDeclaration_26 {
   class=MaterialExpressionNamedRerouteDeclaration
   inputs [None {from=MaterialExpressionSaturate_12}]
   properties {name=RibbonMask, nodeColor=<rgba 0.8281 0 1 1>}
  }
  MaterialExpressionNamedRerouteDeclaration_27 {
   class=MaterialExpressionNamedRerouteDeclaration
   inputs [None {from=MaterialExpressionSkyLightEnvMapSample_0}]
   properties {name=Sky Light Color, nodeColor=<rgba 0 1 0.9062 1>}
  }
  MaterialExpressionNamedRerouteDeclaration_28 {
   class=MaterialExpressionNamedRerouteDeclaration
   inputs [None {from=MaterialExpressionLinearInterpolate_3}]
   properties {name=PrimaryBasecolor, nodeColor=<rgba 0 1 0.5078 1>}
  }
  MaterialExpressionNamedRerouteDeclaration_29 {
   class=MaterialExpressionNamedRerouteDeclaration
   inputs [None {from=MaterialExpressionLinearInterpolate_9}]
   properties {name=PrimaryMetallic, nodeColor=<rgba 0.5 0 1 1>}
  }
  MaterialExpressionNamedRerouteDeclaration_3 {
   class=MaterialExpressionNamedRerouteDeclaration
   inputs [None {from=MaterialExpressionTextureSampleParameter2D_1, output=RGB}]
   properties {name=SuperlevelMasks01, nodeColor=<rgba 1 0.1931 0 1>}
  }
  MaterialExpressionNamedRerouteDeclaration_30 {
   class=MaterialExpressionNamedRerouteDeclaration
   inputs [None {from=MaterialExpressionMultiply_77}]
   properties {name=halftones, nodeColor=<rgba 1 0 0.4687 1>}
  }
  MaterialExpressionNamedRerouteDeclaration_31 {
   class=MaterialExpressionNamedRerouteDeclaration
   inputs [None {from=MaterialExpressionLinearInterpolate_21}]
   properties {name=PrimaryGlitterMask, nodeColor=<rgba 0.4062 0 1 1>}
  }
  MaterialExpressionNamedRerouteDeclaration_4 {
   class=MaterialExpressionNamedRerouteDeclaration
   inputs [None {from=MaterialExpressionMultiply_1}]
   properties {name=MarineCarve_Oliie_Mask, nodeColor=<rgba 0.4766 0 1 1>}
  }
  MaterialExpressionNamedRerouteDeclaration_5 {
   class=MaterialExpressionNamedRerouteDeclaration
   inputs [None {from=MaterialExpressionAdd_19}]
   properties {name=PrimaryElementGradientMask, nodeColor=<rgba 0 0.0625 1 1>}
  }
  MaterialExpressionNamedRerouteDeclaration_6 {
   class=MaterialExpressionNamedRerouteDeclaration
   inputs [None {from=MaterialExpressionFunctionInput_4}]
   properties {name=InFresnel, nodeColor=<rgba 0 1 0.8594 1>}
  }
  MaterialExpressionNamedRerouteDeclaration_7 {
   class=MaterialExpressionNamedRerouteDeclaration
   inputs [None {from=MaterialExpressionBlendMaterialAttributes_6}]
   properties {name=OutMA, nodeColor=<rgba 1 0.5391 0 1>}
  }
  MaterialExpressionNamedRerouteDeclaration_8 {
   class=MaterialExpressionNamedRerouteDeclaration
   inputs [None {from=MaterialExpressionTextureSampleParameter2D_3, output=RGB}]
   properties {name=SuperlevelMasks02, nodeColor=<rgba 0.2188 0 1 1>}
  }
  MaterialExpressionNamedRerouteDeclaration_9 {
   class=MaterialExpressionNamedRerouteDeclaration
   inputs [None {from=MaterialExpressionVertexInterpolator_2, output=PS}]
   properties {name=PreskinnedLocalNormal, nodeColor=<rgba 0 0.4375 1 1>}
  }
  MaterialExpressionNamedRerouteUsage_0 {class=MaterialExpressionNamedRerouteUsage}
  MaterialExpressionNamedRerouteUsage_1 {class=MaterialExpressionNamedRerouteUsage}
  MaterialExpressionNamedRerouteUsage_12 {class=MaterialExpressionNamedRerouteUsage}
  MaterialExpressionNamedRerouteUsage_13 {class=MaterialExpressionNamedRerouteUsage}
  MaterialExpressionNamedRerouteUsage_14 {class=MaterialExpressionNamedRerouteUsage}
  MaterialExpressionNamedRerouteUsage_15 {class=MaterialExpressionNamedRerouteUsage}
  MaterialExpressionNamedRerouteUsage_16 {class=MaterialExpressionNamedRerouteUsage}
  MaterialExpressionNamedRerouteUsage_19 {class=MaterialExpressionNamedRerouteUsage}
  MaterialExpressionNamedRerouteUsage_2 {class=MaterialExpressionNamedRerouteUsage}
  MaterialExpressionNamedRerouteUsage_20 {class=MaterialExpressionNamedRerouteUsage}
  MaterialExpressionNamedRerouteUsage_21 {class=MaterialExpressionNamedRerouteUsage}
  MaterialExpressionNamedRerouteUsage_22 {class=MaterialExpressionNamedRerouteUsage}
  MaterialExpressionNamedRerouteUsage_23 {class=MaterialExpressionNamedRerouteUsage}
  MaterialExpressionNamedRerouteUsage_24 {class=MaterialExpressionNamedRerouteUsage}
  MaterialExpressionNamedRerouteUsage_25 {class=MaterialExpressionNamedRerouteUsage}
  MaterialExpressionNamedRerouteUsage_26 {class=MaterialExpressionNamedRerouteUsage}
  MaterialExpressionNamedRerouteUsage_27 {class=MaterialExpressionNamedRerouteUsage}
  MaterialExpressionNamedRerouteUsage_29 {class=MaterialExpressionNamedRerouteUsage}
  MaterialExpressionNamedRerouteUsage_3 {class=MaterialExpressionNamedRerouteUsage}
  MaterialExpressionNamedRerouteUsage_30 {class=MaterialExpressionNamedRerouteUsage}
  MaterialExpressionNamedRerouteUsage_31 {class=MaterialExpressionNamedRerouteUsage}
  MaterialExpressionNamedRerouteUsage_32 {class=MaterialExpressionNamedRerouteUsage}
  MaterialExpressionNamedRerouteUsage_33 {class=MaterialExpressionNamedRerouteUsage}
  MaterialExpressionNamedRerouteUsage_34 {class=MaterialExpressionNamedRerouteUsage}
  MaterialExpressionNamedRerouteUsage_35 {class=MaterialExpressionNamedRerouteUsage}
  MaterialExpressionNamedRerouteUsage_38 {class=MaterialExpressionNamedRerouteUsage}
  MaterialExpressionNamedRerouteUsage_39 {class=MaterialExpressionNamedRerouteUsage}
  MaterialExpressionNamedRerouteUsage_4 {class=MaterialExpressionNamedRerouteUsage}
  MaterialExpressionNamedRerouteUsage_40 {class=MaterialExpressionNamedRerouteUsage}
  MaterialExpressionNamedRerouteUsage_41 {class=MaterialExpressionNamedRerouteUsage}
  MaterialExpressionNamedRerouteUsage_42 {class=MaterialExpressionNamedRerouteUsage}
  MaterialExpressionNamedRerouteUsage_43 {class=MaterialExpressionNamedRerouteUsage}
  MaterialExpressionNamedRerouteUsage_45 {class=MaterialExpressionNamedRerouteUsage}
  MaterialExpressionNamedRerouteUsage_46 {class=MaterialExpressionNamedRerouteUsage}
  MaterialExpressionNamedRerouteUsage_47 {class=MaterialExpressionNamedRerouteUsage}
  MaterialExpressionNamedRerouteUsage_48 {class=MaterialExpressionNamedRerouteUsage}
  MaterialExpressionNamedRerouteUsage_49 {class=MaterialExpressionNamedRerouteUsage}
  MaterialExpressionNamedRerouteUsage_5 {class=MaterialExpressionNamedRerouteUsage}
  MaterialExpressionNamedRerouteUsage_50 {class=MaterialExpressionNamedRerouteUsage}
  MaterialExpressionNamedRerouteUsage_51 {class=MaterialExpressionNamedRerouteUsage}
  MaterialExpressionNamedRerouteUsage_52 {class=MaterialExpressionNamedRerouteUsage}
  MaterialExpressionNamedRerouteUsage_53 {class=MaterialExpressionNamedRerouteUsage}
  MaterialExpressionNamedRerouteUsage_54 {class=MaterialExpressionNamedRerouteUsage}
  MaterialExpressionNamedRerouteUsage_55 {class=MaterialExpressionNamedRerouteUsage}
  MaterialExpressionNamedRerouteUsage_56 {class=MaterialExpressionNamedRerouteUsage}
  MaterialExpressionNamedRerouteUsage_57 {class=MaterialExpressionNamedRerouteUsage}
  MaterialExpressionNamedRerouteUsage_58 {class=MaterialExpressionNamedRerouteUsage}
  MaterialExpressionNamedRerouteUsage_59 {class=MaterialExpressionNamedRerouteUsage}
  MaterialExpressionNamedRerouteUsage_6 {class=MaterialExpressionNamedRerouteUsage}
  MaterialExpressionNamedRerouteUsage_60 {class=MaterialExpressionNamedRerouteUsage}
  MaterialExpressionNamedRerouteUsage_61 {class=MaterialExpressionNamedRerouteUsage}
  MaterialExpressionNamedRerouteUsage_62 {class=MaterialExpressionNamedRerouteUsage}
  MaterialExpressionNamedRerouteUsage_63 {class=MaterialExpressionNamedRerouteUsage}
  MaterialExpressionNamedRerouteUsage_64 {class=MaterialExpressionNamedRerouteUsage}
  MaterialExpressionNamedRerouteUsage_65 {class=MaterialExpressionNamedRerouteUsage}
  MaterialExpressionNamedRerouteUsage_66 {class=MaterialExpressionNamedRerouteUsage}
  MaterialExpressionNamedRerouteUsage_8 {class=MaterialExpressionNamedRerouteUsage}
  MaterialExpressionNamedRerouteUsage_9 {class=MaterialExpressionNamedRerouteUsage}
  MaterialExpressionNoise_0 {
   class=MaterialExpressionNoise
   inputs [
    World Position {from=MaterialExpressionAdd_41, output=""}
    FilterWidth {from="", output=""}
   ]
   properties {levels=1}
  }
  MaterialExpressionNoise_1 {
   class=MaterialExpressionNoise
   inputs [
    World Position {from=MaterialExpressionAdd_37, output=""}
    FilterWidth {from="", output=""}
   ]
   properties {bTurbulence=false, levels=1, noiseFunction=NOISEFUNCTION_GradientALU, outputMin=0, scale=0.04}
  }
  MaterialExpressionNoise_2 {
   class=MaterialExpressionNoise
   inputs [
    World Position {from=MaterialExpressionAdd_0, output=""}
    FilterWidth {from="", output=""}
   ]
   properties {levels=2, outputMin=0, scale=3}
  }
  MaterialExpressionNormalize_0 {class=MaterialExpressionNormalize, inputs [VectorInput {from=MaterialExpressionSubtract_2, output=""}]}
  MaterialExpressionNormalize_1 {class=MaterialExpressionNormalize, inputs [VectorInput {from=MaterialExpressionReflectionVectorWS_1, output=""}]}
  MaterialExpressionNormalize_2 {class=MaterialExpressionNormalize, inputs [VectorInput {from=MaterialExpressionAdd_12, output=""}]}
  MaterialExpressionNormalize_4 {class=MaterialExpressionNormalize, inputs [VectorInput {from=MaterialExpressionAdd_13, output=""}], properties {desc=GlitterDirection}}
  MaterialExpressionOneMinus_0 {class=MaterialExpressionOneMinus, inputs [None {from=MaterialExpressionLinearInterpolate_10}]}
  MaterialExpressionOneMinus_1 {class=MaterialExpressionOneMinus, inputs [None {from=MaterialExpressionLinearInterpolate_22}]}
  MaterialExpressionOneMinus_11 {class=MaterialExpressionOneMinus, inputs [None {from=MaterialExpressionMultiply_57}]}
  MaterialExpressionOneMinus_12 {class=MaterialExpressionOneMinus, inputs [None {from=MaterialExpressionMultiply_74}]}
  MaterialExpressionOneMinus_13 {class=MaterialExpressionOneMinus, inputs [None {from=MaterialExpressionNamedRerouteUsage_60}]}
  MaterialExpressionOneMinus_14 {class=MaterialExpressionOneMinus, inputs [None {from=MaterialExpressionSaturate_21}]}
  MaterialExpressionOneMinus_2 {class=MaterialExpressionOneMinus, inputs [None {from=MaterialExpressionSaturate_5}]}
  MaterialExpressionOneMinus_3 {class=MaterialExpressionOneMinus, inputs [None {from=MaterialExpressionComponentMask_10}]}
  MaterialExpressionOneMinus_4 {class=MaterialExpressionOneMinus, inputs [None {from=MaterialExpressionSaturate_3}]}
  MaterialExpressionOneMinus_5 {class=MaterialExpressionOneMinus, inputs [None {from=MaterialExpressionLocalPosition_0, output=Z}]}
  MaterialExpressionOneMinus_6 {class=MaterialExpressionOneMinus, inputs [None {from=MaterialExpressionSaturate_8}]}
  MaterialExpressionOneMinus_7 {class=MaterialExpressionOneMinus, inputs [None {from=MaterialExpressionFresnel_1}]}
  MaterialExpressionOneMinus_8 {class=MaterialExpressionOneMinus, inputs [None {from=MaterialExpressionNamedRerouteUsage_43}]}
  MaterialExpressionPanner_0 {
   class=MaterialExpressionPanner
   inputs [
    Coordinate {from=MaterialExpressionDivide_2}
    Time {from=""}
    Speed {from=MaterialExpressionMultiply_40}
   ]
   properties {speedX=0.094, speedY=-0.123}
  }
  MaterialExpressionPanner_4 {
   class=MaterialExpressionPanner
   inputs [
    Coordinate {from=""}
    Time {from=""}
    Speed {from=""}
   ]
   properties {speedY=0.156}
  }
  MaterialExpressionPower_0 {
   class=MaterialExpressionPower
   inputs [
    Base {from=MaterialExpressionComponentMask_19}
    Exp {from=MaterialExpressionScalarParameter_45}
   ]
   properties {constExponent=32}
  }
  MaterialExpressionPower_1 {
   class=MaterialExpressionPower
   inputs [
    Base {from=MaterialExpressionOneMinus_7}
    Exp {from=""}
   ]
   properties {constExponent=3}
  }
  MaterialExpressionPower_10 {
   class=MaterialExpressionPower
   inputs [
    Base {from=MaterialExpressionOneMinus_13}
    Exp {from=""}
   ]
   properties {constExponent=12}
  }
  MaterialExpressionPower_11 {
   class=MaterialExpressionPower
   inputs [
    Base {from=MaterialExpressionAdd_4}
    Exp {from=""}
   ]
   properties {constExponent=0.1}
  }
  MaterialExpressionPower_12 {
   class=MaterialExpressionPower
   inputs [
    Base {from=MaterialExpressionAdd_2}
    Exp {from=""}
   ]
   properties {constExponent=0.1}
  }
  MaterialExpressionPower_2 {
   class=MaterialExpressionPower
   inputs [
    Base {from=MaterialExpressionOneMinus_8}
    Exp {from=""}
   ]
  }
  MaterialExpressionPower_3 {
   class=MaterialExpressionPower
   inputs [
    Base {from=MaterialExpressionMultiply_15}
    Exp {from=""}
   ]
   properties {constExponent=4}
  }
  MaterialExpressionPower_4 {
   class=MaterialExpressionPower
   inputs [
    Base {from=MaterialExpressionDesaturation_2}
    Exp {from=MaterialExpressionScalarParameter_17}
   ]
  }
  MaterialExpressionPower_5 {
   class=MaterialExpressionPower
   inputs [
    Base {from=MaterialExpressionDotProduct_1}
    Exp {from=MaterialExpressionScalarParameter_27}
   ]
  }
  MaterialExpressionPower_6 {
   class=MaterialExpressionPower
   inputs [
    Base {from=MaterialExpressionFrac_1}
    Exp {from=""}
   ]
  }
  MaterialExpressionPower_7 {
   class=MaterialExpressionPower
   inputs [
    Base {from=MaterialExpressionTextureSample_11, output=G}
    Exp {from=MaterialExpressionLinearInterpolate_18}
   ]
   properties {constExponent=8}
  }
  MaterialExpressionPower_8 {
   class=MaterialExpressionPower
   inputs [
    Base {from=MaterialExpressionAdd_42}
    Exp {from=""}
   ]
   properties {constExponent=3}
  }
  MaterialExpressionPower_9 {
   class=MaterialExpressionPower
   inputs [
    Base {from=MaterialExpressionTextureSample_11, output=G}
    Exp {from=MaterialExpressionLinearInterpolate_20}
   ]
   properties {constExponent=8}
  }
  MaterialExpressionPreSkinnedLocalBounds_2 {class=MaterialExpressionPreSkinnedLocalBounds, inputs [], properties {}}
  MaterialExpressionPreSkinnedNormal_1 {class=MaterialExpressionPreSkinnedNormal, inputs [], properties {}}
  MaterialExpressionQualitySwitch_0 {
   class=MaterialExpressionQualitySwitch
   inputs [
    Default {from=MaterialExpressionAdd_25, output=""}
    Low {from=MaterialExpressionReroute_1, output=""}
    High {from="", output=""}
    Medium {from="", output=""}
    Epic {from="", output=""}
   ]
   properties {}
  }
  MaterialExpressionReflectionVectorWS_1 {class=MaterialExpressionReflectionVectorWS, inputs [CustomWorldNormal {from="", output=""}], properties {}}
  MaterialExpressionReroute_1 {class=MaterialExpressionReroute, inputs [None {from=MaterialExpressionAdd_46}]}
  MaterialExpressionReroute_4 {class=MaterialExpressionReroute, inputs [None {from=MaterialExpressionAdd_34}]}
  MaterialExpressionReroute_5 {class=MaterialExpressionReroute, inputs [None {from=MaterialExpressionAdd_32}]}
  MaterialExpressionReroute_6 {class=MaterialExpressionReroute, inputs [None {from=MaterialExpressionMaterialFunctionCall_11, output=Result}]}
  MaterialExpressionRotator_0 {
   class=MaterialExpressionRotator
   inputs [
    Coordinate {from=MaterialExpressionComponentMask_26, output=""}
    Time {from=MaterialExpressionConstant_16, output=""}
   ]
   properties {}
  }
  MaterialExpressionSaturate_0 {class=MaterialExpressionSaturate, inputs [None {from=MaterialExpressionPower_12}]}
  MaterialExpressionSaturate_1 {class=MaterialExpressionSaturate, inputs [None {from=MaterialExpressionConstant_0}]}
  MaterialExpressionSaturate_10 {class=MaterialExpressionSaturate, inputs [None {from=MaterialExpressionAdd_17}]}
  MaterialExpressionSaturate_11 {class=MaterialExpressionSaturate, inputs [None {from=MaterialExpressionPower_11}]}
  MaterialExpressionSaturate_12 {class=MaterialExpressionSaturate, inputs [None {from=MaterialExpressionMultiply_41}]}
  MaterialExpressionSaturate_13 {class=MaterialExpressionSaturate, inputs [None {from=MaterialExpressionMultiply_22}]}
  MaterialExpressionSaturate_14 {class=MaterialExpressionSaturate, inputs [None {from=MaterialExpressionAdd_23}]}
  MaterialExpressionSaturate_15 {class=MaterialExpressionSaturate, inputs [None {from=MaterialExpressionAdd_1}]}
  MaterialExpressionSaturate_16 {class=MaterialExpressionSaturate, inputs [None {from=MaterialExpressionDotProduct_4}]}
  MaterialExpressionSaturate_17 {class=MaterialExpressionSaturate, inputs [None {from=MaterialExpressionAdd_43}]}
  MaterialExpressionSaturate_18 {class=MaterialExpressionSaturate, inputs [None {from=MaterialExpressionNamedRerouteUsage_51}]}
  MaterialExpressionSaturate_19 {class=MaterialExpressionSaturate, inputs [None {from=MaterialExpressionOneMinus_11}]}
  MaterialExpressionSaturate_2 {class=MaterialExpressionSaturate, inputs [None {from=MaterialExpressionDotProduct_0}]}
  MaterialExpressionSaturate_20 {class=MaterialExpressionSaturate, inputs [None {from=MaterialExpressionOneMinus_12}]}
  MaterialExpressionSaturate_21 {class=MaterialExpressionSaturate, inputs [None {from=MaterialExpressionSmoothStep_4}]}
  MaterialExpressionSaturate_22 {class=MaterialExpressionSaturate, inputs [None {from=MaterialExpressionAdd_38}]}
  MaterialExpressionSaturate_23 {class=MaterialExpressionSaturate, inputs [None {from=MaterialExpressionDivide_5}]}
  MaterialExpressionSaturate_3 {class=MaterialExpressionSaturate, inputs [None {from=MaterialExpressionAdd_8}]}
  MaterialExpressionSaturate_4 {class=MaterialExpressionSaturate, inputs [None {from=MaterialExpressionPower_5}]}
  MaterialExpressionSaturate_5 {class=MaterialExpressionSaturate, inputs [None {from=MaterialExpressionComponentMask_3}]}
  MaterialExpressionSaturate_6 {class=MaterialExpressionSaturate, inputs [None {from=MaterialExpressionAdd_10}]}
  MaterialExpressionSaturate_7 {class=MaterialExpressionSaturate, inputs [None {from=MaterialExpressionAdd_11}]}
  MaterialExpressionSaturate_8 {class=MaterialExpressionSaturate, inputs [None {from=MaterialExpressionAdd_6}]}
  MaterialExpressionSaturate_9 {class=MaterialExpressionSaturate, inputs [None {from=MaterialExpressionDotProduct_2}]}
  MaterialExpressionScalarParameter_0 {class=MaterialExpressionScalarParameter, properties {defaultValue=0.6, parameterName=Secondary_ReflectionGlitterOpacity}}
  MaterialExpressionScalarParameter_10 {class=MaterialExpressionScalarParameter, properties {group=SUPERLEVEL, parameterName=PrimaryElement_RoughnessOffset, sortPriority=10}}
  MaterialExpressionScalarParameter_11 {class=MaterialExpressionScalarParameter, properties {group=SUPERLEVEL, parameterName=SecondaryElement_RoughnessOffset, sortPriority=30}}
  MaterialExpressionScalarParameter_12 {
   class=MaterialExpressionScalarParameter
   properties {defaultValue=0.8171, group=CanineCronutDig_Ware, parameterName=Random Glitter - Roughness}
  }
  MaterialExpressionScalarParameter_13 {class=MaterialExpressionScalarParameter, properties {defaultValue=149.8988, group=CanineCronutDig_Ware, parameterName=Sparkle - Fade}}
  MaterialExpressionScalarParameter_14 {
   class=MaterialExpressionScalarParameter
   properties {defaultValue=0.1, group=CanineCronutDig_Ware, parameterName=Global Fake Reflection - Intensity}
  }
  MaterialExpressionScalarParameter_15 {
   class=MaterialExpressionScalarParameter
   properties {defaultValue=15, group=CanineCronutDig_Ware, parameterName=Glitter - Ambient Brightness}
  }
  MaterialExpressionScalarParameter_16 {class=MaterialExpressionScalarParameter, properties {defaultValue=0.001, parameterName=adsaadsafd}}
  MaterialExpressionScalarParameter_17 {class=MaterialExpressionScalarParameter, properties {defaultValue=4.5, group=CanineCronutDig_Ware, parameterName=Reactive Sparkle - Exp}}
  MaterialExpressionScalarParameter_18 {
   class=MaterialExpressionScalarParameter
   properties {defaultValue=0.02, group=SUPERLEVEL, parameterName=PrimaryElement_AdditionalHardness, sortPriority=30}
  }
  MaterialExpressionScalarParameter_19 {class=MaterialExpressionScalarParameter, properties {defaultValue=0.1, group=CanineCronutDig_Ware, parameterName=Reactive Sparkle - Min}}
  MaterialExpressionScalarParameter_2 {class=MaterialExpressionScalarParameter, properties {defaultValue=0.7, group=SUPERLEVEL, parameterName=Tertiary_BaseColorContrast}}
  MaterialExpressionScalarParameter_20 {class=MaterialExpressionScalarParameter, properties {parameterName=PrimaryElement_UseAdditionalMask}}
  MaterialExpressionScalarParameter_21 {class=MaterialExpressionScalarParameter, properties {parameterName=IsMarineCarve_Ollie}}
  MaterialExpressionScalarParameter_22 {
   class=MaterialExpressionScalarParameter
   properties {defaultValue=0.5, group=SUPERLEVEL, parameterName=PrimaryElement_SpecularOverride, sortPriority=10}
  }
  MaterialExpressionScalarParameter_23 {
   class=MaterialExpressionScalarParameter
   properties {defaultValue=0.8, group=SUPERLEVEL, parameterName=PrimaryElement_RoughnessOverride, sortPriority=10}
  }
  MaterialExpressionScalarParameter_24 {class=MaterialExpressionScalarParameter, properties {defaultValue=40, parameterName=Primary_RibbensScale}}
  MaterialExpressionScalarParameter_25 {class=MaterialExpressionScalarParameter, properties {defaultValue=0.1, parameterName=Primary_RibbensSpeed}}
  MaterialExpressionScalarParameter_26 {class=MaterialExpressionScalarParameter, properties {defaultValue=6, parameterName=Primary_RibbensNoiseDistortion}}
  MaterialExpressionScalarParameter_27 {class=MaterialExpressionScalarParameter, properties {defaultValue=1, group=CanineCronutDig_Ware, parameterName=Random Glitter - Exp}}
  MaterialExpressionScalarParameter_28 {class=MaterialExpressionScalarParameter, properties {defaultValue=3, parameterName=Primary_RibbenOpacity}}
  MaterialExpressionScalarParameter_29 {class=MaterialExpressionScalarParameter, properties {defaultValue=-1, parameterName=Primary_ProjectionContrast}}
  MaterialExpressionScalarParameter_3 {class=MaterialExpressionScalarParameter, properties {defaultValue=1.5, group=SUPERLEVEL, parameterName=Tertiary_ColorBrightness}}
  MaterialExpressionScalarParameter_30 {class=MaterialExpressionScalarParameter, properties {defaultValue=2.5, group=SUPERLEVEL, parameterName=Primary_ParallaxDepth}}
  MaterialExpressionScalarParameter_31 {class=MaterialExpressionScalarParameter, properties {defaultValue=2, group=Dance Lines, parameterName=Repetition}}
  MaterialExpressionScalarParameter_32 {class=MaterialExpressionScalarParameter, properties {group=SUPERLEVEL, parameterName=SecondaryElement_GradientStart, sortPriority=30}}
  MaterialExpressionScalarParameter_33 {
   class=MaterialExpressionScalarParameter
   properties {defaultValue=140, group=SUPERLEVEL, parameterName=SecondaryElement_GradientEnd, sortPriority=30}
  }
  MaterialExpressionScalarParameter_34 {
   class=MaterialExpressionScalarParameter
   properties {defaultValue=0.9, group=CanineCronutDig_Ware, parameterName=Random Glitter - Strength, sliderMax=2, sortPriority=6}
  }
  MaterialExpressionScalarParameter_35 {class=MaterialExpressionScalarParameter, properties {defaultValue=20, group=CanineCronutDig_Ware, parameterName=Random Glitter - Tiling}}
  MaterialExpressionScalarParameter_36 {
   class=MaterialExpressionScalarParameter
   properties {defaultValue=12, group=GlitterAnimated, parameterName=Sparkle Detail Tiling, sliderMax=64, sliderMin=8, sortPriority=2}
  }
  MaterialExpressionScalarParameter_37 {
   class=MaterialExpressionScalarParameter
   properties {defaultValue=0.9, group=CanineCronutDig_Ware, parameterName=Glitter - Transition Amount}
  }
  MaterialExpressionScalarParameter_38 {class=MaterialExpressionScalarParameter, properties {defaultValue=-3, parameterName=Glitter Depth}}
  MaterialExpressionScalarParameter_39 {
   class=MaterialExpressionScalarParameter
   properties {defaultValue=1, group=CanineCronutDig_Ware, parameterName=Reflection Glitter - Desaturation}
  }
  MaterialExpressionScalarParameter_4 {
   class=MaterialExpressionScalarParameter
   properties {defaultValue=0.02, group=SUPERLEVEL, parameterName=PrimaryElement_Hardness, sortPriority=30}
  }
  MaterialExpressionScalarParameter_40 {class=MaterialExpressionScalarParameter, properties {defaultValue=0.15, group=CanineCronutDig_Ware, parameterName=Glitter - Transition Fade}}
  MaterialExpressionScalarParameter_41 {class=MaterialExpressionScalarParameter, properties {defaultValue=1, group=SL37 - T1 Butterfly, parameterName=Glitter Tiling Close}}
  MaterialExpressionScalarParameter_44 {
   class=MaterialExpressionScalarParameter
   properties {defaultValue=0.2, group=GlitterAnimated, parameterName=Time Scale, sliderMax=2, sortPriority=3}
  }
  MaterialExpressionScalarParameter_45 {
   class=MaterialExpressionScalarParameter
   properties {defaultValue=0.2, group=GlitterAnimated, parameterName=Fade Curve Exponent, sortPriority=4}
  }
  MaterialExpressionScalarParameter_46 {
   class=MaterialExpressionScalarParameter
   properties {defaultValue=10, group=GlitterAnimated, parameterName=Emissive Intensity, sliderMax=20, sortPriority=5}
  }
  MaterialExpressionScalarParameter_47 {class=MaterialExpressionScalarParameter, properties {defaultValue=0.0125, group=Dance Lines, parameterName=speed}}
  MaterialExpressionScalarParameter_48 {class=MaterialExpressionScalarParameter, properties {defaultValue=1, group=Dance Lines, parameterName=StretchGradient}}
  MaterialExpressionScalarParameter_49 {class=MaterialExpressionScalarParameter, properties {defaultValue=0.5, group=Dance Lines, parameterName=NoiseWarpIntensity}}
  MaterialExpressionScalarParameter_5 {
   class=MaterialExpressionScalarParameter
   properties {defaultValue=-65, group=SUPERLEVEL, parameterName=PrimaryElement_Height, sortPriority=30}
  }
  MaterialExpressionScalarParameter_50 {class=MaterialExpressionScalarParameter, properties {defaultValue=50, desc=50, group=Dance Lines, parameterName=YAxisGradRadius}}
  MaterialExpressionScalarParameter_51 {class=MaterialExpressionScalarParameter, properties {defaultValue=1, group=Dance Lines, parameterName=YAxisGradIntensity}}
  MaterialExpressionScalarParameter_52 {class=MaterialExpressionScalarParameter, properties {defaultValue=150, desc=9.176, group=Dance Lines, parameterName=YAxisGradPosition}}
  MaterialExpressionScalarParameter_53 {class=MaterialExpressionScalarParameter, properties {defaultValue=180, group=Dance Lines, parameterName=ObjectSpaceGradRadius}}
  MaterialExpressionScalarParameter_54 {class=MaterialExpressionScalarParameter, properties {defaultValue=0.2, group=Dance Lines, parameterName=NormalDistortionWeight}}
  MaterialExpressionScalarParameter_55 {class=MaterialExpressionScalarParameter, properties {parameterName=IsRowLiaisonCast_Body}}
  MaterialExpressionScalarParameter_56 {class=MaterialExpressionScalarParameter, properties {parameterName=IsZenCrashSand}}
  MaterialExpressionScalarParameter_57 {class=MaterialExpressionScalarParameter, properties {defaultValue=20, parameterName=Smoke_Intensity}}
  MaterialExpressionScalarParameter_58 {class=MaterialExpressionScalarParameter, properties {defaultValue=3, parameterName=Smoke_Power}}
  MaterialExpressionScalarParameter_59 {class=MaterialExpressionScalarParameter, properties {defaultValue=1, parameterName=Smoke_VerticalStretch}}
  MaterialExpressionScalarParameter_6 {
   class=MaterialExpressionScalarParameter
   properties {defaultValue=0.05, group=SUPERLEVEL, parameterName=SecondaryElement_Hardness, sortPriority=30}
  }
  MaterialExpressionScalarParameter_60 {class=MaterialExpressionScalarParameter, properties {defaultValue=1, parameterName=asda}}
  MaterialExpressionScalarParameter_61 {class=MaterialExpressionScalarParameter, properties {defaultValue=0.11, parameterName=AdditionalSmoke_Scale}}
  MaterialExpressionScalarParameter_62 {class=MaterialExpressionScalarParameter, properties {defaultValue=2.1, group=SUPERLEVEL, parameterName=Primary_RibbonUpwardSpeed}}
  MaterialExpressionScalarParameter_63 {class=MaterialExpressionScalarParameter, properties {defaultValue=2, parameterName=Primary_Repetition}}
  MaterialExpressionScalarParameter_64 {class=MaterialExpressionScalarParameter, properties {defaultValue=0.05, group=SUPERLEVEL, parameterName=Primary_AdditionalRibbonsOpacity}}
  MaterialExpressionScalarParameter_65 {class=MaterialExpressionScalarParameter, properties {defaultValue=1, parameterName=Smoke_Power02}}
  MaterialExpressionScalarParameter_66 {class=MaterialExpressionScalarParameter, properties {group=SUPERLEVEL, parameterName=PrimaryGlitter_GradientStart, sortPriority=30}}
  MaterialExpressionScalarParameter_67 {
   class=MaterialExpressionScalarParameter
   properties {defaultValue=140, group=SUPERLEVEL, parameterName=PrimaryGlitter_GradientEnd, sortPriority=30}
  }
  MaterialExpressionScalarParameter_68 {class=MaterialExpressionScalarParameter, properties {group=SUPERLEVEL, parameterName=PrimaryGlitter_InvertMask, sortPriority=31}}
  MaterialExpressionScalarParameter_69 {class=MaterialExpressionScalarParameter, properties {defaultValue=0.07, parameterName=Ribbons_EmissiveAmount}}
  MaterialExpressionScalarParameter_7 {
   class=MaterialExpressionScalarParameter
   properties {defaultValue=-110, group=SUPERLEVEL, parameterName=SecondaryElement_Height, sortPriority=30}
  }
  MaterialExpressionScalarParameter_70 {class=MaterialExpressionScalarParameter, properties {group=SUPERLEVEL, parameterName=IsFloralMane}}
  MaterialExpressionScalarParameter_71 {class=MaterialExpressionScalarParameter, properties {group=SUPERLEVEL, parameterName=IsMarineCarve}}
  MaterialExpressionScalarParameter_8 {class=MaterialExpressionScalarParameter, properties {group=SUPERLEVEL, parameterName=SecondaryElement_InvertMask, sortPriority=31}}
  MaterialExpressionScalarParameter_9 {class=MaterialExpressionScalarParameter, properties {group=SUPERLEVEL, parameterName=PrimaryElement_InvertMask, sortPriority=31}}
  MaterialExpressionSetMaterialAttributes_0 {
   class=MaterialExpressionSetMaterialAttributes
   inputs [
    MaterialAttributes {from=MaterialExpressionGetMaterialAttributes_3, output=MaterialAttributes}
    EmissiveColor {from=MaterialExpressionMultiply_35, output=""}
    BaseColor {from=MaterialExpressionMultiply_19, output=""}
   ]
   properties {attributeSetTypes [#5, #3]}
  }
  MaterialExpressionSetMaterialAttributes_1 {
   class=MaterialExpressionSetMaterialAttributes
   inputs [
    MaterialAttributes {from=MaterialExpressionGetMaterialAttributes_0, output=MaterialAttributes}
    Roughness {from=MaterialExpressionMin_0, output=""}
    BaseColor {from=MaterialExpressionMultiply_36, output=""}
    Metallic {from=MaterialExpressionConstant_1, output=""}
    Normal {from=MaterialExpressionMaterialFunctionCall_7, output=Result}
   ]
   properties {attributeSetTypes [
     #1
     #3
     #6
     #2
    ]}
  }
  MaterialExpressionSetMaterialAttributes_10 {
   class=MaterialExpressionSetMaterialAttributes
   inputs [
    MaterialAttributes {from=MaterialExpressionBlendMaterialAttributes_6, output=""}
    EmissiveColor {from=MaterialExpressionNamedRerouteUsage_56, output=""}
    SubsurfaceColor {from="", output=""}
    ShadingModel {from="", output=""}
   ]
   properties {attributeSetTypes [#5, #7, #8]}
  }
  MaterialExpressionSetMaterialAttributes_2 {
   class=MaterialExpressionSetMaterialAttributes
   inputs [
    MaterialAttributes {from=MaterialExpressionGetMaterialAttributes_2, output=MaterialAttributes}
    BaseColor {from=MaterialExpressionMultiply_71, output=""}
    Metallic {from=MaterialExpressionNamedRerouteUsage_25, output=""}
    Roughness {from=MaterialExpressionLinearInterpolate_11, output=""}
    EmissiveColor {from=MaterialExpressionNamedRerouteUsage_32, output=""}
    Normal {from="", output=""}
    Specular {from=MaterialExpressionLinearInterpolate_12, output=""}
   ]
   properties {attributeSetTypes [
     #3
     #6
     #1
     #5
     #2
     #4
    ]}
  }
  MaterialExpressionSetMaterialAttributes_3 {
   class=MaterialExpressionSetMaterialAttributes
   inputs [
    MaterialAttributes {from=MaterialExpressionBlendMaterialAttributes_1, output=""}
    BaseColor {from=MaterialExpressionLinearInterpolate_4, output=""}
    Metallic {from=MaterialExpressionConstant_5, output=""}
    Roughness {from=MaterialExpressionSaturate_7, output=""}
    EmissiveColor {from=MaterialExpressionAdd_52, output=""}
    Normal {from="", output=""}
   ]
   properties {attributeSetTypes [
     #3
     #6
     #1
     #5
     #2
    ]}
  }
  MaterialExpressionSetMaterialAttributes_4 {
   class=MaterialExpressionSetMaterialAttributes
   inputs [
    MaterialAttributes {from=MaterialExpressionBlendMaterialAttributes_4, output=""}
    BaseColor {from=MaterialExpressionConstant_12, output=""}
    Metallic {from=MaterialExpressionConstant_2, output=""}
   ]
   properties {attributeSetTypes [#3, #6]}
  }
  MaterialExpressionSetMaterialAttributes_5 {
   class=MaterialExpressionSetMaterialAttributes
   inputs [
    MaterialAttributes {from=MaterialExpressionBlendMaterialAttributes_2, output=""}
    BaseColor {from=MaterialExpressionConstant_3, output=""}
   ]
   properties {attributeSetTypes [#3]}
  }
  MaterialExpressionSetMaterialAttributes_6 {
   class=MaterialExpressionSetMaterialAttributes
   inputs [
    MaterialAttributes {from=MaterialExpressionBlendMaterialAttributes_5, output=""}
    OpacityMask {from=MaterialExpressionComponentMask_12, output=""}
   ]
   properties {attributeSetTypes [#9]}
  }
  MaterialExpressionSetMaterialAttributes_7 {
   class=MaterialExpressionSetMaterialAttributes
   inputs [
    MaterialAttributes {from=MaterialExpressionGetMaterialAttributes_1, output=MaterialAttributes}
    Roughness {from=MaterialExpressionLinearInterpolate_6, output=""}
    Normal {from=MaterialExpressionLinearInterpolate_7, output=""}
    ShadingModel {from=MaterialExpressionShadingModel_0, output=""}
    Specular {from="", output=""}
    EmissiveColor {from="", output=""}
   ]
   properties {attributeSetTypes [
     #1
     #2
     #8
     #4
     #5
    ]}
  }
  MaterialExpressionSetMaterialAttributes_8 {
   class=MaterialExpressionSetMaterialAttributes
   inputs [
    MaterialAttributes {from=MaterialExpressionGetMaterialAttributes_6, output=MaterialAttributes}
    EmissiveColor {from=MaterialExpressionMultiply_20, output=""}
   ]
   properties {attributeSetTypes [#5]}
  }
  MaterialExpressionSetMaterialAttributes_9 {
   class=MaterialExpressionSetMaterialAttributes
   inputs [
    MaterialAttributes {from=MaterialExpressionGetMaterialAttributes_5, output=MaterialAttributes}
    EmissiveColor {from=MaterialExpressionMax_4, output=""}
   ]
   properties {attributeSetTypes [#5]}
  }
  MaterialExpressionShaderStageSwitch_1 {
   class=MaterialExpressionShaderStageSwitch
   inputs [
    PixelShader {from=MaterialExpressionVertexInterpolator_1, output=PS}
    VertexShader {from=MaterialExpressionLocalPosition_1, output=XYZ}
   ]
   properties {}
  }
  MaterialExpressionShadingModel_0 {class=MaterialExpressionShadingModel, inputs [], properties {}}
  MaterialExpressionShadingPathSwitch_0 {
   class=MaterialExpressionShadingPathSwitch
   inputs [
    Default {from=MaterialExpressionQualitySwitch_0, output=""}
    Deferred {from=MaterialExpressionQualitySwitch_0, output=""}
    Forward {from=MaterialExpressionReroute_1, output=""}
    Mobile {from=MaterialExpressionReroute_1, output=""}
   ]
   properties {}
  }
  MaterialExpressionSine_1 {class=MaterialExpressionSine, inputs [None {from=MaterialExpressionMaterialFunctionCall_10, output=VGradient}]}
  MaterialExpressionSine_2 {class=MaterialExpressionSine, inputs [None {from=MaterialExpressionMultiply_38}]}
  MaterialExpressionSine_3 {class=MaterialExpressionSine, inputs [None {from=MaterialExpressionAdd_39}]}
  MaterialExpressionSine_4 {class=MaterialExpressionSine, inputs [None {from=MaterialExpressionSubtract_1}]}
  MaterialExpressionSine_5 {class=MaterialExpressionSine, inputs [None {from=MaterialExpressionAdd_44}]}
  MaterialExpressionSkyLightEnvMapSample_0 {
   class=MaterialExpressionSkyLightEnvMapSample
   inputs [
    Direction {from="", output=""}
    Roughness {from="", output=""}
   ]
   properties {}
  }
  MaterialExpressionSmoothStep_1 {
   class=MaterialExpressionSmoothStep
   inputs [
    Min {from="", output=""}
    Max {from="", output=""}
    Value {from=MaterialExpressionAbs_0, output=""}
   ]
   properties {}
  }
  MaterialExpressionSmoothStep_2 {
   class=MaterialExpressionSmoothStep
   inputs [
    Min {from=MaterialExpressionScalarParameter_37, output=""}
    Max {from=MaterialExpressionAdd_14, output=""}
    Value {from=MaterialExpressionDotProduct_3, output=""}
   ]
   properties {}
  }
  MaterialExpressionSmoothStep_3 {
   class=MaterialExpressionSmoothStep
   inputs [
    Min {from=MaterialExpressionScalarParameter_33, output=""}
    Max {from=MaterialExpressionScalarParameter_32, output=""}
    Value {from=MaterialExpressionComponentMask_6, output=""}
   ]
   properties {}
  }
  MaterialExpressionSmoothStep_4 {
   class=MaterialExpressionSmoothStep
   inputs [
    Min {from=MaterialExpressionScalarParameter_67, output=""}
    Max {from=MaterialExpressionScalarParameter_66, output=""}
    Value {from=MaterialExpressionAdd_51, output=""}
   ]
   properties {}
  }
  MaterialExpressionSphereMask_0 {
   class=MaterialExpressionSphereMask
   inputs [
    A {from=MaterialExpressionComponentMask_13, output=""}
    B {from=MaterialExpressionComponentMask_14, output=""}
    Radius {from="", output=""}
    Hardness {from=MaterialExpressionScalarParameter_18, output=""}
   ]
   properties {attenuationRadius=20}
  }
  MaterialExpressionSphereMask_1 {
   class=MaterialExpressionSphereMask
   inputs [
    A {from=MaterialExpressionComponentMask_13, output=""}
    B {from=MaterialExpressionComponentMask_14, output=""}
    Radius {from="", output=""}
    Hardness {from=MaterialExpressionScalarParameter_18, output=""}
   ]
   properties {attenuationRadius=25}
  }
  MaterialExpressionSphereMask_2 {
   class=MaterialExpressionSphereMask
   inputs [
    A {from=MaterialExpressionMaterialFunctionCall_1, output=G}
    B {from=MaterialExpressionScalarParameter_52, output=""}
    Radius {from=MaterialExpressionScalarParameter_50, output=""}
    Hardness {from=MaterialExpressionConstant_10, output=""}
   ]
   properties {}
  }
  MaterialExpressionSphereMask_3 {
   class=MaterialExpressionSphereMask
   inputs [
    A {from=MaterialExpressionReroute_5, output=""}
    B {from=MaterialExpressionVectorParameter_3, output=RGB}
    Radius {from=MaterialExpressionScalarParameter_53, output=""}
    Hardness {from=MaterialExpressionConstant_11, output=""}
   ]
   properties {}
  }
  MaterialExpressionSubtract_0 {
   class=MaterialExpressionSubtract
   inputs [
    A {from=MaterialExpressionMultiply_11}
    B {from=MaterialExpressionFrac_5}
   ]
  }
  MaterialExpressionSubtract_1 {
   class=MaterialExpressionSubtract
   inputs [
    A {from=MaterialExpressionComponentMask_27}
    B {from=MaterialExpressionMultiply_66}
   ]
  }
  MaterialExpressionSubtract_2 {
   class=MaterialExpressionSubtract
   inputs [
    A {from=MaterialExpressionAdd_46}
    B {from=MaterialExpressionTransformPosition_0}
   ]
  }
  MaterialExpressionTextureCoordinate_0 {class=MaterialExpressionTextureCoordinate}
  MaterialExpressionTextureCoordinate_1 {class=MaterialExpressionTextureCoordinate}
  MaterialExpressionTextureCoordinate_2 {class=MaterialExpressionTextureCoordinate}
  MaterialExpressionTextureCoordinate_3 {class=MaterialExpressionTextureCoordinate}
  MaterialExpressionTextureObject_3 {
   class=MaterialExpressionTextureObject
   inputs []
   properties {
    samplerType=SAMPLERTYPE_LinearColor
    texture=/BRCosmetics/Characters/Player/Male/Medium/Bodies/M_MED_FloralMane_Sand/Skins/SuperLevels/Textures/T_SL_S39_T3_TA
   }
  }
  MaterialExpressionTextureObject_4 {
   class=MaterialExpressionTextureObject
   inputs []
   properties {
    samplerType=SAMPLERTYPE_LinearColor
    texture=&T_SL_S39_T3_TA
   }
  }
  MaterialExpressionTextureSampleParameter2D_0 {
   class=MaterialExpressionTextureSampleParameter2D
   inputs [
    UVs {from=MaterialExpressionBumpOffset_2, output=""}
    Level {from=MaterialExpressionConstant_18, output=""}
    Apply View MipBias {from="", output=""}
   ]
   properties {
    group=GlitterAnimated
    mipValueMode=TMVM_MipLevel
    parameterName=Sparkle Mask
    samplerSource=SSM_Wrap_WorldGroupSettings
    samplerType=SAMPLERTYPE_Masks
    sortPriority=0
    texture=/Game/Weapons/WeaponSkins/Textures/T_SparkleMask
   }
  }
  MaterialExpressionTextureSampleParameter2D_1 {
   class=MaterialExpressionTextureSampleParameter2D
   inputs [
    UVs {from="", output=""}
    Apply View MipBias {from="", output=""}
   ]
   properties {
    group=SUPERLEVEL
    parameterName=SuperlevelMask_01
    samplerSource=SSM_Wrap_WorldGroupSettings
    samplerType=SAMPLERTYPE_Masks
    sortPriority=0
    texture=/BRCosmetics/Characters/Player/Male/Medium/Bodies/M_MED_ZenCrash/Textures/T_ZenCrash_Body_SL_FX1
   }
  }
  MaterialExpressionTextureSampleParameter2D_2 {
   class=MaterialExpressionTextureSampleParameter2D
   inputs [
    UVs {from=MaterialExpressionMultiply_11, output=""}
    Level {from=MaterialExpressionConstant_18, output=""}
    Apply View MipBias {from="", output=""}
   ]
   properties {
    group=GlitterAnimated
    mipValueMode=TMVM_MipLevel
    parameterName=Sparkle Mask
    samplerSource=SSM_Wrap_WorldGroupSettings
    samplerType=SAMPLERTYPE_Masks
    sortPriority=0
    texture=&T_SparkleMask
   }
  }
  MaterialExpressionTextureSampleParameter2D_3 {
   class=MaterialExpressionTextureSampleParameter2D
   inputs [
    UVs {from="", output=""}
    Apply View MipBias {from="", output=""}
   ]
   properties {
    group=SUPERLEVEL
    parameterName=SuperLevelsMask_02
    samplerSource=SSM_Wrap_WorldGroupSettings
    samplerType=SAMPLERTYPE_Masks
    sortPriority=0
    texture=/BRCosmetics/Characters/Player/Male/Medium/Bodies/M_MED_ZenCrash/Textures/T_ZenCrash_Body_SL_FX2
   }
  }
  MaterialExpressionTextureSample_1 {
   class=MaterialExpressionTextureSample
   inputs [
    UVs {from=MaterialExpressionMultiply_31}
    Tex {from=""}
    Apply View MipBias {from=""}
   ]
   properties {samplerSource=SSM_Wrap_WorldGroupSettings, samplerType=SAMPLERTYPE_Masks, texture=/Game/Weapons/WeaponSkins/Textures/T_GlitterDirectionMask}
  }
  MaterialExpressionTextureSample_11 {
   class=MaterialExpressionTextureSample
   inputs [
    UVs {from=MaterialExpressionFrac_1}
    Tex {from=""}
    Apply View MipBias {from=""}
   ]
   properties {
    samplerSource=SSM_Wrap_WorldGroupSettings
    texture=/BRCosmetics/Characters/Player/Male/Medium/Bodies/M_MED_FloralMane_Sand/Skins/SuperLevels/Textures/T_S39_Superlevels_SmokeGradient
   }
  }
  MaterialExpressionTextureSample_12 {
   class=MaterialExpressionTextureSample
   inputs [
    UVs {from=MaterialExpressionMultiply_78}
    Tex {from=""}
    Apply View MipBias {from=""}
   ]
   properties {
    samplerSource=SSM_Wrap_WorldGroupSettings
    texture=/BRCosmetics/Characters/Player/Male/Medium/Bodies/M_MED_FloralMane_Sand/Skins/SuperLevels/Textures/T_TropicalSL_T3_Sparkles
   }
  }
  MaterialExpressionTextureSample_15 {
   class=MaterialExpressionTextureSample
   inputs [
    UVs {from=MaterialExpressionPanner_4}
    Tex {from=MaterialExpressionTextureObject_4}
    Apply View MipBias {from=""}
   ]
   properties {samplerSource=SSM_Wrap_WorldGroupSettings, samplerType=SAMPLERTYPE_Masks}
  }
  MaterialExpressionTextureSample_2 {
   class=MaterialExpressionTextureSample
   inputs [
    UVs {from=MaterialExpressionPanner_0}
    Tex {from=MaterialExpressionTextureObject_4}
    Level {from=MaterialExpressionConstant_20}
    Apply View MipBias {from=""}
   ]
   properties {mipValueMode=TMVM_MipLevel, samplerSource=SSM_Wrap_WorldGroupSettings, samplerType=SAMPLERTYPE_Masks}
  }
  MaterialExpressionTime_10 {class=MaterialExpressionTime}
  MaterialExpressionTime_2 {class=MaterialExpressionTime}
  MaterialExpressionTime_3 {class=MaterialExpressionTime}
  MaterialExpressionTime_4 {class=MaterialExpressionTime}
  MaterialExpressionTime_5 {class=MaterialExpressionTime}
  MaterialExpressionTime_6 {class=MaterialExpressionTime}
  MaterialExpressionTime_7 {class=MaterialExpressionTime}
  MaterialExpressionTime_8 {class=MaterialExpressionTime}
  MaterialExpressionTime_9 {class=MaterialExpressionTime}
  MaterialExpressionTransformPosition_0 {
   class=MaterialExpressionTransformPosition
   inputs [None {from=MaterialExpressionCameraPositionWS_0, output=""}]
   properties {transformSourceType=TRANSFORMPOSSOURCE_World}
  }
  MaterialExpressionTransform_0 {
   class=MaterialExpressionTransform
   inputs [None {from=MaterialExpressionNamedRerouteUsage_45}]
   properties {transformSourceType=TRANSFORMSOURCE_Local}
  }
  MaterialExpressionTransform_3 {
   class=MaterialExpressionTransform
   inputs [None {from=MaterialExpressionNormalize_2}]
   properties {transformSourceType=TRANSFORMSOURCE_World, transformType=TRANSFORM_Tangent}
  }
  MaterialExpressionVectorParameter_0 {class=MaterialExpressionVectorParameter, properties {parameterName=PrimaryElement_AdditionalHeight}}
  MaterialExpressionVectorParameter_1 {class=MaterialExpressionVectorParameter, properties {parameterName=Primary_PanningSpeed, defaultValue=<rgba 0.015 0.0035 0 1>}}
  MaterialExpressionVectorParameter_2 {class=MaterialExpressionVectorParameter, properties {group=Dance Lines, parameterName=offset, defaultValue=<rgba 0 0 0 1>}}
  MaterialExpressionVectorParameter_3 {class=MaterialExpressionVectorParameter, properties {parameterName=ObjectSpaceGradStart, defaultValue=<rgba 0 40 150 1>}}
  MaterialExpressionVectorParameter_7 {class=MaterialExpressionVectorParameter, properties {parameterName=Sepia Sparkle Overtone, defaultValue=<rgba 1 0.9512 0.901 1>}}
  MaterialExpressionVertexInterpolator_1 {class=MaterialExpressionVertexInterpolator, inputs [VS {from=MaterialExpressionLocalPosition_1, output=XYZ}], properties {}}
  MaterialExpressionVertexInterpolator_2 {class=MaterialExpressionVertexInterpolator, inputs [VS {from=MaterialExpressionPreSkinnedNormal_1, output=""}], properties {}}
  MaterialExpressionVertexNormalWS_1 {class=MaterialExpressionVertexNormalWS, inputs [], properties {}}
  MaterialExpressionVertexNormalWS_2 {class=MaterialExpressionVertexNormalWS, inputs [], properties {}}
  MaterialExpressionVertexNormalWS_3 {class=MaterialExpressionVertexNormalWS, inputs [], properties {}}
 }}
