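# Parent material for the tier 3 Super Level fur skin of the FloralMane Sand character body, supporting multiple shading models.
# Character PreFX attributes feed the deprecated fur function, toggled by UseFNHair2_0 (on by default).
# On mobile, a brightened strands-mask green channel darkens base color instead.
# The hair mask blends either result over the base, then MF_SL_S39_T3 applies the skin effect before character PostFX.
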Material /BRCosmetics/Characters/Player/Male/Medium/Bodies/M_MED_FloralMane_Sand/Skins/SuperLevels/ParentMaterials/M_SL_Tropical_FurParent_T3
graph {
 properties {
  bUseMaterialAttributes=true
  bUsedWithMeshDeformer=true
  bUsedWithMorphTargets=true
  bUsedWithSkeletalMesh=true
  pixelDepthOffsetMode=PDOM_Legacy
  shadingModel=MSM_FromMaterialExpression
  usedShadingModels=Default Lit | Subsurface | Clear Coat | Subsurface Profile | Cloth
 }
 nodes [
  n1 = MaterialFunctionCall() {materialFunction=&MF_CharacterParent_2023_PreFX}
  n2 = MaterialFunctionCall!(In=@n1.Result, Mask Input=@n1.Result) {materialFunction=/Game/Global/Materials/Character_Parent_2023/Fur_Parent_2023/MF_Fur_Deprecated}
  n3 = GetMaterialAttributes(None=@n1.Result) {attributeGetTypes [#1]}
  n4 = TextureSample() {
   samplerSource=SSM_Wrap_WorldGroupSettings
   samplerType=SAMPLERTYPE_Masks
   texture=/BRCosmetics/Characters/Player/Male/Medium/Bodies/M_MED_FloralMane_Sand/Textures/T_FloralMane_Sand_Body_Strands
  }
  n5 = Saturate(None=Add(A=Power(Base=@n4.G) {constExponent=0.5}) {constB=0.05})
  n6 = ShadingPathSwitch(
   Default=StaticSwitchParameter(True=@n2.Result, False=@n1.Result) {defaultValue=true, group=Skin FX, parameterName=UseFNHair2_0}
   Mobile=SetMaterialAttributes!(MaterialAttributes=@n3.MaterialAttributes, Base Color=Multiply(A=@n3.Base Color, B=@n5)) {attributeSetTypes [#1]}
  )
  MaterialFunctionCall(
   Material=MaterialFunctionCall!(
    Material=BlendMaterialAttributes(A=@n1.Result, B=@n6, Alpha=@n2.HairMask)
    M=@n1.M
    PreskinnedLocPos=@n1.Pre skinned Local Position
    Tech Art Mask=@n1.Tech Art Mask
    Fresnel=@n1.Pre skinned Local Position
   ) {materialFunction=&/MF_SL_S39_T3}.Result
   M=@n1.M
   Pre Skinned Local Position=@n1.Pre skinned Local Position
   Distance Blend=@n1.Pre skinned Local Position
   Fresnel=@n1.Pre skinned Local Position
  ) {materialFunction=&MF_CharacterParent_2023_PostFX}
 ]
}
