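# Parent character material for the Tropical Super Level skin's hair, wrapping recolor and glow layers between the 2023 character PreFX and PostFX functions.
# Switchable hair and flyaway-hair functions feed the original attributes.
# Chained attribute blends, driven by Super Level mask channels, add gold and blue metals, darkened areas and orange emissive.
# Blue hair glow combines mask, inverted fresnel and gradient terms, while IsRowLiaison_Cast can fade out the first mask texture.
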Material /BRCosmetics/Characters/Player/Male/Medium/Bodies/M_MED_FloralMane_Sand/Skins/SuperLevels/ParentMaterials/M_SL_Tropical_HairParent_T1
graph {
 properties {
  bUseMaterialAttributes=true
  bUsedWithMeshDeformer=true
  bUsedWithSkeletalMesh=true
  pixelDepthOffsetMode=PDOM_Legacy
  shadingModel=MSM_FromMaterialExpression
  usedShadingModels=Default Lit | Subsurface | Clear Coat | Subsurface Profile | Cloth
 }
 nodes [
  n1 = MaterialFunctionCall() {materialFunction=&MF_CharacterParent_2023_PreFX}
  MaterialFunctionCall(Material=NamedRerouteUsage(), M=@n1.M, Pre Skinned Local Position=@n1.M, Distance Blend=@n1.M, Fresnel=@n1.M) {materialFunction=&MF_CharacterParent_2023_PostFX}
  NamedRerouteDeclaration(None=@n1.Pre skinned Local Position) {name=PreskinnedLocalPosition, nodeColor=<rgba 1 0.8672 0 1>}
  n2 = MaterialFunctionCall!(In=@n1.Result, Mask Input=@n1.Result) {materialFunction=/Game/Global/Materials/Character_Parent_2023/Hair_Parent_2023/MaterialFunctions/MF_FN_Hair_2023}
  n3 = BlendMaterialAttributes(
   A=@n1.Result
   B=StaticSwitchParameter(True=@n2.Result, False=@n1.Result) {defaultValue=true, group=Skin FX, parameterName=UseFNHair2_0}
   Alpha=@n2.HairMask
  )
  n4 = MaterialFunctionCall!(MaterialInput=@n3) {materialFunction=/Game/Global/Materials/Character_Parent_2023/Hair_Parent_2023/MaterialFunctions/MF_FlyAwayHair_2023}
  NamedRerouteDeclaration(None=SetMaterialAttributes!(MaterialAttributes=StaticSwitchParameter(True=@n4.Result, False=@n3) {group=Skin FX, parameterName=IsFlyAwayHair?}, Base Color=Constant()) {attributeSetTypes [#1]}) {name=OriginalMA, nodeColor=<rgba 0.5469 0 1 1>}
  NamedRerouteDeclaration(None=@n1.Tech Art Mask) {name=InTechArtMask, nodeColor=<rgba 0.1484 0 1 1>}
  NamedRerouteDeclaration(None=@n1.M) {name=InMasks, nodeColor=<rgba 0 1 0.2266 1>}
  n5 = TextureSampleParameter2D() {
   group=SUPERLEVEL
   parameterName=SuperlevelMask_01
   samplerSource=SSM_Wrap_WorldGroupSettings
   samplerType=SAMPLERTYPE_Masks
   sortPriority=0
   texture=/BRCosmetics/Characters/Player/Male/Medium/Bodies/M_MED_ZenCrash/Textures/T_ZenCrash_Body_SL_FX1
  }
  NamedRerouteDeclaration(None=LinearInterpolate(A=@n5.RGB, B=Constant(), Alpha=ScalarParameter() {parameterName=IsRowLiaison_Cast})) {name=SuperlevelMasks01, nodeColor=<rgba 1 0.1931 0 1>}
  n6 = NamedRerouteUsage()
  n7 = GetMaterialAttributes(None=@n6) {attributeGetTypes [#2]}
  NamedRerouteDeclaration(None=@n7.Roughness) {name=OriginalRoughness, nodeColor=<rgba 0 1 0.5313 1>}
  NamedRerouteDeclaration(None=@n2.HairMask) {name=HairMask, nodeColor=<rgba 0.4297 0 1 1>}
  NamedRerouteDeclaration(None=@n1.Fresnel) {name=InFresnel, nodeColor=<rgba 0 1 0.8594 1>}
  n8 = BlendMaterialAttributes(
   A=@n6
   B=SetMaterialAttributes!(
    MaterialAttributes=@n7.MaterialAttributes
    Base Color=Constant3Vector() {constant=<rgba 0.7795 0.3123 0.0003 1>}
    Metallic=Constant() {r=0.6}
    Roughness=Min(A=@n7.Roughness) {constB=0.45}
   ) {attributeSetTypes [#1, #3, #2]}
   Alpha=ComponentMask(None=NamedRerouteUsage()) {g=true}
  )
  n9 = DotProduct(A=NamedRerouteUsage(), B=Constant3Vector() {constant=<rgba 1 1 1 0>})
  n10 = BlendMaterialAttributes(
   A=@n8
   B=SetMaterialAttributes!(MaterialAttributes=@n8, Base Color=Constant() {r=0.025}) {attributeSetTypes [#1]}
   Alpha=OneMinus(None=Saturate(None=@n9))
  )
  n11 = ComponentMask(None=NamedRerouteUsage()) {b=true}
  n12 = ScalarParameter() {defaultValue=0.05, group=SUPERLEVEL, parameterName=SecondaryElement_Hardness, sortPriority=30}
  n13 = ScalarParameter() {defaultValue=-110, group=SUPERLEVEL, parameterName=SecondaryElement_Height, sortPriority=30}
  n14 = Saturate(None=Add(A=Multiply(A=@n11, B=@n12) {constB=0.1}, B=Multiply(A=@n12, B=@n13)) {constB=-12})
  n15 = OneMinus(None=@n14)
  n16 = ScalarParameter() {group=SUPERLEVEL, parameterName=SecondaryElement_InvertMask, sortPriority=31}
  n17 = NamedRerouteUsage()
  n18 = ScalarParameter() {group=SUPERLEVEL, parameterName=SecondaryElement_RoughnessOffset, sortPriority=30}
  n19 = SetMaterialAttributes!(
   MaterialAttributes=@n10
   Base Color=LinearInterpolate(
    A=Constant3Vector() {constant=<rgba 0.0208 0.0448 0.0677 1>}
    B=Constant3Vector() {constant=<rgba 0.0058 0.1729 0.3316 1>}
    Alpha=LinearInterpolate(A=@n15, B=@n14, Alpha=@n16)
   )
   Metallic=Constant() {r=0.5}
   Roughness=Saturate(None=Add(A=@n17, B=@n18))
   Emissive Color
  ) {attributeSetTypes [
    #1
    #3
    #2
    #4
   ]}
  n20 = BlendMaterialAttributes(A=@n10, B=@n19, Alpha=ComponentMask(None=NamedRerouteUsage()) {b=true})
  n21 = GetMaterialAttributes(None=@n20) {attributeGetTypes [#1]}
  n22 = Multiply(A=Constant3Vector() {constant=<rgba 0.3762 0.5517 1.1 1>}, B=ScalarParameter() {defaultValue=5, parameterName=HairEmissive_Multiplier}) {constB=4}
  n23 = Max(A=ComponentMask(None=NamedRerouteUsage()) {g=true}) {constB=0.1}
  n24 = NamedRerouteUsage()
  n25 = ComponentMask(None=@n24) {r=true}
  n26 = NamedRerouteUsage()
  n27 = Multiply(A=Multiply(A=@n25, B=@n26), B=ScalarParameter() {defaultValue=1, group=SUPERLEVEL, parameterName=HairEmissive_BaseMultiplier})
  n28 = VertexNormalWS()
  n29 = MaterialFunctionCall() {materialFunction=&CameraDirectionVector}
  n30 = Power(
   Base=ConstantBiasScale(None=DotProduct(A=@n28, B=@n29.Result))
   Exp=ScalarParameter() {defaultValue=0.1, group=SUPERLEVEL, parameterName=HairFresnelPower}
  )
  n31 = Multiply(A=@n26, B=Multiply(A=OneMinus(None=@n30), B=ScalarParameter() {defaultValue=1, group=SUPERLEVEL, parameterName=HairFresnelIntensity}))
  n32 = ScalarParameter() {defaultValue=140, group=SUPERLEVEL, parameterName=Hair_GradientEnd, sortPriority=28}
  n33 = ScalarParameter() {group=SUPERLEVEL, parameterName=Hair_GradientStart, sortPriority=28}
  n34 = ComponentMask(None=NamedRerouteUsage()) {b=true}
  n35 = Add(
   A=Saturate(None=Add(A=@n27, B=@n31))
   B=Multiply(A=ScalarParameter() {group=SUPERLEVEL, parameterName=HairGradient_Multiplier, sortPriority=28}, B=SmoothStep(Min=@n32, Max=@n33, Value=@n34))
  )
  n36 = BlendMaterialAttributes(
   A=@n20
   B=SetMaterialAttributes!(MaterialAttributes=@n21.MaterialAttributes, Emissive Color=Multiply(A=@n22, B=@n23), Base Color=Multiply(A=@n21.Base Color) {constB=0.1}) {attributeSetTypes [#4, #1]}
   Alpha=@n35
  )
  n37 = NamedRerouteUsage()
  n38 = ComponentMask(None=@n37) {r=true}
  n39 = ComponentMask(None=@n37) {b=true}
  n40 = BlendMaterialAttributes(
   A=@n36
   B=SetMaterialAttributes!(MaterialAttributes=@n36, Emissive Color=Constant3Vector() {constant=<rgba 0.7865 0.183 0 1>}) {attributeSetTypes [#4]}
   Alpha=Saturate(None=Subtract(A=@n38, B=@n39))
  )
  n41 = Saturate(None=DotProduct(A=NamedRerouteUsage(), B=Constant3Vector() {constant=<rgba 1 1 1 0>}))
  n42 = Saturate(None=ComponentMask(None=NamedRerouteUsage()) {b=true})
  n43 = BlendMaterialAttributes(
   A=@n40
   B=SetMaterialAttributes!(MaterialAttributes=@n40, Base Color=Constant() {r=0.025}) {attributeSetTypes [#1]}
   Alpha=Multiply(A=OneMinus(None=@n41), B=OneMinus(None=@n42))
  )
  n44 = BlendMaterialAttributes(
   A=@n43
   B=SetMaterialAttributes!(MaterialAttributes=@n43, Base Color=Constant3Vector() {constant=<rgba 0.0983 0.2603 0.6823 1>}, Metallic=Constant() {r=1}) {attributeSetTypes [#1, #3]}
   Alpha=ComponentMask(None=NamedRerouteUsage()) {g=true}
  )
  n45 = ComponentMask(None=NamedRerouteUsage()) {g=true}
  n46 = Constant3Vector() {constant=<rgba 0.0257 0.0823 0.2691 1>}
  n47 = Saturate(None=Multiply(A=@n35) {constB=4})
  n48 = SetMaterialAttributes!(MaterialAttributes=@n44, Base Color=Multiply(A=Max(A=@n45) {constB=0.25}, B=Desaturation(None=@n46, Fraction=@n47))) {attributeSetTypes [#1]}
  n49 = NamedRerouteUsage()
  n50 = NamedRerouteUsage()
  NamedRerouteDeclaration(None=SetMaterialAttributes!(MaterialAttributes=BlendMaterialAttributes(A=@n44, B=@n48, Alpha=@n49), Opacity Mask=ComponentMask(None=@n50) {r=true}) {attributeSetTypes [#5]}) {name=OutMA, nodeColor=<rgba 1 0.5391 0 1>}
  n51 = TextureSampleParameter2D() {
   group=SUPERLEVEL
   parameterName=SuperLevelsMask_02
   samplerType=SAMPLERTYPE_Masks
   sortPriority=0
   texture=/BRCosmetics/Characters/Player/Male/Medium/Bodies/M_MED_ZenCrash/Textures/T_ZenCrash_Body_SL_FX2
  }
  NamedRerouteDeclaration(None=@n51.RGB) {name=SuperlevelMasks02, nodeColor=<rgba 0.2188 0 1 1>}
  n52 = NamedRerouteUsage()
  Saturate(None=Multiply(A=ComponentMask(None=@n52) {b=true}) {constB=2})
  Saturate(None=Subtract(A=@n25, B=ComponentMask(None=@n24) {b=true}))
 ]
}
