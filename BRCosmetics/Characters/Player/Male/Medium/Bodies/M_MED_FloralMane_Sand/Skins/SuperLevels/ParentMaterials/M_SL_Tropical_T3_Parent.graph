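# Parent material for the Tropical Tier 3 Super Level skins of the FloralMane Sand character, supporting several shading models through material attributes.
# The main chain runs the shared character PreFX function, the MF_SL_S39_T3 effect, then PostFX.
# A separate named-reroute network builds random normal glitter, skylight reflection glitter, a height fade and emissive-reactive sparkle.
# That glitter network has no visible connection into the main chain.
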Material /BRCosmetics/Characters/Player/Male/Medium/Bodies/M_MED_FloralMane_Sand/Skins/SuperLevels/ParentMaterials/M_SL_Tropical_T3_Parent
graph {
 properties {
  bUseMaterialAttributes=true
  bUsedWithClothing=true
  bUsedWithMeshDeformer=true
  bUsedWithMorphTargets=true
  bUsedWithNanite=true
  bUsedWithNiagaraMeshParticles=true
  bUsedWithSkeletalMesh=true
  pixelDepthOffsetMode=PDOM_Legacy
  shadingModel=MSM_FromMaterialExpression
  usedShadingModels=Default Lit | Subsurface | Clear Coat | Subsurface Profile | Cloth
 }
 nodes [
  n1 = MaterialFunctionCall() {materialFunction=&MF_CharacterParent_2023_PreFX}
  MaterialFunctionCall(
   Material=MaterialFunctionCall!(Material=@n1.Result, M=@n1.Result, PreskinnedLocPos=@n1.Result, Tech Art Mask=@n1.Result, Fresnel=@n1.Result) {materialFunction=&/MF_SL_S39_T3}.Result
   M=@n1.M
   Pre Skinned Local Position=@n1.M
   Distance Blend=@n1.M
   Fresnel=@n1.M
  ) {materialFunction=&MF_CharacterParent_2023_PostFX}
  n2 = TextureSample(UVs=Multiply(A=TextureCoordinate(), B=ScalarParameter() {defaultValue=25, group=CanineCronutDig_Ware, parameterName=Random Glitter - Tiling})) {samplerType=SAMPLERTYPE_Masks, texture=/Game/Weapons/WeaponSkins/Textures/T_GlitterDirectionMask}
  n3 = Multiply(
   A=Add(A=Multiply(A=@n2.RGB) {constB=2}) {constB=-1}
   B=ScalarParameter() {defaultValue=1, group=CanineCronutDig_Ware, parameterName=Random Glitter - Strength, sliderMax=2, sortPriority=6}
  ) {constB=0.5}
  n4 = Normalize(VectorInput=Add(A=NamedRerouteDeclaration(None=@n3) {name=Random Normal Dir, nodeColor=<rgba 0 0.3501 1 1>}, B=VertexNormalWS()))
  NamedRerouteDeclaration(None=Transform(None=@n4) {transformSourceType=TRANSFORMSOURCE_World, transformType=TRANSFORM_Tangent}) {name=Sparkle Normal, nodeColor=<rgba 0.7812 1 0 1>}
  n5 = DotProduct(A=VertexNormalWS(), B=AtmosphericLightVector())
  NamedRerouteDeclaration(None=Multiply(A=Saturate(None=@n5), B=NamedRerouteUsage())) {name=LDotN, nodeColor=<rgba 0 0.2031 1 1>}
  n6 = DotProduct(A=@n4, B=CameraVectorWS())
  n7 = ScalarParameter() {defaultValue=1, group=CanineCronutDig_Ware, parameterName=Random Glitter - Exp}
  NamedRerouteDeclaration(None=Saturate(None=Power(Base=@n6, Exp=@n7))) {name=Random Glitter, nodeColor=<rgba 0 1 0.4141 1>}
  n8 = ScalarParameter() {defaultValue=0.9, group=CanineCronutDig_Ware, parameterName=Glitter - Transition Amount}
  n9 = ScalarParameter() {defaultValue=0.15, group=CanineCronutDig_Ware, parameterName=Glitter - Transition Fade}
  n10 = Normalize(VectorInput=ReflectionVectorWS())
  n11 = Normalize(VectorInput=Add(A=NamedRerouteUsage(), B=VertexNormalWS())) {desc=GlitterDirection}
  NamedRerouteDeclaration(None=SmoothStep(Min=@n8, Max=Add(A=@n8, B=@n9), Value=DotProduct(A=@n10, B=@n11))) {name=Reflection Glitter - Mask, nodeColor=<rgba 0 0.2266 1 1>}
  n12 = NamedRerouteUsage()
  n13 = SkyLightEnvMapSample()
  n14 = Desaturation(None=@n13, Fraction=ScalarParameter() {defaultValue=1, group=CanineCronutDig_Ware, parameterName=Reflection Glitter - Desaturation})
  NamedRerouteDeclaration(None=Multiply(
    A=Multiply(A=@n12, B=@n14)
    B=ScalarParameter() {defaultValue=0.1, group=CanineCronutDig_Ware, parameterName=Global Fake Reflection - Intensity}
   )) {name=Reflection Glitter, nodeColor=<rgba 0.5469 1 0 1>}
  n15 = Divide(A=OneMinus(None=LocalPosition().Z), B=ScalarParameter() {defaultValue=149.8988, group=CanineCronutDig_Ware, parameterName=Sparkle - Fade})
  n16 = Divide(A=ComponentMask(None=PreSkinnedLocalBounds().Half Extents) {b=true}) {constB=100}
  NamedRerouteDeclaration(None=Saturate(None=Add(A=@n15, B=@n16))) {name=Glitter - Top Down Fade, nodeColor=<rgba 1 0 0.8672 1>}
  n17 = GetMaterialAttributes(None=NamedRerouteUsage()) {attributeGetTypes [#1]}
  n18 = NamedRerouteUsage()
  n19 = Desaturation(None=@n17.Emissive Color)
  n20 = ScalarParameter() {defaultValue=4.5, group=CanineCronutDig_Ware, parameterName=Reactive Sparkle - Exp}
  n21 = Multiply(
   A=Multiply(A=@n17.Emissive Color, B=Power(Base=@n19, Exp=@n20))
   B=ScalarParameter() {defaultValue=0.1, group=CanineCronutDig_Ware, parameterName=Reactive Sparkle - Min}
  )
  n22 = LinearInterpolate(
   A=Multiply(A=Add(A=@n17.Emissive Color, B=@n21), B=ScalarParameter() {defaultValue=0.001, parameterName=adsaadsafd})
   Alpha=NamedRerouteUsage()
  ) {constB=0}
  NamedRerouteDeclaration(None=SetMaterialAttributes(MaterialAttributes=@n17.MaterialAttributes, Emissive Color=Max(A=@n18, B=@n22)) {attributeSetTypes [#1]}) {name=Reactive Sparkle, nodeColor=<rgba 0 1 0.8125 1>}
  NamedRerouteDeclaration(None=@n13) {name=Sky Light Color, nodeColor=<rgba 0 1 0.9062 1>}
 ]
}
